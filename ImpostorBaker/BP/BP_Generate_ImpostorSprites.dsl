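; Impostor baker actor Blueprint that sets up and processes render targets for baking impostor sprite maps and a card mesh.
; It allocates power-of-two atlas targets, a scene-capture mip chain and scratch targets, then derives the list of maps to save.
; Material draws resample targets, pack scalar masks into RGBA channels and composite custom lighting into base color.
; Geometry Script builds the card from vertex, triangle and UV arrays, then writes it to a static mesh LOD.
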(fn BakeDynamicMesh
  (StaticMesh|AddMaterial (Variables|/Script/Engine.StaticMeshComponent|GetStaticMesh (Variables|/Script/Engine.StaticMeshActor|GetStaticMeshComponent (Variables|Default|GetStatic%20Mesh%20Actor))) (Variables|Debug|GetCreated%20MIC))
  (bind _returnvalue (DynamicMeshComponent|GetDynamicMesh (Variables|Default|GetDynamicMeshComponent)))
  (bind _output_get (Variables|Default|SetDynamicMesh _returnvalue))
  (GeometryScript|Materials|ClearMaterialIDs _output_get (Utilities|Array|LastIndex (Variables|/Script/Engine.StaticMesh|GetStaticMaterials (Variables|/Script/Engine.StaticMeshComponent|GetStaticMesh (Variables|/Script/Engine.StaticMeshActor|GetStaticMeshComponent (Variables|Default|GetStatic%20Mesh%20Actor))))))
  (bind _returnvalue_2 (StaticMesh|GetStaticMeshDescription (Variables|/Script/Engine.StaticMeshComponent|GetStaticMesh (Variables|/Script/Engine.StaticMeshActor|GetStaticMeshComponent (Variables|Default|GetStatic%20Mesh%20Actor))) (- (StaticMesh|GetNumLODs (Variables|/Script/Engine.StaticMeshComponent|GetStaticMesh (Variables|/Script/Engine.StaticMeshActor|GetStaticMeshComponent (Variables|Default|GetStatic%20Mesh%20Actor)))) 1)))
  (Development|PrintString (Utilities|String|Append "Triangles/Verts: " (Utilities|String|ToString(Integer) (MeshDescription|GetTriangleCount _returnvalue_2)) " / " (Utilities|String|ToString(Integer) (MeshDescription|GetVertexCount _returnvalue_2))))
  (GeometryScript|StaticMesh|CopyMeshtoStaticMesh (Variables|Default|GetDynamicMesh) (Variables|/Script/Engine.StaticMeshComponent|GetStaticMesh (Variables|/Script/Engine.StaticMeshActor|GetStaticMeshComponent (Variables|Default|GetStatic%20Mesh%20Actor))) (Utilities|Struct|MakeGeometryScriptCopyMeshtoAssetOptions true true) (Utilities|Struct|MakeGeometryScriptMeshWriteLOD false (StaticMesh|GetNumLODs (Variables|/Script/Engine.StaticMeshComponent|GetStaticMesh (Variables|/Script/Engine.StaticMeshActor|GetStaticMeshComponent (Variables|Default|GetStatic%20Mesh%20Actor)))))))

(fn ChannelPackMasks
  (bind _targetmaps (Variables|RenderTargets|GetTargetMaps))
  (bind (_value _returnvalue) (Utilities|Map|Find _targetmaps "NewEnumerator0"))
  (bind _targetmaps_2 (Variables|RenderTargets|GetTargetMaps))
  (bind _channelpackedmaps (Variables|RenderTargets|GetChannelPackedMaps))
  (bind _channelpackmasksmid (Variables|MIDs|GetChannelPackMasksMID))
  (bind _keys (Utilities|Map|Keys (Variables|RenderTargets|GetChannelPackedMaps)))
  (for _array_element _keys
    (for (_array_element_2 _array_index) (Utilities|Array|MakeArray (Default|ScalarMaptoBakeType (Utilities|Struct|BreakImpostorChannelPackSettings _array_element)) (Default|ScalarMaptoBakeType (.G_11_B3F571D648FCD1B25D023996A7D395BB (Utilities|Struct|BreakImpostorChannelPackSettings _array_element))) (Default|ScalarMaptoBakeType (.B_12_71D42C9349B0E42D3950BE82BCAE1C44 (Utilities|Struct|BreakImpostorChannelPackSettings _array_element))) (Default|ScalarMaptoBakeType (.A_13_655F3E8A4BC083A5432B43AE8BAC9BDE (Utilities|Struct|BreakImpostorChannelPackSettings _array_element))))
      (if (or (.ReturnValue (Utilities|Map|Find _targetmaps_2 _array_element_2)) _returnvalue)
        (Rendering|Material|SetTextureParameterValue _channelpackmasksmid (Utilities|String|StringToName (Utilities|String|Append "RT" (Utilities|String|ToString(Integer) (+ _array_index 1)))) (select (Utilities|Enum|Equal(Enum) _array_element_2 "NewEnumerator12") _value (Utilities|Map|Find _targetmaps_2 _array_element_2)))
        (Rendering|Material|SetScalarParameterValue _channelpackmasksmid (Utilities|String|StringToName (Utilities|String|Append (Utilities|String|Append "RT" (Utilities|String|ToString(Integer) (+ _array_index 1))) "Alpha")) (Math|Conversions|ToFloat(Boolean) (Utilities|Enum|Equal(Enum) _array_element_2 "NewEnumerator12")) :!class /Script/Engine.MaterialInstanceDynamic)))
    (Rendering|ClearRenderTarget2D (Utilities|Map|Find _channelpackedmaps _array_element))
    (Rendering|DrawMaterialtoRenderTarget (Utilities|Map|Find _channelpackedmaps _array_element) (Variables|MIDs|GetChannelPackMasksMID))))

(fn ClearRenderTargets
  (Rendering|ClearRenderTarget2D (Variables|RenderTargets|GetCombinedAlphasRT))
  (Rendering|ClearRenderTarget2D (Variables|RenderTargets|GetScratchRT))
  (for _array_element (Variables|RenderTargets|GetSceneCaptureMipChain)
    (Rendering|ClearRenderTarget2D _array_element))
  (bind _values (Utilities|Map|Values (Variables|RenderTargets|GetTargetMaps)))
  (for _array_element_2 _values
    (Rendering|ClearRenderTarget2D _array_element_2)))

(fn GenerateMeshData)

(fn ResampleRT
  (params (Source%20RT :type Texture%20Render%20Target%202D%20Object%20Reference) (Dest%20RT :type Texture%20Render%20Target%202D%20Object%20Reference))
  (Rendering|Material|SetTextureParameterValue (Variables|MIDs|GetResample%20MID) "RT" Source%20RT)
  (Rendering|ClearRenderTarget2D Dest%20RT)
  (Rendering|DrawMaterialtoRenderTarget Dest%20RT (Variables|MIDs|GetResample%20MID)))

(fn SetupRTandSaveList
  (bind _scene_capture_resolution (Variables|Advanced|GetScene%20Capture%20Resolution))
  (bind _returnvalue (Math|Conversions|ToFloat(Integer) _scene_capture_resolution))
  (bind _returnvalue_2 (Math|Float|Log _returnvalue 1.99999))
  (bind _returnvalue_3 (Math|Float|Floor _returnvalue_2))
  (bind _returnvalue_4 (Math|Integer|Clamp(Integer) _returnvalue_3 7 11))
  (bind _curmaptype (LocalVariables|GetCurMapType))
  (bind _targetmaps (Variables|RenderTargets|GetTargetMaps))
  (bind (_value _returnvalue_5) (Utilities|Map|Find _targetmaps _curmaptype))
  (bind _mipidx (LocalVariables|GetMipIdx))
  (bind _returnvalue_6 (+ _mipidx 0))
  (bind _returnvalue_7 (Math|Conversions|ToFloat(Integer) _returnvalue_6))
  (bind _returnvalue_8 (Math|Float|Power 2 _returnvalue_7))
  (bind _returnvalue_9 (Math|Float|Truncate _returnvalue_8))
  (bind _scene_capture_resolution_2 (Variables|Advanced|GetScene%20Capture%20Resolution))
  (bind _returnvalue_10 (/ _scene_capture_resolution_2 _returnvalue_9))
  (bind _returnvalue_11 (- _returnvalue_4 4))
  (bind _mipidx_2 (LocalVariables|GetMipIdx))
  (bind _scenecapturemipchain (Variables|RenderTargets|GetSceneCaptureMipChain))
  (bind _output (Utilities|Array|Get(acopy) _scenecapturemipchain _mipidx_2))
  (bind _maps_to_save (Variables|Default|GetMaps%20to%20Save))
  (bind _maps_to_save_2 (Variables|Default|GetMaps%20to%20Save))
  (bind _channelpackedmaps (Variables|RenderTargets|GetChannelPackedMaps))
  ;[0.3122 0.9801 1] Resolution Setup
  (Variables|Advanced|SetAtlas_Res_Per_Frame (/ (Variables|Advanced|GetResolution) (Variables|Default|GetFrames%20XY))) ; Get Resolution Per Final Frame
  (Variables|Advanced|SetScene%20Capture%20Resolution (Math|Integer|Max(Integer) (Variables|Advanced|GetScene%20Capture%20Resolution) (Variables|Advanced|GetAtlas_Res_Per_Frame))) ; Scene Capture Res needs to be at least\nas large as the Final Sub Frame Res
  (Variables|Advanced|SetResolution (Math|Float|Truncate (Math|Float|Power 2 (Math|Conversions|ToFloat(Integer) (Math|Integer|Clamp(Integer) (Math|Float|Floor (Math|Float|Log (Math|Conversions|ToFloat(Integer) (Variables|Advanced|GetResolution)) 1.99999)) 9 14)))))
  (Variables|Advanced|SetScene%20Capture%20Resolution (Math|Float|Truncate (Math|Float|Power 2 (Math|Conversions|ToFloat(Integer) _returnvalue_4))))
  (Variables|Advanced|SetSceneCaptureMips _returnvalue_4)
  (Variables|Advanced|SetCutoutMipTarget _returnvalue_11) ; Used to Optimize DF generation
  (Variables|Advanced|SetDFMipTarget (+ _returnvalue_11 2))
  (Variables|Advanced|SetSceneCaptureMipToComposite (- (Math|Float|Floor (Math|Float|Log (Math|Conversions|ToFloat(Integer) (Variables|Advanced|GetScene%20Capture%20Resolution)) 1.99999)) (Math|Float|Floor (Math|Float|Log (Math|Conversions|ToFloat(Integer) (Variables|Advanced|GetAtlas_Res_Per_Frame)) 1.99999))))
  (Default|IncludeChannelPackedMaps
    (:Completed
      (Utilities|FlowControl|Sequence
        (:then_0
          ;[0.1476 0.5379 1] Allocate Targets for Color Maps when Necessary
          (for _array_element (Variables|Default|GetMaps%20to%20Render_Internal)
            (LocalVariables|SetCurMapType _array_element)
            (if _returnvalue_5
              (Utilities|IsValid _value
                (:Is%20Valid
                  (if (== (Variables|/Script/Engine.TextureRenderTarget2D|GetSizeX _value) (Variables|Advanced|GetResolution))
                    _
                    (else
                      _)))
                (:Is%20Not%20Valid
                  _)) ; Is the RT valid?
            )
            (bind _returnvalue_12 (Rendering|CreateRenderTarget2D (Variables|Advanced|GetResolution) (Variables|Advanced|GetResolution) (select (Math|Boolean|ORBoolean (Utilities|Enum|Equal(Enum) (LocalVariables|GetCurMapType) "NewEnumerator0") (Utilities|Enum|Equal(Enum) (LocalVariables|GetCurMapType) "NewEnumerator4") (Utilities|Enum|Equal(Enum) (LocalVariables|GetCurMapType) "NewEnumerator6") (Utilities|Enum|Equal(Enum) (LocalVariables|GetCurMapType) "NewEnumerator15") (Utilities|Enum|Equal(Enum) (LocalVariables|GetCurMapType) "NewEnumerator7") (Utilities|Enum|Equal(Enum) (LocalVariables|GetCurMapType) "NewEnumerator9") (Utilities|Enum|Equal(Enum) (LocalVariables|GetCurMapType) "NewEnumerator10") (Utilities|Enum|Equal(Enum) (LocalVariables|GetCurMapType) "NewEnumerator8")) "RTF_RGBA16f" "RTF_R8")))
            (Utilities|Map|Add (Variables|RenderTargets|GetTargetMaps) (LocalVariables|GetCurMapType) _returnvalue_12))
          (bind _keys (Utilities|Map|Keys (Variables|RenderTargets|GetTargetMaps)))
          (for _array_element_2 _keys
            (if (not (>= (Utilities|Array|FindItem (Variables|Default|GetMaps%20to%20Render_Internal) _array_element_2) 0))
              (Utilities|Map|Remove (Variables|RenderTargets|GetTargetMaps) _array_element_2))))
        (:then_1
          ;[1 0.6642 0.1083] Channel Packed Masks
          (for _array_element_3 (Variables|Default|GetChannel%20Packed%20Masks)
            (if (.ReturnValue (Utilities|Map|Find _channelpackedmaps _array_element_3))
              (Utilities|IsValid (Utilities|Map|Find _channelpackedmaps _array_element_3)
                (:Is%20Valid
                  (if (== (Variables|/Script/Engine.TextureRenderTarget2D|GetSizeX (Utilities|Map|Find _channelpackedmaps _array_element_3)) (Variables|Advanced|GetResolution))
                    _
                    (else
                      _)))
                (:Is%20Not%20Valid
                  _)) ; Is the RT valid?
            )
            (bind _returnvalue_13 (Rendering|CreateRenderTarget2D (Variables|Advanced|GetResolution) (Variables|Advanced|GetResolution) "RTF_RGBA8"))
            (Utilities|Map|Add (Variables|RenderTargets|GetChannelPackedMaps) _array_element_3 _returnvalue_13))
          (bind _keys_2 (Utilities|Map|Keys (Variables|RenderTargets|GetChannelPackedMaps)))
          (for _array_element_4 _keys_2
            (if (not (>= (Utilities|Array|FindItem (Variables|Default|GetChannel%20Packed%20Masks) _array_element_4) 0))
              (Utilities|Map|Remove (Variables|RenderTargets|GetChannelPackedMaps) _array_element_4))))
        (:then_2
          ;[0.1476 0.5379 1] Scene Capture Render Target with Mips
          (for _index (range (Variables|Advanced|GetSceneCaptureMips))
            (bind _output_get (LocalVariables|SetMipIdx _index))
            (if (<= _output_get (Utilities|Array|LastIndex (Variables|RenderTargets|GetSceneCaptureMipChain)))
              (Utilities|IsValid _output
                (:Is%20Valid
                  (if (== (Variables|/Script/Engine.TextureRenderTarget2D|GetSizeX _output) _returnvalue_10)
                    _
                    (else
                      _)))
                (:Is%20Not%20Valid
                  _)) ; Valid RT?
            )
            (bind _returnvalue_14 (Rendering|CreateRenderTarget2D _returnvalue_10 _returnvalue_10))
            (Utilities|Array|SetArrayElem (Variables|RenderTargets|GetSceneCaptureMipChain) (LocalVariables|GetMipIdx) _returnvalue_14 true)))
        (:then_3
          ;[1 1 1] Combined Alphas, Scratch and Debug
          (Rendering|CreateAndAssignRenderTarget (Variables|RenderTargets|GetCombinedAlphasRT) (Utilities|Select :Index (Variables|Default|GetImpostor%20Type) :NewEnumerator0 (+ 16 0) :NewEnumerator1 (+ 16 0) :NewEnumerator2 (+ 48 0)) (Utilities|Select :Index (Variables|Default|GetImpostor%20Type) :NewEnumerator0 (+ 16 0) :NewEnumerator1 (+ 16 0) :NewEnumerator2 (+ 48 0))
            (:Out
              (Rendering|CreateAndAssignRenderTarget (Variables|RenderTargets|GetCardVertexRT) 2 2 "RTF_RGBA32f"
                (:Out
                  (Utilities|CheckandCreateRT (Variables|RenderTargets|GetScratchRT) (Variables|Advanced|GetResolution) true
                    (:Out
                      (Variables|Advanced|SetCheckTargetMipSize (Variables|/Script/Engine.TextureRenderTarget2D|GetSizeX (Utilities|Array|Get(acopy) (Variables|RenderTargets|GetSceneCaptureMipChain) (Math|Integer|Min(Integer) (Utilities|Array|LastIndex (Variables|RenderTargets|GetSceneCaptureMipChain)) (Variables|Advanced|GetCutoutMipTarget))))))) ; RT used for Utility Combination maps
                )))))
        (:then_4
          ;[0.1979 0.1979 0.1979] Save List allows combined maps to be processed individually
          (Variables|Default|SetMaps%20to%20Save (Variables|Default|GetMaps%20to%20Render_Internal))
          (Utilities|FlowControl|Sequence
            (:then_0
              (if (Variables|Advanced|GetCombine%20Customlighting%20and%20Basecolor)
                (Utilities|Array|RemoveItem (Variables|Default|GetMaps%20to%20Save) "NewEnumerator10")))
            (:then_1
              (if (Variables|Advanced|GetCombine%20Normal%20and%20Depth)
                (Utilities|Array|RemoveItem (Variables|Default|GetMaps%20to%20Save) "NewEnumerator8")))
            (:then_2
              (for _array_element_5 (Variables|Default|GetChannel%20Packed%20Masks)
                (Utilities|FlowControl|Sequence
                  (:then_0
                    (Utilities|Array|RemoveItem _maps_to_save_2 (Default|ScalarMaptoBakeType (Utilities|Struct|BreakImpostorChannelPackSettings _array_element_5)))
                    (Utilities|Array|RemoveItem _maps_to_save_2 (Default|ScalarMaptoBakeType (.G_11_B3F571D648FCD1B25D023996A7D395BB (Utilities|Struct|BreakImpostorChannelPackSettings _array_element_5)))))
                  (:then_1
                    (Utilities|Array|RemoveItem _maps_to_save (Default|ScalarMaptoBakeType (.B_12_71D42C9349B0E42D3950BE82BCAE1C44 (Utilities|Struct|BreakImpostorChannelPackSettings _array_element_5))))
                    (Utilities|Array|RemoveItem _maps_to_save (Default|ScalarMaptoBakeType (.A_13_655F3E8A4BC083A5432B43AE8BAC9BDE (Utilities|Struct|BreakImpostorChannelPackSettings _array_element_5))))))))))))))

(fn UpdateCombinedCustomLighting
  ;[1 1 1] CustomLighting BaseColor Composited Last so Scratch RT can be used to update BaseColor Map for live tweaking.
  (if (and (Variables|Advanced|GetCombine%20Customlighting%20and%20Basecolor) (and (>= (Utilities|Array|FindItem (Variables|Default|GetMaps%20to%20Render_Internal) "NewEnumerator0") 0) (>= (Utilities|Array|FindItem (Variables|Default|GetMaps%20to%20Render_Internal) "NewEnumerator10") 0)))
    (Rendering|ClearRenderTarget2D (Utilities|Map|Find (Variables|RenderTargets|GetTargetMaps) "NewEnumerator0"))
    (Rendering|DrawMaterialtoRenderTarget (Utilities|Map|Find (Variables|RenderTargets|GetTargetMaps) "NewEnumerator0") (Variables|MIDs|GetBaseColorCustomLightingMID))))

(fn UpdateDynamicMesh
  (Utilities|Array|Clear (Variables|Debug|GetDynamicMeshTris))
  (bind _returnvalue (DynamicMeshComponent|GetDynamicMesh (Variables|Default|GetDynamicMeshComponent)))
  (Variables|Default|SetDynamicMesh _returnvalue)
  (DynamicMesh|Reset (Variables|Default|GetDynamicMesh))
  (bind _vectorlist (GeometryScript|ListUtils|ConvertArraytoVectorList (Variables|Debug|GetVerts)))
  (GeometryScript|MeshEdits|AddVerticestoMesh (Variables|Default|GetDynamicMesh) _vectorlist)
  (Utilities|FlowControl|Sequence
    (:then_0
      (for _index (range (/ (Utilities|Array|Length (Variables|Debug|GetTris)) 3))
        (bind _newtriangleindex (GeometryScript|MeshEdits|AddTriangletoMesh (Variables|Default|GetDynamicMesh) (Utilities|Struct|MakeIntVector (Utilities|Array|Get(acopy) (Variables|Debug|GetTris) (* _index 3)) (Utilities|Array|Get(acopy) (Variables|Debug|GetTris) (+ (* _index 3) 1)) (Utilities|Array|Get(acopy) (Variables|Debug|GetTris) (+ (* _index 3) 2)))))
        (Utilities|Array|Add (Variables|Debug|GetDynamicMeshTris) _newtriangleindex)))
    (:then_1
      (GeometryScript|UVs|SetNumUVChannels (Variables|Default|GetDynamicMesh) 1)
      (if (> (Utilities|Array|Length (Variables|Debug|GetUVs)) 0)
        (for (_ _array_index) (Variables|Debug|GetDynamicMeshTris)
          (GeometryScript|UVs|SetMeshTriangleUVs (Variables|Default|GetDynamicMesh) 0 _array_index (Utilities|Struct|MakeGeometryScriptUVTriangle (Utilities|Array|Get(acopy) (Variables|Debug|GetUVs) (Utilities|Struct|BreakIntVector (.ReturnValue (GeometryScript|MeshQueries|GetTriangleIndices (Variables|Default|GetDynamicMesh) _array_index)))) (Utilities|Array|Get(acopy) (Variables|Debug|GetUVs) (.Y (Utilities|Struct|BreakIntVector (.ReturnValue (GeometryScript|MeshQueries|GetTriangleIndices (Variables|Default|GetDynamicMesh) _array_index))))) (Utilities|Array|Get(acopy) (Variables|Debug|GetUVs) (.Z (Utilities|Struct|BreakIntVector (.ReturnValue (GeometryScript|MeshQueries|GetTriangleIndices (Variables|Default|GetDynamicMesh) _array_index)))))))
          (GeometryScript|Normals|SetMeshTriangleNormals (Variables|Default|GetDynamicMesh) _array_index (Utilities|Struct|MakeGeometryScriptTriangle "0.000000,0.000000,1.000000" "0.000000,0.000000,1.000000" "0.000000,0.000000,1.000000")))))))

(fn ConstructionScript
  (Default|InitializeRender
    (:Out
      (Variables|Default|Setallmaps (Utilities|Select :Index (Variables|Default|Getscalarmaps) :NewEnumerator1 "NewEnumerator1" :NewEnumerator2 "NewEnumerator2" :NewEnumerator3 "NewEnumerator3" :NewEnumerator14 "NewEnumerator14" :NewEnumerator5 "NewEnumerator5" :NewEnumerator12 "NewEnumerator12" :NewEnumerator16 "NewEnumerator16" :NewEnumerator17 "NewEnumerator17" :NewEnumerator13 "NewEnumerator13" :NewEnumerator8 "NewEnumerator8" :NewEnumerator18 "NewEnumerator0")))))
