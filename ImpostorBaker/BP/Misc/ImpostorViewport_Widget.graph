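# Editor utility widget providing the Impostor Baker's interactive preview viewport, shown through a material-driven RetainerBox.
# Construct spawns an ImpostorScene inside the viewport and aims the camera at the origin.
# Input overrides set state that Tick polls: drag orbits with a pole clamp, right-drag or wheel dollies, L-drag rotates the sun, showing its arrow and updating skylight lower-hemisphere color.
# The every-tick wheel step also refreshes the previous mouse position.
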EditorUtilityWidgetBlueprint /ImpostorBaker/BP/Misc/ImpostorViewport_Widget
graphs [
 EventGraph {nodes [
   Event Construct()
   > Set World(World:Object Reference=GetViewportWorld!(self:Viewport Object Reference=Get ImpostorViewport() -> (ImpostorViewport:Impostor Viewport Object Reference)) -> (ReturnValue:World Object Reference)) -> (Output_Get:Object Reference)
   > SetViewLocation!(execute, self:Viewport Object Reference=Get ImpostorViewport(), Location:Vector="-100,100,75") -> (then)
   n1 = GetViewLocation!(self:Viewport Object Reference=Get ImpostorViewport()) -> (ReturnValue:Vector)
   > SetViewRotation!(
    execute
    self:Viewport Object Reference=Get ImpostorViewport()
    Rotation:Rotator=Rotation From X Vector!(InVec:Vector=vector * vector(A=@n1, B:float=-1) -> (ReturnValue=~)) -> (ReturnValue:Rotator)
   ) -> (then)
   > n2 = Spawn!(
    execute
    self:Viewport Object Reference=Get ImpostorViewport()
    ActorClass:Actor Class Reference=&/ImpostorScene.ImpostorScene_C
   ) -> (then, ReturnValue:Actor Object Reference)
   > n3 = Cast To ImpostorScene!(execute, Object=@n2) -> (then, CastFailed, AsImpostor Scene:Impostor Scene Object Reference) {is=title K2Node_DynamicCast}
   > Set Impostor Scene(Impostor Scene:Impostor Scene Object Reference=@n3) -> (Output_Get:Impostor Scene Object Reference)
   Event Tick()
   > n4 = Sequence()
   > Branch(Condition=Get Dragging() -> (Dragging:bool))
   > n5 = Branch(Condition=Get RightMouse() -> (RightMouse:bool))
   > K2Node_Composite_2 = Right Click: Distance Pan() {is=title K2Node_Composite}
   n6 = Branch(execute=@n5.else, Condition=Get L_Key_Pressed() -> (L_Key_Pressed:bool))
   > K2Node_Composite_0 = Mouse Light Rotate!(InputPin:Exec) {is=title K2Node_Composite}
   K2Node_Composite_1 = Mouse Camera Orbit!(InputPin:Exec=@n6.else) {is=title K2Node_Composite}
   K2Node_Composite_3 = Mouse Wheel: Distance Pan(execute=@n4.then_1) {is=title K2Node_Composite}
   Event OnMouseLeave!() -> (OutputDelegate, then, MouseEvent:"Pointer Event Structure (by ref)") {is=title K2Node_Event}
   > Set Dragging(Dragging:bool=false) -> (Output_Get:bool)
   > Set RightMouse(RightMouse:bool=false) -> (Output_Get:bool)
  ]}
 Mouse Light Rotate {
  path=EventGraph.K2Node_Composite_0.Mouse Light Rotate
  nodes [
   Inputs!() -> (InputPin:Exec)
   > n1 = GetMousePositionOnViewport!(execute, WorldContextObject:Object Reference=Get World() -> (World:Object Reference)) -> (then, ReturnValue:Vector 2D Structure)
   n2 = Get Impostor Scene() -> (Impostor Scene:Impostor Scene Object Reference)
   n3 = BreakRotator(InRot=Get World Rotation!(self:Scene Component Object Reference=Get DirectionalLight!(self:Impostor Scene Object Reference=@n2) -> (DirectionalLight:Directional Light Component Object Reference)) -> (ReturnValue:Rotator) {is=title K2Node_CallFunction})
   n4 = BreakVector2D(InVec=@n1)
   n5 = BreakVector2D(InVec=Get mouse_prev() -> (mouse_prev:Vector 2D Structure))
   n6 = float + float(A=@n3.Pitch, B=float / float(A=float - float(A=@n4.Y, B=@n5.Y), B=2))
   n7 = float - float(A=BreakVector2D(InVec=@n1), B=BreakVector2D(InVec=Get mouse_prev()))
   > Set World Rotation!(
    execute
    self:Scene Component Object Reference=Get DirectionalLight(self=Get Impostor Scene())
    NewRotation:Rotator=MakeRotator(Roll=@n3, Pitch="Clamp (Float)"(Value=@n6, Min=-89, Max=89), Yaw=float - float(A=@n3.Yaw, B=@n7))
    bSweep:bool=false
    bTeleport:bool=false
   ) -> (then, SweepHitResult:Hit Result Structure) {is=title K2Node_CallFunction}
   n8 = Get Impostor Scene()
   n9 = GetAxes!(A:Rotator=Get World Rotation(self=Get DirectionalLight(self=@n8))) -> (X:Vector, Y:Vector, Z:Vector)
   n10 = BreakVector(InVec=@n9)
   > SetLowerHemisphereColor!(
    execute
    self:Sky Light Component Object Reference=Get SkyLight!(self:Impostor Scene Object Reference=@n8) -> (SkyLight:Sky Light Component Object Reference)
    InLowerHemisphereColor:"Linear Color Structure (by ref)"=Select(
     Option 0:"Linear Color Structure (by ref)"=Make Linear Color!(R:float=0.05, G:float=0.05, B:float=0.075, A:float=1) -> (LinearColor:Linear Color Structure) {is=title K2Node_MakeStruct}
     Option 1:"Linear Color Structure (by ref)"=Make Linear Color(R=0, G=0, B=0, A=1)
     Index="float < float"(A=@n10.Z)
    ) -> (ReturnValue:"Linear Color Structure (by ref)")
   ) -> (then)
   n11 = Get SunDirectionArrow!(self:Impostor Scene Object Reference=Get Impostor Scene()) -> (SunDirectionArrow:Static Mesh Component Object Reference)
   > Set World Location!(execute, self:Scene Component Object Reference=@n11, NewLocation:Vector="0, 0, 0", bSweep:bool=false, bTeleport:bool=false) -> (then, SweepHitResult:Hit Result Structure) {is=title K2Node_CallFunction}
   n12 = Get Impostor Scene()
   > Set World Rotation(self=@n11, NewRotation=Get World Rotation(self=Get DirectionalLight(self=@n12)), bSweep=false, bTeleport=false)
   n13 = Get SunDirectionArrow(self=@n12)
   n14 = float + float(
    A=GetComponentBounds!(Component:Scene Component Object Reference=Get Capture_StaticMesh!(self:Impostor Scene Object Reference=@n12) -> (Capture_StaticMesh:Static Mesh Component Object Reference)) -> (Origin:Vector, BoxExtent:Vector, SphereRadius:float).SphereRadius
    B=float * float(A=GetComponentBounds(Component=@n13).SphereRadius, B=2) {is=title K2Node_PromotableOperator}
   )
   > Add Local Offset!(
    execute
    self:Scene Component Object Reference=@n11
    DeltaLocation:Vector=MakeVector(X=float * float(A=@n14, B=-1), Y=0)
    bSweep:bool=false
    bTeleport:bool=false
   ) -> (then, SweepHitResult:Hit Result Structure)
   > SetVisibility!(
    execute
    self:Scene Component Object Reference=Get SunDirectionArrow(self=Get Impostor Scene())
    bNewVisibility:bool=true
    bPropagateToChildren:bool=false
   ) -> (then)
   "To String (Vector)"!(InVec:Vector=@n9) -> (ReturnValue:String)
   Outputs()
  ]
 }
 Mouse Camera Orbit {
  path=EventGraph.K2Node_Composite_1.Mouse Camera Orbit
  nodes [
   Inputs()
   > n1 = GetMousePositionOnViewport(WorldContextObject=Get World())
   n2 = Get ImpostorViewport()
   n3 = GetViewLocation(self=@n2)
   n4 = BreakVector2D(InVec=@n1)
   n5 = BreakVector2D(InVec=Get mouse_prev())
   n6 = Get ImpostorViewport()
   > SetViewLocation(
    self=@n6
    Location=Rotate Vector Around Axis!(InVect:Vector=@n3, AngleDeg:float=float - float(A=@n4, B=@n5), Axis:Vector="0,0,1") -> (ReturnValue:Vector)
   )
   > SetViewRotation(self=@n6, Rotation=Rotation From X Vector(InVec=vector * vector(A=@n3, B:Vector="-1,-1,-1") -> (ReturnValue=~)))
   > n7 = GetMousePositionOnViewport(WorldContextObject=Get World())
   n8 = Get ImpostorViewport()
   n9 = BreakVector2D(InVec=@n7)
   n10 = BreakVector2D(InVec=Get mouse_prev())
   n11 = GetViewLocation(self=@n8)
   n12 = Rotate Vector Around Axis(
    InVect=@n11
    AngleDeg=float / float(A=float - float(A=@n9.Y, B=@n10.Y), B=2)
    Axis=BreakRotIntoAxes!(InRot:"Rotator (by ref)"=GetViewRotation!(self:Viewport Object Reference=@n8) -> (ReturnValue:Rotator)) -> (X:Vector, Y:Vector, Z:Vector).Y
   )
   K2Node_Composite_1 = Prevent Gimble Lock!(A:Vector=vector + vector(A=@n11, B=@n12)) -> (ReturnValue:bool=~) {is=title K2Node_Composite}
   > Branch(Condition=@K2Node_Composite_1)
   > SetViewLocation(self=Get ImpostorViewport(), Location=@n12)
   > SetViewRotation(self=Get ImpostorViewport(), Rotation=Rotation From X Vector(InVec=vector * vector(A=@n11, B="-1,-1,-1") -> (ReturnValue=~)))
   GetViewRotation(self=@n2)
   Outputs()
  ]
 }
 Prevent Gimble Lock {
  path=EventGraph.K2Node_Composite_1.Mouse Camera Orbit.K2Node_Composite_1.Prevent Gimble Lock
  nodes [
   n1 = Dot Product!(
    A:Vector=Normalize!(A:Vector=Inputs!() -> (A:Vector=~), Tolerance:float=0.0001) -> (ReturnValue:Vector) {is=title K2Node_CallFunction}
    B:Vector="0,0,1"
   ) -> (ReturnValue:double)
   Outputs!(ReturnValue:bool="float < float"(A="Absolute (Float)"(A=@n1), B=0.9875))
  ]
 }
 "Right Click: Distance Pan" {
  path=EventGraph.K2Node_Composite_2.Right Click: Distance Pan
  nodes [
   Inputs!() -> (execute:Exec)
   > n1 = GetMousePositionOnViewport(WorldContextObject=Get World())
   n2 = Get ImpostorViewport()
   n3 = Get Rotation X Vector!(InRot:Rotator=GetViewRotation(self=@n2)) -> (ReturnValue:Vector) {is=title K2Node_CallFunction}
   n4 = Get mouse_prev()
   n5 = float * float(A=float - float(A=BreakVector2D(InVec=@n4).Y, B=BreakVector2D(InVec=@n1).Y), B=Get Zoom Sensitivity() -> (Zoom Sensitivity:double))
   > SetViewLocation(
    self=Get ImpostorViewport()
    Location=vector + vector(A=GetViewLocation(self=@n2), B=vector * vector(A=@n3, B:double=@n5) -> (ReturnValue=~))
   )
   > n6 = GetMousePositionOnViewport(WorldContextObject=Get World())
   > Set mouse_prev(mouse_prev:Vector 2D Structure=@n6) -> (Output_Get:Vector 2D Structure)
   > Set MouseWheelDelta(MouseWheelDelta:double=0) -> (Output_Get:double)
   Outputs!()
  ]
 }
 "Mouse Wheel: Distance Pan" {
  path=EventGraph.K2Node_Composite_3.Mouse Wheel: Distance Pan
  nodes [
   Inputs()
   > n1 = GetMousePositionOnViewport(WorldContextObject=Get World())
   n2 = Get ImpostorViewport()
   n3 = Get Rotation X Vector(InRot=GetViewRotation(self=@n2))
   n4 = float * float(A=float * float(A=Get MouseWheelDelta() -> (MouseWheelDelta:double), B=20), B=Get Zoom Sensitivity())
   > SetViewLocation(self=Get ImpostorViewport(), Location=vector + vector(A=GetViewLocation(self=@n2), B=vector * vector(A=@n3, B=@n4) -> (ReturnValue=~)))
   > n5 = GetMousePositionOnViewport(WorldContextObject=Get World())
   > Set mouse_prev(mouse_prev=@n5)
   > Set MouseWheelDelta(MouseWheelDelta=0)
   vector - vector(A:double=BreakVector2D(InVec=Get mouse_prev()).Y, B:double=BreakVector2D(InVec=@n1).Y)
   Outputs()
  ]
 }
 OnKeyDown {nodes [
   n1 = PrintString(InString=Append(A=L Pressed))
   Return Node!(execute=[@n1.then, @n6.then, @n5.then], ReturnValue:Event Reply Structure=Handled())
   n2 = OnKeyDown!() -> (then, MyGeometry:Geometry Structure, InKeyEvent:Key Event Structure) {is=title K2Node_FunctionEntry}
   n3 = Get Key Display Name!(
    Key:"Key Structure (by ref)"=GetKey!(Input:"Key Event Structure (by ref)"=@n2.InKeyEvent) -> (ReturnValue:Key Structure)
    bLongDisplayName:bool=true
   ) -> (ReturnValue:Text) {is=title K2Node_CallFunction}
   > n4 = Branch(Condition="Equal Exactly (Text)"!(A:"Text (by ref)"=@n3, B:"Text (by ref)"=MakeLiteralText(Value="LOCTEXT(\"L\")")) -> (ReturnValue:bool))
   > n5 = Set L_Key_Pressed(L_Key_Pressed:bool=true) -> (Output_Get:bool)
   n6 = Set L_Key_Pressed(execute=@n4.else, L_Key_Pressed=false)
   "To String (Text)"(InText=@n3)
  ]}
 OnKeyUp {nodes [
   n1 = OnKeyUp!() -> (then, MyGeometry:Geometry Structure, InKeyEvent:Key Event Structure) {is=title K2Node_FunctionEntry}
   n2 = GetKey(Input=@n1.InKeyEvent)
   > n3 = Branch(Condition="Equal Exactly (Text)"(A=Get Key Display Name(Key=@n2, bLongDisplayName=true), B=MakeLiteralText(Value="LOCTEXT(\"L\")")))
   > Set L_Key_Pressed(L_Key_Pressed=false)
   > PrintString(InString=Append(A=L Released))
   > n4 = SetVisibility(self=Get SunDirectionArrow(self=Get Impostor Scene()), bNewVisibility=false, bPropagateToChildren=false)
   Return Node(execute=[@n4.then, @n5.then], ReturnValue=Handled())
   n5 = Set L_Key_Pressed(execute=@n3.else, L_Key_Pressed=false)
  ]}
 OnMouseButtonDown {nodes [
   n1 = OnMouseButtonDown!() -> (then, MyGeometry:Geometry Structure, MouseEvent:"Pointer Event Structure (by ref)") {is=title K2Node_FunctionEntry}
   > n2 = Set RightMouse(RightMouse=Is Mouse Button Down!(Input:"Pointer Event Structure (by ref)"=@n1.MouseEvent, MouseButton:Key Structure=RightMouseButton) -> (ReturnValue:bool))
   n3 = PrintString(InString=Append(A="RMB: ", B="To String (Boolean)"(InBool=@n2)))
   Return Node(execute=[@n3.then, @n5.then], ReturnValue=Make Event Reply!() -> (EventReply:Event Reply Structure) {is=title K2Node_MakeStruct})
   n4 = GetMousePositionOnViewport(execute=@n2.then, WorldContextObject=Get World())
   > Set mouse_prev(mouse_prev=@n4)
   > n5 = Set Dragging(Dragging=true)
  ]}
 OnMouseButtonUp {nodes [
   OnMouseButtonUp!() -> (then, MyGeometry:Geometry Structure, MouseEvent:"Pointer Event Structure (by ref)") {is=title K2Node_FunctionEntry}
   > Set RightMouse(RightMouse=false)
   > Set Dragging(Dragging=false)
   > Return Node(ReturnValue=Make Event Reply())
  ]}
 OnMouseWheel {nodes [
   n1 = OnMouseWheel!() -> (then, MyGeometry:Geometry Structure, MouseEvent:"Pointer Event Structure (by ref)") {is=title K2Node_FunctionEntry}
   > Set MouseWheelDelta(MouseWheelDelta=Get Wheel Delta!(Input:"Pointer Event Structure (by ref)"=@n1.MouseEvent) -> (ReturnValue:float) {is=title K2Node_CallFunction})
   > Return Node(ReturnValue=Make Event Reply())
  ]}
]
variables [
 Dragging {type=boolean}
 Impostor Scene {type=ImpostorScene_C, default=null}
 L_Key_Pressed {type=boolean}
 mouse_prev {type=Vector2D, default=<0 0>}
 MouseWheelDelta {type=number}
 NewVar {type=boolean}
 RightMouse {type=boolean}
 Static Mesh {type=boolean}
 World {type=Object, default=null}
 Zoom Sensitivity {type=number, default=1}
]
widgets {
 info {rootWidgetClass=&Overlay, widgetCount=3}
 widgets [
  Overlay_0 {widgetClassPath=&Overlay, bIsVariable=true, widget=Overlay_0, properties {visibility=Visible}}
  RetainerBox_35 {
   widgetClassPath=/Script/UMG.RetainerBox
   bIsVariable=true
   parent=Overlay_0
   widget=RetainerBox_35
   properties {bRetainRender=false, effectMaterial=/ImpostorBaker/Materials/PostProcess/M_RetainerBox, visibility=SelfHitTestInvisible}
   slot {class=OverlaySlot, properties {horizontalAlignment=HAlign_Fill, verticalAlignment=VAlign_Fill}}
  }
  ImpostorViewport {
   widgetClassPath=&/ImpostorViewport
   bIsVariable=true
   parent=RetainerBox_35
   widget=ImpostorViewport
   properties {visibility=SelfHitTestInvisible}
   slot {class=PanelSlot}
  }
 ]
}
