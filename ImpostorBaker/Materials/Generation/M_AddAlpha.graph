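# M_AddAlpha is an unlit, additive utility material in the ImpostorBaker generation folder, apparently for adding alpha-derived values during impostor baking.
# It samples the MipRT texture parameter, which defaults to a black texture, and inverts its alpha channel with OneMinus.
# The additive blend suggests the inverted alpha is accumulated onto a render target, though the graph shows no output connection.
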Material /ImpostorBaker/Materials/Generation/M_AddAlpha
graph {
 properties {blendMode=BLEND_Additive, pixelDepthOffsetMode=PDOM_Legacy, shadingModel=MSM_Unlit, usedShadingModels=Unlit}
 nodes [OneMinus(None=TextureSampleParameter2D() {parameterName=MipRT, texture=&Black}.A)]
}
