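# Unlit alpha-composite material in ImpostorBaker's generation set that packs four textures (RT1 to RT4) into separate channels.
# Each texture passes through a lerp driven by its own alpha scalar, though both lerp inputs show the red channel.
# The first three results form a float3 via MakeFloat3, while the fourth is inverted with OneMinus as a separate value.
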Material /ImpostorBaker/Materials/Generation/M_ChannelPack
graph {
 properties {blendMode=BLEND_AlphaComposite, pixelDepthOffsetMode=PDOM_Legacy, shadingModel=MSM_Unlit, usedShadingModels=Unlit}
 nodes [
  n1 = TextureSampleParameter2D() {
   parameterName=RT1
   samplerType=SAMPLERTYPE_LinearColor
   texture=/Engine/Functions/Engine_MaterialFunctions02/PivotPainter2/Black_1x1_EXR_Texture
  }
  n2 = TextureSampleParameter2D() {
   parameterName=RT2
   samplerType=SAMPLERTYPE_LinearColor
   texture=&Black_1x1_EXR_Texture
  }
  n3 = TextureSampleParameter2D() {
   parameterName=RT3
   samplerType=SAMPLERTYPE_LinearColor
   texture=&Black_1x1_EXR_Texture
  }
  MaterialFunctionCall(
   X=LinearInterpolate(A=@n1.R, B=@n1.R, Alpha=ScalarParameter() {parameterName=RT1Alpha})
   Y=LinearInterpolate(A=@n2.R, B=@n2.R, Alpha=ScalarParameter() {parameterName=RT2Alpha})
   Z=LinearInterpolate(A=@n3.R, B=@n3.R, Alpha=ScalarParameter() {parameterName=RT3Alpha})
  ) {materialFunction=&MakeFloat3}
  n4 = TextureSampleParameter2D() {
   parameterName=RT4
   samplerType=SAMPLERTYPE_LinearColor
   texture=&Black_1x1_EXR_Texture
  }
  OneMinus(None=LinearInterpolate(A=@n4.R, B=@n4.R, Alpha=ScalarParameter() {parameterName=RT4Alpha}))
 ]
}
