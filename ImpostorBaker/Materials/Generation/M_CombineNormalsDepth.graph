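# Unlit alpha-composite material in ImpostorBaker's generation set, named for combining baked normal and depth captures from its Normal and Depth texture parameters.
# It inverts the Depth red channel, and an If node replaces values equal to 1 (zero source depth) with a neutral 0.5, passing others through.
# Normal is sampled separately, alongside a tiled engine placeholder sprite; the file shows no links to material outputs.
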Material /ImpostorBaker/Materials/Generation/M_CombineNormalsDepth
graph {
 properties {blendMode=BLEND_AlphaComposite, pixelDepthOffsetMode=PDOM_Legacy, shadingModel=MSM_Unlit, usedShadingModels=Unlit}
 nodes [
  n1 = OneMinus(None=TextureSampleParameter2D() {parameterName=Depth, texture=&Black}.R)
  If(A=@n1, "A > B"=@n1, "A == B"=Constant() {r=0.5}, "A < B"=@n1) {constB=1}
  TextureSampleParameter2D() {parameterName=Normal, texture=&Black}
  TextureSample(UVs=TextureCoordinate() {uTiling=16, vTiling=16}) {texture=/Engine/EditorResources/BadBlueprintSprite}
 ]
}
