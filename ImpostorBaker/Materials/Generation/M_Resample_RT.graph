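# M_Resample_RT is an unlit Alpha Composite material in ImpostorBaker's generation set, apparently for resampling a render target during impostor baking.
# It samples a texture parameter named RT, which defaults to a black texture, and feeds the sample's alpha through a OneMinus node to invert it.
# The graph does not show which material outputs these nodes drive.
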Material /ImpostorBaker/Materials/Generation/M_Resample_RT
graph {
 properties {blendMode=BLEND_AlphaComposite, pixelDepthOffsetMode=PDOM_Legacy, shadingModel=MSM_Unlit, usedShadingModels=Unlit}
 nodes [OneMinus(None=TextureSampleParameter2D() {parameterName=RT, texture=&Black}.A)]
}
