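# Unlit alpha-composite material in ImpostorBaker's generation set that samples a render target and dilates its colors into background texels.
# UVs snap to texel centers from TextureSize, then custom HLSL searches eight directions up to MaxRadius for the nearest texel with alpha below one.
# Branches blend the result over MipRT and substitute a midpoint value.
# Debug Custom nodes remain, and no material output connections are shown.
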Material /ImpostorBaker/Materials/Generation/M_Sample_Frame
graph {
 properties {blendMode=BLEND_AlphaComposite, pixelDepthOffsetMode=PDOM_Legacy, shadingModel=MSM_Unlit, usedShadingModels=Unlit}
 nodes [
  MaterialExpressionTextureSampleParameter2D_0 = TextureSampleParameter2D() {parameterName=RT, samplerType=SAMPLERTYPE_LinearColor, texture=/ImpostorBaker/TestContent/RTs/RT_Test_01}
  Custom!(inX=@MaterialExpressionTextureSampleParameter2D_0.R) {
   code="if (inX > 0)\r\n{\r\nreturn float3(1,0,0);\r\n}\r\n\r\nreturn 0;"
   showCode=true
   inputs [{
     inputName=inX
     input {expression=&self:MaterialExpressionTextureSampleParameter2D_0, mask=1, maskR=1, outputIndex=1}
    }]
  }
  MaterialExpressionTextureObjectParameter_0 = TextureObjectParameter() {parameterName=RT, samplerType=SAMPLERTYPE_LinearColor, texture=&RT_Test_01}
  n1 = ScalarParameter() {defaultValue=512, parameterName=TextureSize}
  n2 = Floor(None=Multiply(A=TextureCoordinate(), B=@n1) {constB=1024})
  MaterialExpressionReroute_0 = Reroute(None=Add(A=Divide(A=@n2, B=@n1) {constB=1024}, B=Divide(B=@n1) {constA=0.5, constB=2048}))
  MaterialExpressionReroute_1 = Reroute(None=@n1)
  MaterialExpressionScalarParameter_13 = ScalarParameter() {defaultValue=64, parameterName=MaxRadius}
  MaterialExpressionReroute_2 = Reroute(None=@n1)
  MaterialExpressionVectorParameter_0 = VectorParameter() {parameterName=BackgroundColor, defaultValue=<rgba 0.5 0.5 0.5 0.5>}
  Custom!(
   Tex=@MaterialExpressionTextureObjectParameter_0
   UV=@MaterialExpressionReroute_0
   Resolution=@MaterialExpressionReroute_1
   MaxSteps=@MaterialExpressionScalarParameter_13
   TextureSize=@MaterialExpressionReroute_2
   BackgroundColor=@MaterialExpressionVectorParameter_0.A
  ) {
   code="//////////////// UV Dilation ///////////////////////////\r\n\r\n//** Tex        **// Input Texture Object storing Volume Data\r\n//** UV         **// Input float2 for UVs\r\n//** Resolution **// Resolution of render target\r\n\r\nfloat texelsize = 1 / Resolution;\r\nfloat mindist = 10000000;\r\n\r\nfloat2 offsets[8] = {float2(-1,0), float2(1,0), float2(0,1), float2(0,-1), float2(-1,1), float2(1,1), float2(1,-1), float2(-1,-1)};\r\n\r\nfloat4 sample = Tex.SampleLevel(TexSampler, UV, 0);\r\n\r\nfloat4 curminsample = sample;\r\n\r\nif(sample.a > 0)\r\n{\r\n\t//curminsample = BackgroundColor;\r\n\r\n\tint j = 0;\r\n\twhile(j < MaxSteps)\r\n\t{\r\n\t\tj++;\r\n\t\tint i = 0;\r\n\t\twhile (i < 8)\r\n\t\t{\r\n\r\n\t\t\tfloat2 curUV = UV + offsets[i] * texelsize * j;\r\n\t\t\tfloat4 offsetsample = Tex.SampleLevel(TexSampler, curUV, 0);\r\n\r\n\t\t\tif(offsetsample.a < 1)\r\n\t\t\t{\r\n\r\n\t\t\t\tfloat curdist = length(UV - curUV);\r\n\r\n\t\t\t\tif (curdist < mindist)\r\n\t\t\t\t{\r\n\t\t\t\t\tmindist = curdist;\r\n\t\t\t\t\tcurminsample = offsetsample;\r\n\t\t\t\t}\r\n\t\t\t}\r\n\r\n\t\r\n\t\t\ti++;\r\n\t\t}\r\n\r\n\t}\r\nreturn float4(0,1,0,0);\r\n}\r\n\r\nreturn float4(0,0,1,0);\r\n//return float4(curminsample.xyz,1);"
   outputType=CMOT_Float4
   inputs [
    {inputName=Tex, input {expression=&self:MaterialExpressionTextureObjectParameter_0}}
    {inputName=UV, input {expression=&self:MaterialExpressionReroute_0}}
    {inputName=Resolution, input {expression=&self:MaterialExpressionReroute_1}}
    {inputName=MaxSteps, input {expression=&self:MaterialExpressionScalarParameter_13}}
    {inputName=TextureSize, input {expression=&self:MaterialExpressionReroute_2}}
    {
     inputName=BackgroundColor
     input {expression=&self:MaterialExpressionVectorParameter_0, mask=1, maskA=1, outputIndex=4}
    }
   ]
  }
  n3 = Custom(
   Tex=@MaterialExpressionTextureObjectParameter_0
   UV=@MaterialExpressionReroute_0
   Resolution=@MaterialExpressionReroute_1
   MaxSteps=@MaterialExpressionScalarParameter_13
   TextureSize=@MaterialExpressionReroute_2
   BackgroundColor=@MaterialExpressionVectorParameter_0.A
  ) {
   code="//////////////// UV Dilation ///////////////////////////\r\n\r\n//** Tex        **// Input Texture Object storing Volume Data\r\n//** UV         **// Input float2 for UVs\r\n//** Resolution **// Resolution of render target\r\n\r\nfloat texelsize = 1 / Resolution;\r\nfloat mindist = 10000000;\r\n\r\nfloat2 offsets[8] = {float2(-1,0), float2(1,0), float2(0,1), float2(0,-1), float2(-1,1), float2(1,1), float2(1,-1), float2(-1,-1)};\r\n\r\nfloat4 sample = Tex.SampleLevel(TexSampler, UV, 0);\r\n\r\nfloat4 curminsample = sample;\r\n\r\nif(sample.a > 0)\r\n{\r\n\t//curminsample = BackgroundColor;\r\n\r\n\tint j = 0;\r\n\twhile(j < MaxSteps)\r\n\t{\r\n\t\tj++;\r\n\t\tint i = 0;\r\n\t\twhile (i < 8)\r\n\t\t{\r\n\r\n\t\t\tfloat2 curUV = UV + offsets[i] * texelsize * j;\r\n\t\t\tfloat4 offsetsample = Tex.SampleLevel(TexSampler, curUV, 0);\r\n\t\t\tfloat colortest = dot(offsetsample.xyz, 1);\r\n\r\n\t\t\tif(offsetsample.a < 1)\r\n\t\t\t{\r\n\r\n\t\t\t\tfloat curdist = length(UV - curUV);\r\n\r\n\t\t\t\tif (curdist < mindist)\r\n\t\t\t\t{\r\n\t\t\t\t\tfloat2 projectUV = curUV + offsets[i] * texelsize * j * 1;\r\n\t\t\t\t\tfloat3 direction = Tex.SampleLevel(TexSampler, projectUV, 0);\r\n\t\t\t\t\t\r\n\t\t\t\t\tif(direction.x != 0 || direction.y != 0 || direction.z != 0)\r\n\t\t\t\t\t{\r\n\t\t\t\t\t\r\n\t\t\t\t\t\tmindist = curdist;\r\n\t\t\t\t\t\tcurminsample = offsetsample;\r\n\r\n\t\t\t\t\t}\r\n\t\t\t\t\telse\r\n\t\t\t\t\t{\r\n\t\t\t\t\tmindist = curdist;\r\n\t\t\t\t\tcurminsample = offsetsample;\r\n\t\t\t\t\t}\r\n\t\t\t\t}\r\n\t\t\t}\r\n\r\n\t\r\n\t\t\ti++;\r\n\t\t}\r\n\r\n\t}\r\n}\r\nelse\r\n{\r\n\treturn float4(curminsample.xyz,0);\r\n}\r\n\t\r\nreturn float4(curminsample.xyz,1);"
   outputType=CMOT_Float4
   inputs [
    {inputName=Tex, input {expression=&self:MaterialExpressionTextureObjectParameter_0}}
    {inputName=UV, input {expression=&self:MaterialExpressionReroute_0}}
    {inputName=Resolution, input {expression=&self:MaterialExpressionReroute_1}}
    {inputName=MaxSteps, input {expression=&self:MaterialExpressionScalarParameter_13}}
    {inputName=TextureSize, input {expression=&self:MaterialExpressionReroute_2}}
    {
     inputName=BackgroundColor
     input {expression=&self:MaterialExpressionVectorParameter_0, mask=1, maskA=1, outputIndex=4}
    }
   ]
  }
  n4 = Constant() {r=0.5}
  If(A=ComponentMask(None=@n3) {r=true}, "A > B"=@n3, "A == B"=@n4, "A < B"=@n4) {desc=Replace background with midpoint value}
  n5 = TextureSampleParameter2D() {parameterName=MipRT, texture=&Black}
  LinearInterpolate(A=@n5.RGB, B=ComponentMask(None=@n3) {b=true, g=true, r=true}, Alpha=Max(A=Ceil(None=@n5.A), B=ComponentMask(None=@n3) {a=true}))
  Multiply(A=@MaterialExpressionTextureSampleParameter2D_0.A)
  TextureObject() {texture=&Black}
 ]
}
