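# Editor Utility Widget that bakes impostors (full sphere, hemisphere or billboard) from a static mesh.
# An orthographic or perspective SceneCapture2D shoots the mesh along view vectors, switching map types through a material parameter collection and drawing frames into atlas render targets.
# Dynamic materials handle mip-chain sampling, channel-packed masks and preview.
# Saving creates or overwrites textures and a material instance, then writes the impostor mesh to the static mesh's LODs.
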EditorUtilityWidgetBlueprint /ImpostorBaker/BP/EUW_Impostor_Editor
graphs [
 Apply Custom Material Params {nodes [
   Apply Custom Material Params!() -> (
    then
    MID:Material Instance Dynamic Object Reference
    Floats:"Map of Names to Float (double-precision)s"
    Vectors:Map of Names to Linear Color Structures
    Textures:Map of Names to Texture 2D Object References
   ) {is=title K2Node_FunctionEntry}
   > n1 = Sequence!(execute) -> (then_0, then_1, then_2:Exec)
   n2 = Get Floats() -> (Floats:"Map of Names to Float (double-precision)s")
   > Branch(Condition="integer > integer"(A=Length!(TargetMap:"Map of Names to Float (double-precision)s"=@n2) -> (ReturnValue:int=0) {is=title K2Node_CallFunction}))
   > n3 = Keys!(execute, TargetMap:"Map of Names to Float (double-precision)s"=Get Floats()) -> (then, Keys:Array of Names)
   > n4 = For Each Loop(Array:Array of Names=@n3) -> (Array Element:Name)
   n5 = Find(TargetMap:"Map of Names to Float (double-precision)s"=Get Floats(), Key:"Name (by ref)"=@n4) -> (Value:double)
   > Branch(Condition=@n5.ReturnValue)
   > SetScalarParameterValue(self=Get MID() -> (MID:Material Instance Dynamic Object Reference), ParameterName=@n4, Value=@n5)
   n6 = Get Vectors() -> (Vectors:Map of Names to Linear Color Structures)
   Branch(execute=@n1.then_1, Condition="integer > integer"(A=Length(TargetMap:Map of Names to Linear Color Structures=@n6) -> (ReturnValue=0)))
   > n7 = Keys(TargetMap:Map of Names to Linear Color Structures=Get Vectors())
   > n8 = For Each Loop(Array=@n7)
   n9 = Find(TargetMap:Map of Names to Linear Color Structures=Get Vectors(), Key=@n8) -> (Value:Linear Color Structure)
   > Branch(Condition=@n9.ReturnValue)
   > SetVectorParameterValue(self=Get MID(), ParameterName=@n8, Value=@n9)
   n10 = Get Textures() -> (Textures:Map of Names to Texture 2D Object References)
   Branch(execute=@n1.then_2, Condition="integer > integer"(A=Length(TargetMap:Map of Names to Texture 2D Object References=@n10) -> (ReturnValue=0)))
   > n11 = Keys(TargetMap:Map of Names to Texture 2D Object References=Get Textures())
   > n12 = For Each Loop(Array=@n11)
   n13 = Find(TargetMap:Map of Names to Texture 2D Object References=Get Textures(), Key=@n12) -> (Value:Texture 2D Object Reference)
   > Branch(Condition=@n13.ReturnValue)
   > SetTextureParameterValue(self=Get MID(), ParameterName=@n12, Value=@n13)
  ]}
 BakeDynamicMesh {nodes [
   BakeDynamicMesh() {is=title K2Node_FunctionEntry}
   n1 = Get Static Mesh() -> (Static Mesh:Static Mesh Object Reference)
   > n2 = GetStaticMeshDescription!(execute, self:Static Mesh Object Reference=Get Static Mesh(), LODIndex:int=int - int(A=GetNumLODs(self=@n1))) -> (then, ReturnValue:Static Mesh Description Object Reference)
   n3 = GetTriangleCount!(self:Mesh Description Base Object Reference=@n2) -> (ReturnValue:int=0)
   n4 = GetVertexCount!(self:Mesh Description Base Object Reference=@n2) -> (ReturnValue:int=0)
   > PrintString(InString=Append!(A="Triangles/Verts: ", B="To String (Integer)"(InInt=@n3), C:String=" / ", D:String="To String (Integer)"(InInt=@n4)) -> (ReturnValue))
   n5 = "Equal (Integer)"(A=@n3, B=72)
   n6 = "Equal (Integer)"(A=@n4, B=81)
   n7 = "Equal (Integer)"(A=@n3, B=8)
   n8 = "Equal (Integer)"(A=@n4, B=9)
   n9 = Get Impostor Type() -> (Impostor Type:EImposterLayoutType Enum=NewEnumerator0)
   > n10 = Branch(Condition=Select(
     Option 0:bool=AND Boolean(A=@n7, B=@n8)
     Option 1:bool=AND Boolean(A=@n5, B=@n6)
     Index="Equal (Enum)"(A:EImposterLayoutType Enum=@n9, B:EImposterLayoutType Enum=NewEnumerator2)
    ) -> (ReturnValue:bool))
   > n11 = Switch on EImpostorOverwriteLOD!(
    execute
    Selection:EImpostorOverwriteLOD Enum=Get ExistingImpostorFoundBehavior() -> (ExistingImpostorFoundBehavior:EImpostorOverwriteLOD Enum=NewEnumerator0)
   ) -> (NewEnumerator0:Exec, NewEnumerator1:Exec, NewEnumerator2:Exec, NewEnumerator3:Exec) {is=title K2Node_SwitchEnum}
   > SetVisibility(self=Get VerticalBox_LODDialogue() -> (VerticalBox_LODDialogue:Vertical Box Object Reference), InVisibility=Visible)
   n12 = Save Dynamic Mesh To Staticmesh!(execute:Exec=@n11.NewEnumerator1, Overwrite Last LOD:bool=true) -> (then:Exec) {is=title K2Node_MacroInstance}
   Save Complete(execute=[@n12.then, @n13.then, @n11.NewEnumerator3])
   n13 = Save Dynamic Mesh To Staticmesh(execute=@n11.NewEnumerator2)
   Save Dynamic Mesh To Staticmesh(execute=@n10.else)
   > Save Complete()
  ]}
 Capture Imposter Grid {nodes [
   Capture Imposter Grid() {is=title K2Node_FunctionEntry}
   > Set CurrentMap(CurrentMap:EImpostorBakeTypes Enum="Get (a ref)"(
     Array:Array of EImpostorBakeTypes Enums=Get Maps to Render_Internal() -> (Maps to Render_Internal:Array of EImpostorBakeTypes Enums=NewEnumerator0)
     Dimension 1=Get CurrentMapIndex() -> (CurrentMapIndex:int=0)
    ) -> (Output:"EImpostorBakeTypes Enum (by ref)")) -> (Output_Get:EImpostorBakeTypes Enum=NewEnumerator0)
   n1 = Find(
    TargetMap:Map of EImpostorBakeTypes Enums to Texture Render Target 2D Object References=Get TargetMaps() -> (TargetMaps:Map of EImpostorBakeTypes Enums to Texture Render Target 2D Object References=NewEnumerator0)
    Key:"EImpostorBakeTypes Enum (by ref)"=Get CurrentMap() -> (CurrentMap:EImpostorBakeTypes Enum=NewEnumerator0)
   ) -> (Value:Texture Render Target 2D Object Reference)
   > Branch(Condition=@n1.ReturnValue)
   > Set CurrentTargetRT(CurrentTargetRT:Texture Render Target 2D Object Reference=@n1) -> (Output_Get:Texture Render Target 2D Object Reference)
   n2 = Get CurrentMap() -> (CurrentMap=NewEnumerator0)
   n3 = AND Boolean(
    A="Equal (Enum)"(A:EImpostorBakeTypes Enum=Get CurrentMap() -> (CurrentMap=NewEnumerator0), B:EImpostorBakeTypes Enum=NewEnumerator10)
    B=Get CaptureUsingGBuffer() -> (CaptureUsingGBuffer:bool)
   )
   n4 = Select(
    Option 0:float="To Float (Boolean)"(InBool=NOT Boolean(A=@n3))
    Option 1:float=0
    Index="Equal (Enum)"(A=Get CurrentMap() -> (CurrentMap=NewEnumerator0), B=NewEnumerator9)
   ) -> (ReturnValue:float)
   n5 = Get World() -> (World:Object Reference)
   > SetScalarParameterValue!(
    execute
    WorldContextObject=@n5
    Collection:Material Parameter Collection Object Reference=/ImpostorBaker/Materials/MPC/MPC_ImposterCapture
    ParameterName:Name=String To Name(InString=Enum to String!(Enumerator:EImpostorBakeTypes Enum=@n2) -> (ReturnValue:String))
    ParameterValue:float=@n4
   ) -> (then) {is=title K2Node_CallMaterialParameterCollectionFunction}
   > SetScalarParameterValue(WorldContextObject=@n5, Collection=&MPC_ImposterCapture, ParameterName=EnableCapture, ParameterValue=@n4)
   > SetScalarParameterValue(WorldContextObject=@n5, Collection=&MPC_ImposterCapture, ParameterName=Dilation, ParameterValue=1)
   > n6 = Branch(Condition="Equal (Enum)"(A=Get CurrentMap() -> (CurrentMap=NewEnumerator0), B=NewEnumerator10))
   > n7 = Display Custom Lighting!(execute:Exec) -> (out:Exec) {is=title K2Node_MacroInstance}
   Set curframe(execute=[@n7.out, @n6.else], curframe:int=0) -> (Output_Get:int=0)
   > n8 = For Each Loop(Array:Array of Vectors=Get ViewCaptureVectors() -> (ViewCaptureVectors:Array of Vectors)) -> (Array Element:Vector)
   n9 = Get CurrentMap() -> (CurrentMap=NewEnumerator0)
   n10 = Get World()
   SetScalarParameterValue(
    execute=@n8.Completed
    WorldContextObject=@n10
    Collection=&MPC_ImposterCapture
    ParameterName=String To Name(InString=Enum to String(Enumerator=@n9))
    ParameterValue=0
   )
   > SetScalarParameterValue(WorldContextObject=@n10, Collection=&MPC_ImposterCapture, ParameterName=EnableCapture, ParameterValue=0)
   > Restore Lighting State!(execute:Exec) {is=title K2Node_MacroInstance}
   Set CurVector(execute=@n8.LoopBody, CurVector:Vector=@n8) -> (Output_Get:Vector)
   > n11 = Sequence!(execute) -> (then_0, then_1)
   n12 = float * float(A=Get Object Radius() -> (Object Radius:double), B=1.25)
   n13 = Get Orthographic() -> (Orthographic:bool)
   n14 = Get Camera Distance() -> (Camera Distance:double)
   > Set World Location!(
    execute
    self:Scene Component Object Reference=Get SceneCaptureComponent2D() -> (SceneCaptureComponent2D:Scene Capture Component 2D Object Reference)
    NewLocation:Vector=vector * float(A=Get CurVector() -> (CurVector:Vector), B=Select(Option 0:double=@n14, Option 1:double=@n12, Index=@n13) -> (ReturnValue:double))
    bSweep:bool=false
    bTeleport:bool=false
   ) -> (then, SweepHitResult:Hit Result Structure) {is=title K2Node_CallFunction}
   n15 = Get CurVector()
   > Set World Rotation!(
    execute
    self:Scene Component Object Reference=Get SceneCaptureComponent2D()
    NewRotation:Rotator=Rotation From X Vector!(InVec:Vector=vector * float(A=@n15, B=-1)) -> (ReturnValue:Rotator)
    bSweep:bool=false
    bTeleport:bool=false
   ) -> (then, SweepHitResult:Hit Result Structure) {is=title K2Node_CallFunction}
   n16 = Get World()
   > SetVectorParameterValue!(
    execute
    WorldContextObject=@n16
    Collection:Material Parameter Collection Object Reference=&MPC_ImposterCapture
    ParameterName:Name=Origin
    ParameterValue:"Linear Color Structure (by ref)"="To LinearColor (Vector)"!(InVec:Vector="0, 0, 0") -> (ReturnValue:Linear Color Structure)
   ) -> (then) {is=title K2Node_CallMaterialParameterCollectionFunction}
   > SetScalarParameterValue(WorldContextObject=@n16, Collection=&MPC_ImposterCapture, ParameterName=R, ParameterValue=Get Object Radius())
   n17 = Derive Axes!(A:Vector=Get CurVector()) -> (X:Vector=~, Y:Vector=~, Z:Vector=~) {is=title K2Node_MacroInstance}
   n18 = Get World()
   > SetVectorParameterValue(
    WorldContextObject=@n18
    Collection=&MPC_ImposterCapture
    ParameterName=X
    ParameterValue="To LinearColor (Vector)"(InVec=@n17)
   )
   > SetVectorParameterValue(
    WorldContextObject=@n18
    Collection=&MPC_ImposterCapture
    ParameterName=Y
    ParameterValue="To LinearColor (Vector)"(InVec=@n17.Y)
   )
   > SetVectorParameterValue(
    WorldContextObject=@n18
    Collection=&MPC_ImposterCapture
    ParameterName=Z
    ParameterValue="To LinearColor (Vector)"(InVec=@n17.Z)
   )
   > CaptureScene(self:Scene Capture Component 2D Object Reference=Get SceneCaptureComponent2D())
   > Render Scene Capture Mip Chain!(execute:Exec) -> (then:Exec) {is=title K2Node_MacroInstance}
   > Draw Frame to Atlas!(
    execute
    self
    TextureRenderTarget:Texture Render Target 2D Object Reference=Get CurrentTargetRT() -> (CurrentTargetRT:Texture Render Target 2D Object Reference)
   ) -> (then)
   n19 = Get CurrentMap() -> (CurrentMap=NewEnumerator0)
   n20 = Get curframe() -> (curframe:int=0)
   > Branch(Condition=AND Boolean(A="Equal (Integer)"(A=@n20, B=6), B="Equal (Enum)"(A=@n19, B=NewEnumerator8)))
   n21 = Get SceneCaptureComponent2D()
   > Set TextureTarget!(
    execute
    TextureTarget:Texture Render Target 2D Object Reference=/ImpostorBaker/TestContent/RTs/RT_Test_01
    self:Scene Capture Component 2D Object Reference=@n21
   ) -> (then, Output_Get:Texture Render Target 2D Object Reference)
   > CaptureScene(self=@n21)
   > Set TextureTarget(
    TextureTarget="Get (a copy)"(Array:Array of Texture Render Target 2D Object References=Get SceneCaptureMipChain() -> (SceneCaptureMipChain:Array of Texture Render Target 2D Object References)) -> (Output:Texture Render Target 2D Object Reference)
    self=Get SceneCaptureComponent2D()
   )
   Increment Int(" "=@n11.then_1, Value=Get curframe() -> (curframe=0))
  ]}
 ChannelPackMasks {nodes [
   ChannelPackMasks() {is=title K2Node_FunctionEntry}
   > n1 = Keys(TargetMap:Map of Impostor Channel Pack Settings Structures to Texture Render Target 2D Object References=Get ChannelPackedMaps() -> (ChannelPackedMaps:Map of Impostor Channel Pack Settings Structures to Texture Render Target 2D Object References)) -> (Keys:Array of Impostor Channel Pack Settings Structures)
   > n2 = For Each Loop(Array:Array of Impostor Channel Pack Settings Structures=@n1) -> (Array Element:Impostor Channel Pack Settings Structure)
   n3 = Break Impostor Channel Pack Settings!(ImpostorChannelPackSettings:"Impostor Channel Pack Settings Structure (by ref)"=@n2) -> (
    R_10_#1:EImpostorScalarMaps Enum
    G_11_#2:EImpostorScalarMaps Enum
    B_12_#3:EImpostorScalarMaps Enum
    A_13_#4:EImpostorScalarMaps Enum
   ) {is=title K2Node_BreakStruct}
   > n4 = For Each Loop(Array:Array of EImpostorBakeTypes Enums=Make Array!(
     "[0]":EImpostorBakeTypes Enum=Scalar Map to Bake Type!(Index:EImpostorScalarMaps Enum=@n3) -> (ReturnValue:EImpostorBakeTypes Enum=NewEnumerator0) {is=title K2Node_MacroInstance}
     "[1]":EImpostorBakeTypes Enum=Scalar Map to Bake Type(Index=@n3.G_11_#2) -> (ReturnValue=NewEnumerator0)
     "[2]":EImpostorBakeTypes Enum=Scalar Map to Bake Type(Index=@n3.B_12_#3) -> (ReturnValue=NewEnumerator0)
     "[3]":EImpostorBakeTypes Enum=Scalar Map to Bake Type(Index=@n3.A_13_#4) -> (ReturnValue=NewEnumerator0)
    ) -> (Array:Array of EImpostorBakeTypes Enums)) -> (Array Element:EImpostorBakeTypes Enum)
   n5 = Find(TargetMap=Get TargetMaps() -> (TargetMaps=NewEnumerator0), Key=@n4)
   n6 = Find(TargetMap=Get TargetMaps() -> (TargetMaps=NewEnumerator0), Key=NewEnumerator0)
   n7 = "Equal (Enum)"(A=@n4, B=NewEnumerator12)
   > Branch(Condition=OR Boolean(A=@n5.ReturnValue, B=AND Boolean(A=@n6.ReturnValue, B=@n7)))
   n8 = Append!(A=RT, B="To String (Integer)"(InInt=int + int(A=@n4.Array Index))) -> (ReturnValue)
   n9 = Get ChannelPackMasksMID() -> (ChannelPackMasksMID:Material Instance Dynamic Object Reference)
   > SetTextureParameterValue(
    self=@n9
    ParameterName=String To Name(InString=@n8)
    Value=Select(Option 0:Texture Object Reference=@n5, Option 1:Texture Object Reference=@n6, Index=@n7) -> (ReturnValue:Texture Object Reference)
   )
   > SetScalarParameterValue(self=@n9, ParameterName=String To Name(InString=Append(A=@n8, B=Alpha)), Value="To Float (Boolean)"(InBool=@n7)) {is=title K2Node_CallFunction}
   n10 = Find(
    TargetMap:Map of Impostor Channel Pack Settings Structures to Texture Render Target 2D Object References=Get ChannelPackedMaps()
    Key:"Impostor Channel Pack Settings Structure (by ref)"=@n2
   )
   ClearRenderTarget2D!(
    execute=@n4.Completed
    WorldContextObject:Object Reference
    TextureRenderTarget:Texture Render Target 2D Object Reference=@n10
    ClearColor:Linear Color Structure=(R=0,G=0,B=0,A=1)
   ) -> (then)
   > DrawMaterialToRenderTarget!(
    execute
    WorldContextObject:Object Reference
    TextureRenderTarget:Texture Render Target 2D Object Reference=@n10
    Material:Material Interface Object Reference=Get ChannelPackMasksMID()
   ) -> (then)
  ]}
 Check and Create RT {nodes [
   n1 = Inputs!() -> (In:Exec, RT:Texture Render Target 2D Object Reference, Width:int, Height:int, Format:ETextureRenderTargetFormat Enum=RTF_R8)
   > n2 = Is Valid(InputObject=@n1)
   n3 = Get SizeY!(self:Texture Render Target 2D Object Reference=@n1) -> (SizeY:int=0)
   n4 = Get SizeX!(self:Texture Render Target 2D Object Reference=@n1) -> (SizeX:int=0)
   n5 = Get RenderTargetFormat!(self:Texture Render Target 2D Object Reference=@n1) -> (RenderTargetFormat:ETextureRenderTargetFormat Enum=RTF_R8)
   > n6 = Branch(Condition=AND Boolean!(
     A="Equal (Integer)"(A=@n4, B=@n1.Width) {is=title K2Node_CallFunction}
     B="Equal (Integer)"(A=@n3, B=@n1.Height)
     C:bool="Equal (Enum)"(A:ETextureRenderTargetFormat Enum=@n5, B:ETextureRenderTargetFormat Enum=@n1.Format)
    ) -> (ReturnValue))
   Outputs!(Out:Exec=[@n8.then, @n6.then])
   n7 = CreateRenderTarget2D!(
    execute=[@n6.else, @n2.Is Not Valid]
    WorldContextObject:Object Reference=Get World()
    Width:int=@n1.Width
    Height:int=@n1.Height
    Format:ETextureRenderTargetFormat Enum=@n1.Format
    ClearColor:Linear Color Structure=(R=0,G=0,B=0,A=1)
    bAutoGenerateMipMaps:bool=false
    bSupportUAVs:bool=false
   ) -> (then, ReturnValue:Texture Render Target 2D Object Reference)
   > n8 = "Set Texture Render Target 2D Object Reference (by ref)"!(execute:Exec, Target:"Texture Render Target 2D Object Reference (by ref)"=@n1, Value:Texture Render Target 2D Object Reference=@n7) -> (then:Exec) {is=title K2Node_VariableSetRef}
  ]}
 Clear RenderTargets {nodes [
   Clear RenderTargets() {is=title K2Node_FunctionEntry}
   > ClearRenderTarget2D(TextureRenderTarget=Get CombinedAlphasRT() -> (CombinedAlphasRT:Texture Render Target 2D Object Reference), ClearColor=(R=0,G=0,B=0,A=1))
   > ClearRenderTarget2D(TextureRenderTarget=Get ScratchRT() -> (ScratchRT:Texture Render Target 2D Object Reference), ClearColor=(R=0,G=0,B=0,A=1))
   > n1 = For Each Loop(Array:Array of Texture Render Target 2D Object References=Get SceneCaptureMipChain()) -> (Array Element:Texture Render Target 2D Object Reference)
   n2 = Values!(
    execute=@n1.Completed
    TargetMap:Map of EImpostorBakeTypes Enums to Texture Render Target 2D Object References=Get TargetMaps() -> (TargetMaps=NewEnumerator0)
   ) -> (then, Values:Array of Texture Render Target 2D Object References)
   > n3 = For Each Loop(Array=@n2)
   > ClearRenderTarget2D(TextureRenderTarget=@n3, ClearColor=(R=0,G=0,B=0,A=1))
   ClearRenderTarget2D(execute=@n1.LoopBody, TextureRenderTarget=@n1, ClearColor=(R=0,G=0,B=0,A=1))
  ]}
 CollapsedOrVisible {nodes [Outputs!(ReturnValue:ESlateVisibility Enum=Select(Option 0:ESlateVisibility Enum=Collapsed, Option 1:ESlateVisibility Enum=Visible, Index=Inputs!() -> (Index:bool=~)) -> (ReturnValue:ESlateVisibility Enum))]}
 CombineRenderTypes {nodes [
   Inputs!() -> (Exec:Exec)
   > Set BakingDepth-to-ScalarMaps(BakingDepth-to-ScalarMaps:bool=false) -> (Output_Get:bool)
   > Clear(TargetArray:Array of EImpostorBakeTypes Enums=Get Maps to Render_Internal() -> (Maps to Render_Internal=NewEnumerator0))
   > n1 = Sequence!(execute) -> (then_0, then_1, then_2:Exec)
   > n2 = For Each Loop(Array:Array of EImpostorColorMaps Enums=Get Color Maps to Render() -> (Color Maps to Render:Array of EImpostorColorMaps Enums=NewEnumerator0)) -> (Array Element:EImpostorColorMaps Enum)
   > Add Unique(
    TargetArray:Array of EImpostorBakeTypes Enums=Get Maps to Render_Internal() -> (Maps to Render_Internal=NewEnumerator0)
    NewItem:"EImpostorBakeTypes Enum (by ref)"=Select!(
     NewEnumerator0:EImpostorBakeTypes Enum=NewEnumerator0
     NewEnumerator4:EImpostorBakeTypes Enum=NewEnumerator4
     NewEnumerator6:EImpostorBakeTypes Enum=NewEnumerator6
     NewEnumerator15:EImpostorBakeTypes Enum=NewEnumerator15
     NewEnumerator7:EImpostorBakeTypes Enum=NewEnumerator7
     NewEnumerator9:EImpostorBakeTypes Enum=NewEnumerator9
     NewEnumerator10:EImpostorBakeTypes Enum=NewEnumerator10
     Index:EImpostorColorMaps Enum=@n2
    ) -> (ReturnValue:EImpostorBakeTypes Enum)
   )
   n3 = For Each Loop(
    Exec=@n1.then_1
    Array:Array of Impostor Channel Pack Settings Structures=Get Channel Packed Masks() -> (Channel Packed Masks:Array of Impostor Channel Pack Settings Structures)
   ) -> (Array Element:Impostor Channel Pack Settings Structure)
   Add Unique(execute=@n3.Completed, TargetArray=Get Maps to Render_Internal() -> (Maps to Render_Internal=NewEnumerator0), NewItem=NewEnumerator8)
   > Remove Item!(
    execute
    TargetArray:Array of EImpostorBakeTypes Enums=Get Maps to Render_Internal() -> (Maps to Render_Internal=NewEnumerator0)
    Item:"EImpostorBakeTypes Enum (by ref)"=NewEnumerator12
   ) -> (then, ReturnValue:bool)
   > Remove Item(TargetArray=Get Maps to Render_Internal() -> (Maps to Render_Internal=NewEnumerator0), Item=NewEnumerator18)
   > Outputs!(Completed:Exec)
   n4 = Break Impostor Channel Pack Settings(ImpostorChannelPackSettings=@n3)
   n5 = For Each Loop(
    Exec=@n3.LoopBody
    Array:Array of EImpostorBakeTypes Enums=Make Array(
     "[0]"=Scalar Map to Bake Type(Index=@n4) -> (ReturnValue=NewEnumerator0)
     "[1]"=Scalar Map to Bake Type(Index=@n4.G_11_#2) -> (ReturnValue=NewEnumerator0)
     "[2]"=Scalar Map to Bake Type(Index=@n4.B_12_#3) -> (ReturnValue=NewEnumerator0)
     "[3]"=Scalar Map to Bake Type(Index=@n4.A_13_#4) -> (ReturnValue=NewEnumerator0)
    )
   ) -> (Array Element:EImpostorBakeTypes Enum)
   > Add Unique(TargetArray=Get Maps to Render_Internal() -> (Maps to Render_Internal=NewEnumerator0), NewItem=@n5)
   > Branch(Condition="Equal (Enum)"(A:EImpostorBakeTypes Enum=@n5, B:EImpostorBakeTypes Enum=NewEnumerator8))
   > Set BakingDepth-to-ScalarMaps(BakingDepth-to-ScalarMaps=true)
   Add Unique(
    TargetArray:Array of EImpostorColorMaps Enums=Get Color Maps to Render() -> (Color Maps to Render=NewEnumerator0)
    NewItem:"EImpostorColorMaps Enum (by ref)"=NewEnumerator0
   )
  ]}
 Create Static Assets {nodes [
   Create Static Assets() {is=title K2Node_FunctionEntry}
   > Process Path Name!(execute:Exec) -> (then:Exec) {is=title K2Node_MacroInstance}
   > Clear(TargetArray:Array of Texture 2D Object References=Get TexturesToRemoveAlpha() -> (TexturesToRemoveAlpha:Array of Texture 2D Object References))
   > n1 = Sequence!(execute) -> (then_0, then_1, then_2, then_3:Exec)
   n2 = Get Processed MIC Path() -> (Processed MIC Path:String)
   n3 = Is Path Valid Soft Obj!(PathString:String=Object Path Add Outer!(Path:String=@n2) -> (OutPath:String) {is=title K2Node_MacroInstance}) -> (ReturnValue:bool=~, Object:Object Reference) {is=title K2Node_MacroInstance}
   > n4 = Branch(Condition=@n3)
   > PrintString(
    WorldContextObject=Get World()
    InString="ImpostorGenerator: MIC Path pointed to already existing Asset.\r\nAttempting to reuse MIC and set to updated parent and params."
   )
   > n5 = Cast To MaterialInstanceConstant(Object=@n3.Object)
   > n6 = Set Created MIC(Created MIC:Material Instance Constant Object Reference=@n5) -> (Output_Get:Material Instance Constant Object Reference)
   n7 = Select!(
    NewEnumerator0:Material Interface Object Reference=Get Full Sphere Material() -> (Full Sphere Material:Material Interface Object Reference)
    NewEnumerator1:Material Interface Object Reference=Get Hemisphere Material() -> (Hemisphere Material:Material Interface Object Reference)
    NewEnumerator2:Material Interface Object Reference=Get Billboard Material() -> (Billboard Material:Material Interface Object Reference)
    Index:EImposterLayoutType Enum=Get Impostor Type() -> (Impostor Type=NewEnumerator0)
   ) -> (ReturnValue:Material Interface Object Reference)
   > n8 = SetMaterialInstanceParent!(execute, Instance:Material Instance Constant Object Reference=@n6, NewParent:Material Interface Object Reference=@n7) -> (then)
   Set MIC Switches and ShadingModel!(
    execute=[@n8.then, @n10.then]
    self
    In MIC:Material Instance Constant Object Reference=Get Created MIC() -> (Created MIC:Material Instance Constant Object Reference)
    Creating Final Assets:bool=true
   ) -> (then) {is=title K2Node_CallFunction}
   > Set MIC Parameters!(execute:Exec) {is=title K2Node_MacroInstance}
   n9 = Create MIC Editor Only!(execute=@n4.else, Material:Material Interface Object Reference=@n7, Name:String=@n2) -> (then, ReturnValue:Material Instance Constant Object Reference)
   > n10 = Set Created MIC(Created MIC=@n9)
   n11 = For Each Loop(Exec=@n1.then_1, Array=Get Maps to Save() -> (Maps to Save:Array of EImpostorBakeTypes Enums=NewEnumerator0))
   n12 = Find(
    TargetMap:Map of EImpostorBakeTypes Enums to Texture Render Target 2D Object References=Get TargetMaps() -> (TargetMaps=NewEnumerator0)
    Key:"EImpostorBakeTypes Enum (by ref)"=@n11
   )
   > Branch(Condition=@n12.ReturnValue)
   n13 = Enum to String(Enumerator=@n11)
   n14 = Append!(A=Get Processed Texture Path() -> (Processed Texture Path:String), B=_, C:String=@n13) -> (ReturnValue)
   > PrintString(InString=Append!(A="trying to make texture at: ", B=@n14) -> (ReturnValue))
   n15 = Is Path Valid Soft Obj(PathString=Object Path Add Outer(Path=@n14)) -> (ReturnValue=~)
   > n16 = Branch(Condition=@n15)
   > n17 = Cast To Texture2D(Object=@n15.Object)
   > n18 = Set CurrentTexture(CurrentTexture:Texture 2D Object Reference=@n17) -> (Output_Get:Texture 2D Object Reference)
   > ConvertRenderTargetToTexture2DEditorOnly(RenderTarget=@n12, Texture=@n18)
   > PrintString(InString=Append(A="Existing texture found, overwrote existing asset: ", B=@n14))
   > n19 = Set MIC Texture Parameter Editor Only!(
    execute
    Material:Material Instance Constant Object Reference=Get Created MIC()
    ParamName:String=@n13
    Texture:Texture 2D Object Reference=Get CurrentTexture() -> (CurrentTexture:Texture 2D Object Reference)
   ) -> (then, ReturnValue:bool) {is=title K2Node_CallFunction}
   K2Node_Composite_0 = Check if Needs Alpha!(Map Type:EImpostorBakeTypes Enum=@n11) -> (Requires Alpha:bool=~) {is=title K2Node_Composite}
   Branch(execute=[@n22.then, @n19.then], Condition=NOT Boolean(A=@K2Node_Composite_0))
   > Add(
    TargetArray:Array of Texture 2D Object References=Get TexturesToRemoveAlpha()
    NewItem:"Texture 2D Object Reference (by ref)"=Get CurrentTexture()
   ) {is=title K2Node_CallArrayFunction}
   PrintString(execute=@n17.CastFailed, InString=Cast Failed to existing render target)
   n20 = Render Target 2D Create Static Texture 2D Editor Only!(
    execute=@n16.else
    RenderTarget:Texture Render Target 2D Object Reference=@n12
    Name:String=@n14
    CompressionSettings:TextureCompressionSettings Enum=TC_Default
    MipSettings:TextureMipGenSettings Enum=TMGS_FromTextureGroup
   ) -> (then, ReturnValue:Texture 2D Object Reference)
   > n21 = Set CurrentTexture(CurrentTexture=@n20)
   > n22 = Set MIC Texture Parameter Editor Only(Material=Get Created MIC(), ParamName=@n13, Texture=@n21)
   BakeDynamicMesh(execute=@n1.then_3) {is=title K2Node_CallFunction}
   n23 = Keys(execute=@n1.then_2, TargetMap=Get ChannelPackedMaps())
   > n24 = For Each Loop(Array:Array of Impostor Channel Pack Settings Structures=@n23) -> (Array Element:Impostor Channel Pack Settings Structure)
   n25 = Append(A=PackedMasks_, B="To String (Integer)"(InInt=int + int(A=@n24.Array Index)))
   n26 = Append!(A=Get Processed Texture Path(), B=_, C:String=@n25) -> (ReturnValue)
   n27 = Is Path Valid Soft Obj(PathString=Object Path Add Outer(Path=@n26)) -> (ReturnValue=~)
   > n28 = Branch(Condition=@n27)
   > n29 = Cast To Texture2D(Object=@n27.Object)
   > n30 = Set CurrentTexture(CurrentTexture=@n29)
   n31 = Find(
    TargetMap:Map of Impostor Channel Pack Settings Structures to Texture Render Target 2D Object References=Get ChannelPackedMaps()
    Key:"Impostor Channel Pack Settings Structure (by ref)"=@n24
   )
   > ConvertRenderTargetToTexture2DEditorOnly(RenderTarget=@n31, Texture=@n30)
   > PrintString(InString=Append!(A="Existing texture found, overwrote existing asset: ", B=@n26) -> (ReturnValue))
   > n32 = Set MIC Texture Parameter Editor Only(Material=Get Created MIC(), ParamName=@n25, Texture=Get CurrentTexture())
   Branch(
    execute=[@n32.then, @n35.then]
    Condition="Equal (Enum)"(
     A:EImpostorScalarMaps Enum=Break Impostor Channel Pack Settings(ImpostorChannelPackSettings=@n24).A_13_#4
     B:EImpostorScalarMaps Enum=NewEnumerator18
    )
   )
   > Add(TargetArray=Get TexturesToRemoveAlpha(), NewItem=Get CurrentTexture())
   PrintString(execute=@n29.CastFailed, InString=Cast Failed to existing render target)
   n33 = Render Target 2D Create Static Texture 2D Editor Only(execute=@n28.else, RenderTarget=@n31, Name=@n26, CompressionSettings=TC_Default, MipSettings=TMGS_FromTextureGroup)
   > n34 = Set CurrentTexture(CurrentTexture=@n33)
   > n35 = Set MIC Texture Parameter Editor Only(Material=Get Created MIC(), ParamName=@n25, Texture=@n34)
   Set CompressionNoAlpha!(execute, CompressionNoAlpha:bool=false, self:Texture Object Reference) -> (then, Output_Get:bool)
   Set CompressionNoAlpha(CompressionNoAlpha=false)
  ]}
 Check if Needs Alpha {
  path=Create Static Assets.K2Node_Composite_0.Check if Needs Alpha
  nodes [
   n1 = Inputs!() -> (Map Type:EImpostorBakeTypes Enum=NewEnumerator0)
   n2 = NOT Boolean(A=Get BakingDepth-to-ScalarMaps() -> (BakingDepth-to-ScalarMaps:bool))
   n3 = "Equal (Enum)"(A:EImpostorBakeTypes Enum=@n1, B:EImpostorBakeTypes Enum=NewEnumerator6)
   Outputs!(Requires Alpha:bool=Select!(Option 0:bool=Select(Option 0=false, Option 1=@n2, Index=@n3), Option 1:bool=true, Index:bool="Equal (Enum)"(A=@n1, B=NewEnumerator0)) -> (ReturnValue:bool))
  ]
 }
 Create and Setup MIDs {nodes [
   Inputs!() -> (In:Exec)
   > n1 = Sequence!(execute) -> (then_0, then_1)
   > UpdatePreviewMaterialProxy()
   > Create And Assign MID!(
    execute:Exec
    MID:Material Instance Dynamic Object Reference=Get ImpostorPreviewMID() -> (ImpostorPreviewMID:Material Instance Dynamic Object Reference)
    Parent:Material Interface Object Reference=Get Material Preview Parent() -> (Material Preview Parent:Material Interface Object Reference)
    Optional Name:Name=ImpostorPreviewMID
    Creation Flags:EMIDCreationFlags Enum=None
   ) -> (Out:Exec) {is=title K2Node_MacroInstance}
   > n2 = CreateDynamicMaterialInstance(Parent=/ImpostorBaker/Materials/Generation/M_BaseColor_CustomLighting, OptionalName=None, CreationFlags=None)
   > Set BaseColorCustomLightingMID(BaseColorCustomLightingMID:Material Instance Dynamic Object Reference=@n2) -> (Output_Get:Material Instance Dynamic Object Reference)
   > n3 = CreateDynamicMaterialInstance(Parent=/ImpostorBaker/Materials/Generation/M_CombineNormalsDepth, OptionalName=None, CreationFlags=None)
   > Set CombineNormalsDepthMID(CombineNormalsDepthMID:Material Instance Dynamic Object Reference=@n3) -> (Output_Get:Material Instance Dynamic Object Reference)
   > n4 = CreateDynamicMaterialInstance(Parent=/ImpostorBaker/Materials/Generation/M_Sample_Frame, OptionalName=None, CreationFlags=None)
   > Set SampleFrameMID(SampleFrameMID:Material Instance Dynamic Object Reference=@n4) -> (Output_Get:Material Instance Dynamic Object Reference)
   > n5 = CreateDynamicMaterialInstance(Parent=/ImpostorBaker/Materials/Generation/M_Sample_Frame_DFAlpha, OptionalName=None, CreationFlags=None)
   > Set SampleFrame_DFAlphaMID(SampleFrame_DFAlphaMID:Material Instance Dynamic Object Reference=@n5) -> (Output_Get:Material Instance Dynamic Object Reference)
   > n6 = CreateDynamicMaterialInstance(Parent=/ImpostorBaker/Materials/Generation/M_AddAlpha, OptionalName=None, CreationFlags=None)
   > Set AddAlphasMID(AddAlphasMID:Material Instance Dynamic Object Reference=@n6) -> (Output_Get:Material Instance Dynamic Object Reference)
   > n7 = CreateDynamicMaterialInstance(Parent=/ImpostorBaker/Materials/Generation/M_Resample_RT, OptionalName=None, CreationFlags=None)
   > Set Resample MID(Resample MID:Material Instance Dynamic Object Reference=@n7) -> (Output_Get:Material Instance Dynamic Object Reference)
   > n8 = CreateDynamicMaterialInstance(Parent=/ImpostorBaker/Materials/Generation/M_ChannelPack, OptionalName=None, CreationFlags=None)
   > Set ChannelPackMasksMID(ChannelPackMasksMID:Material Instance Dynamic Object Reference=@n8) -> (Output_Get:Material Instance Dynamic Object Reference)
   > SetMaterial(
    self=Get DynamicMeshComponent() -> (DynamicMeshComponent:Dynamic Mesh Component Object Reference)
    ElementIndex=0
    Material=Get ImpostorPreviewMID()
   )
   > n9 = Sequence!(execute) -> (then_0, then_1, then_2:Exec, then_3:Exec)
   > n10 = Is Valid(InputObject=Get ImpostorPreviewMID())
   PrintString(execute=@n10.Is Not Valid, InString=Impostor Generator: Preview MID setup Failed due to missing Impostor Material Selection.)
   n11 = Sequence(execute=@n10.Is Valid)
   n12 = Get ImpostorPreviewMID()
   > SetScalarParameterValue(self=@n12, ParameterName=Specular, Value=Get Constant Specular() -> (Constant Specular:double))
   > SetScalarParameterValue(self=@n12, ParameterName=Roughness, Value=Get Constant Roughness() -> (Constant Roughness:double))
   > SetScalarParameterValue(self=@n12, ParameterName=Opacity, Value=Get Constant Opacity() -> (Constant Opacity:double))
   > SetScalarParameterValue(self=@n12, ParameterName=Scatter Mask Min, Value=Get Scatter Mask Min() -> (Scatter Mask Min:double))
   > SetScalarParameterValue(self=@n12, ParameterName=Scatter Mask Length, Value=Get Scatter Mask Length() -> (Scatter Mask Length:double))
   > SetVectorParameterValue(self=@n12, ParameterName=SubsurfaceColor, Value=Get Subsurface Color() -> (Subsurface Color:Linear Color Structure)) {is=title K2Node_CallFunction}
   > SetScalarParameterValue(self=@n12, ParameterName=MaskOffset, Value=Get Mask Offset() -> (Mask Offset:double))
   > Branch(Condition="Equal (Enum)"(A:EImposterLayoutType Enum=Get Impostor Type() -> (Impostor Type=NewEnumerator0), B:EImposterLayoutType Enum=NewEnumerator2))
   n13 = Get ImpostorPreviewMID()
   > SetScalarParameterValue(self=@n13, ParameterName=Reference Plane Sides, Value=Get Reference Plane Sides() -> (Reference Plane Sides:double))
   > SetScalarParameterValue(self=@n13, ParameterName=Reference Plane Top, Value=Get Reference Plane Top() -> (Reference Plane Top:double))
   n14 = For Each Loop(Exec=@n11.then_2, Array:Array of EImpostorBakeTypes Enums=Get Maps to Save() -> (Maps to Save=NewEnumerator0)) -> (Array Element:EImpostorBakeTypes Enum)
   > SetTextureParameterValue(
    self=Get ImpostorPreviewMID()
    ParameterName=String To Name(InString=Enum to String(Enumerator=@n14))
    Value=Find(
     TargetMap:Map of EImpostorBakeTypes Enums to Texture Render Target 2D Object References=Get TargetMaps() -> (TargetMaps=NewEnumerator0)
     Key:"EImpostorBakeTypes Enum (by ref)"=@n14
    )
   )
   n15 = Keys(execute=@n11.then_3, TargetMap=Get ChannelPackedMaps())
   > n16 = For Each Loop(Array:Array of Impostor Channel Pack Settings Structures=@n15) -> (Array Element:Impostor Channel Pack Settings Structure)
   n17 = "To String (Integer)"(InInt=int + int(A=@n16.Array Index))
   > SetTextureParameterValue(
    self=Get ImpostorPreviewMID()
    ParameterName=String To Name(InString=Append(A=PackedMasks_, B=@n17))
    Value=Find(
     TargetMap:Map of Impostor Channel Pack Settings Structures to Texture Render Target 2D Object References=Get ChannelPackedMaps()
     Key:"Impostor Channel Pack Settings Structure (by ref)"=@n16
    )
   )
   n18 = Get ImpostorPreviewMID()
   SetScalarParameterValue(execute=@n11.then_1, self=@n18, ParameterName=FramesXY, Value="To Float (Integer)"(InInt=Get FramesXYInternal() -> (FramesXYInternal:int=0)))
   > SetScalarParameterValue(self=@n18, ParameterName=Default Mesh Size, Value=float * float(A=Get Object Radius(), B=2))
   > SetVectorParameterValue(self=@n18, ParameterName=Pivot Offset, Value="To LinearColor (Vector)"(InVec=Get Offset Vector() -> (Offset Vector:Vector)))
   n19 = Get SampleFrameMID() -> (SampleFrameMID:Material Instance Dynamic Object Reference)
   n20 = Get SampleFrame_DFAlphaMID() -> (SampleFrame_DFAlphaMID:Material Instance Dynamic Object Reference)
   SetTextureParameterValue(execute=@n9.then_3, self=[@n19, @n20], ParameterName=RT, Value="Get (a copy)"(Array=Get SceneCaptureMipChain()))
   > SetTextureParameterValue(
    self=[@n19, @n20]
    ParameterName=MipRT
    Value="Get (a copy)"(Array=Get SceneCaptureMipChain(), Dimension 1=Get SceneCaptureMipToComposite() -> (SceneCaptureMipToComposite:int=0))
   )
   n21 = Get SceneCaptureMipChain()
   n22 = Last Index!(TargetArray:Array of Texture Render Target 2D Object References=@n21) -> (ReturnValue=0)
   n23 = Get DFMipTarget() -> (DFMipTarget:int=0)
   > SetTextureParameterValue(
    self=Get SampleFrame_DFAlphaMID()
    ParameterName=LowResMaskRT
    Value="Get (a copy)"(Array=@n21, Dimension 1="Min (Integer)"(A=@n22, B=@n23) -> (ReturnValue=0))
   )
   n24 = Get SceneCaptureMipChain()
   n25 = Last Index(TargetArray=@n24) -> (ReturnValue=0)
   n26 = Get CutoutMipTarget() -> (CutoutMipTarget:int=0)
   > SetTextureParameterValue(
    self=Get AddAlphasMID() -> (AddAlphasMID:Material Instance Dynamic Object Reference)
    ParameterName=MipRT
    Value="Get (a copy)"(Array=@n24, Dimension 1="Min (Integer)"(A=@n25, B=@n26) -> (ReturnValue=0))
   )
   > SetTextureParameterValue(
    self=Get BaseColorCustomLightingMID() -> (BaseColorCustomLightingMID:Material Instance Dynamic Object Reference)
    ParameterName=BaseColor
    Value=Get ScratchRT()
   )
   n27 = Find(
    TargetMap:Map of EImpostorBakeTypes Enums to Texture Render Target 2D Object References=Get TargetMaps() -> (TargetMaps=NewEnumerator0)
    Key:"EImpostorBakeTypes Enum (by ref)"=NewEnumerator10
   )
   > SetTextureParameterValue(
    self=Get BaseColorCustomLightingMID()
    ParameterName=CustomLighting
    Value=Select(
     Option 0:Texture Object Reference=/Engine/ArtTools/RenderToTexture/Textures/T_EV_BlankWhite_01
     Option 1:Texture Object Reference=@n27
     Index=@n27.ReturnValue
    ) -> (ReturnValue:Texture Object Reference)
   )
   > SetTextureParameterValue(
    self=Get CombineNormalsDepthMID() -> (CombineNormalsDepthMID:Material Instance Dynamic Object Reference)
    ParameterName=Normal
    Value=Get ScratchRT()
   )
   n28 = Find(TargetMap=Get TargetMaps() -> (TargetMaps=NewEnumerator0), Key=NewEnumerator8)
   > SetTextureParameterValue(
    self=Get CombineNormalsDepthMID()
    ParameterName=Depth
    Value=Select(Option 0=/Engine/ArtTools/RenderToTexture/Textures/127grey, Option 1=@n28, Index=@n28.ReturnValue)
   )
   > SetScalarParameterValue(
    Collection=&MPC_ImposterCapture
    ParameterName=CustomLightingPower
    ParameterValue=Get Custom Lighting Power() -> (Custom Lighting Power:double)
   ) {is=title K2Node_CallMaterialParameterCollectionFunction}
   n29 = Get Apply to BaseColor() -> (Apply to BaseColor:bool)
   > SetScalarParameterValue(
    Collection=&MPC_ImposterCapture
    ParameterName=CustomLightingBasecolorOpacity
    ParameterValue=float * float(A="To Float (Boolean)"(InBool=@n29), B=Get Custom Lighting Opacity() -> (Custom Lighting Opacity:double))
   )
   > SetScalarParameterValue(
    self=[Get SampleFrame_DFAlphaMID(), Get SampleFrameMID()]
    ParameterName=TextureSize
    Value="To Float (Integer)"(InInt=Get Scene Capture Resolution() -> (Scene Capture Resolution:int=0))
   ) {is=title K2Node_CallFunction}
   n30 = Sequence!(execute=@n1.then_1) -> (then_0, then_1)
   > n31 = Branch(Condition=Get Enable Asset Material Overrides() -> (Enable Asset Material Overrides:bool))
   > n32 = Is Valid(InputObject=Get Override Material() -> (Override Material:Material Interface Object Reference))
   SetMaterial(execute=@n32.Is Not Valid, self=Get StaticMeshComponent!() -> (StaticMeshComponent), ElementIndex=0)
   > n33 = Create And Assign MID(
    MID=Get AssetMeshMID() -> (AssetMeshMID:Material Instance Dynamic Object Reference)
    Parent=GetMaterial!(self:Static Mesh Object Reference=Get Static Mesh(), MaterialIndex:int=0) -> (ReturnValue:Material Interface Object Reference)
    Optional Name=none
    Creation Flags=None
   )
   SetMaterial(execute=[@n33.Out, @n41.Out], self=Get StaticMeshComponent(), ElementIndex=0, Material=Get AssetMeshMID())
   n34 = Get Pseudo Volume Texture() -> (Pseudo Volume Texture:Texture Object Reference)
   > n35 = Is Valid(InputObject=@n34)
   n36 = Get Volume Texture() -> (Volume Texture:Volume Texture Object Reference)
   n37 = Is Valid(exec=[@n40.then, @n35.Is Not Valid], InputObject=@n36)
   > n38 = SetTextureParameterValue(self=Get AssetMeshMID(), ParameterName=Volume, Value=@n36)
   n39 = Break Int Point!(IntPoint:"Int Point Structure (by ref)"=Get PsuedoVolume XY Frames() -> (PsuedoVolume XY Frames:Int Point Structure)) -> (X:int, Y:int) {is=title K2Node_BreakStruct}
   SetVectorParameterValue(
    execute=[@n38.then, @n37.Is Not Valid]
    self=Get AssetMeshMID()
    ParameterName=XYFrames
    Value=MakeColor!(R:float="To Float (Integer)"(InInt=@n39), G:float="To Float (Integer)"(InInt=@n39.Y), B:float=0, A:float=0) -> (ReturnValue:Linear Color Structure)
   )
   > Apply Custom Material Params!(
    execute
    self
    MID:Material Instance Dynamic Object Reference=Get AssetMeshMID()
    Floats:"Map of Names to Float (double-precision)s"=Get Custom Floats() -> (Custom Floats:"Map of Names to Float (double-precision)s")
    Vectors:Map of Names to Linear Color Structures=Get Custom Vectors() -> (Custom Vectors:Map of Names to Linear Color Structures)
    Textures:Map of Names to Texture 2D Object References=Get Custom Textures() -> (Custom Textures:Map of Names to Texture 2D Object References)
   ) -> (then) {is=title K2Node_CallFunction}
   n40 = SetTextureParameterValue(execute=@n35.Is Valid, self=Get AssetMeshMID(), ParameterName=PseudoVolume, Value=@n34)
   n41 = Create And Assign MID(execute=@n32.Is Valid, MID=Get AssetMeshMID(), Parent=Get Override Material(), Optional Name=none, Creation Flags=None)
   SetMaterial(execute=@n31.else, self=Get StaticMeshComponent(), ElementIndex=0)
   Outputs!(Out:Exec=@n30.then_1)
  ]}
 Custom Compositing {nodes [
   Custom Compositing() {is=title K2Node_FunctionEntry}
   > n1 = Sequence()
   n2 = Get Maps to Render_Internal() -> (Maps to Render_Internal=NewEnumerator0)
   n3 = Get Maps to Render_Internal() -> (Maps to Render_Internal=NewEnumerator0)
   n4 = Get BakingDepth-to-ScalarMaps()
   > Branch(Condition=AND Boolean(
     A=Contains Item(TargetArray:Array of EImpostorBakeTypes Enums=@n3, ItemToFind:"EImpostorBakeTypes Enum (by ref)"=NewEnumerator6)
     B=Contains Item(TargetArray=@n2, ItemToFind=NewEnumerator8)
     C=NOT Boolean(A=@n4)
    ))
   > ClearRenderTarget2D(TextureRenderTarget=Get ScratchRT(), ClearColor=(R=0,G=0,B=0,A=1))
   n5 = Find(TargetMap=Get TargetMaps() -> (TargetMaps=NewEnumerator0), Key=NewEnumerator6)
   > Resample RT!(execute, self, Source RT:Texture Render Target 2D Object Reference=@n5, Dest RT:Texture Render Target 2D Object Reference=Get ScratchRT()) -> (then)
   > ClearRenderTarget2D(TextureRenderTarget=@n5, ClearColor=(R=0,G=0,B=0,A=1))
   > DrawMaterialToRenderTarget(
    TextureRenderTarget=Find(TargetMap=Get TargetMaps() -> (TargetMaps=NewEnumerator0), Key=NewEnumerator6)
    Material=Get CombineNormalsDepthMID()
   )
   n6 = Contains Item(TargetArray=Get Maps to Render_Internal() -> (Maps to Render_Internal=NewEnumerator0), ItemToFind=NewEnumerator10)
   n7 = Contains Item(TargetArray=Get Maps to Render_Internal() -> (Maps to Render_Internal=NewEnumerator0), ItemToFind=NewEnumerator0)
   Branch(execute=@n1.then_1, Condition=AND Boolean!(A=Get Apply to BaseColor(), B=AND Boolean(A=@n7, B=@n6)) -> (ReturnValue))
   > ClearRenderTarget2D(TextureRenderTarget=Get ScratchRT(), ClearColor=(R=0,G=0,B=0,A=1))
   n8 = Find(TargetMap=Get TargetMaps() -> (TargetMaps=NewEnumerator0), Key=NewEnumerator0)
   > Resample RT(Source RT=@n8, Dest RT=Get ScratchRT())
   > ClearRenderTarget2D(TextureRenderTarget=@n8, ClearColor=(R=0,G=0,B=0,A=1))
   > DrawMaterialToRenderTarget(
    TextureRenderTarget=Find(TargetMap=Get TargetMaps() -> (TargetMaps=NewEnumerator0), Key=NewEnumerator0)
    Material=Get BaseColorCustomLightingMID()
   )
  ]}
 DeriveAxes {nodes [
   n1 = Inputs!() -> (A:Vector=~)
   n2 = GetAxes!(A:Rotator=Rotation From X Vector(InVec=vector * float(A=@n1, B=-1))) -> (X:Vector, Y:Vector, Z:Vector)
   Outputs!(X:Vector=@n2.Y, Y:Vector=vector * float(A=@n2.Z, B=-1), Z:Vector=@n2)
  ]}
 Display Custom Lighting {nodes [
   Inputs!() -> (execute:Exec)
   > n1 = Sequence()
   > Branch(Condition=Contains Item(
     TargetArray:Array of EImpostorColorMaps Enums=Get Color Maps to Render() -> (Color Maps to Render=NewEnumerator0)
     ItemToFind:"EImpostorColorMaps Enum (by ref)"=NewEnumerator10
    ))
   n2 = Get Impostor Scene() -> (Impostor Scene:Impostor Scene Object Reference)
   > SetIntensity!(
    execute
    self:Sky Light Component Object Reference=Get SkyLight!(self:Impostor Scene Object Reference=@n2) -> (SkyLight:Sky Light Component Object Reference)
    NewIntensity:float=Get Custom SkyLight Intensity() -> (Custom SkyLight Intensity:double)
   ) -> (then)
   > SetVisibility!(
    execute
    self:Scene Component Object Reference=Get DirectionalLight!(self:Impostor Scene Object Reference=@n2) -> (DirectionalLight:Directional Light Component Object Reference)
    bNewVisibility:bool=false
    bPropagateToChildren:bool=false
   ) -> (then)
   > n3 = For Each Loop(Array:Array of Vectors=Get LightVectors() -> (LightVectors:Array of Vectors)) -> (Array Element:Vector)
   SetVisibility(
    execute=@n3.Completed
    self=Get Decal_CaptureCustomLighting!(self:Impostor Scene Object Reference=Get Impostor Scene()) -> (Decal_CaptureCustomLighting:Decal Component Object Reference)
    bNewVisibility=true
    bPropagateToChildren=false
   )
   n4 = Spawn!(
    execute=@n3.LoopBody
    self:Viewport Object Reference=Get ImpostorViewport!(self:Impostor Viewport Widget Object Reference=Get ImpostorViewport_Widget() -> (ImpostorViewport_Widget:Impostor Viewport Widget Object Reference)) -> (ImpostorViewport:Impostor Viewport Object Reference)
    ActorClass:Actor Class Reference=/Script/Engine.DirectionalLight
   ) -> (then, ReturnValue:Actor Object Reference)
   > n5 = Cast To DirectionalLight!(execute, Object=@n4) -> (then, CastFailed, AsDirectional Light:Directional Light Object Reference) {is=title K2Node_DynamicCast}
   n6 = Get Lighting Grid XY() -> (Lighting Grid XY:int=0)
   n7 = int * int(A:int=@n6, B:int=@n6) -> (ReturnValue:int=0) {is=title K2Node_CommutativeAssociativeBinaryOperator}
   n8 = Get LightComponent!(self:Light Object Reference=@n5) -> (LightComponent:Light Component Object Reference)
   > SetIntensity(
    self:Light Component Object Reference=@n8
    NewIntensity=float / float(A=Get DirLightBrightness() -> (DirLightBrightness:double), B="To Float (Integer)"(InInt=@n7)) {is=title K2Node_CallFunction}
   )
   > SetMobility!(execute, self:Scene Component Object Reference=@n8, NewMobility:EComponentMobility Enum=Movable) -> (then)
   n9 = float * float(A=Get Object Radius(), B=2.5)
   n10 = vector * float(A=@n3, B=-1)
   > Set World Transform!(
    execute
    self:Scene Component Object Reference=@n8
    NewTransform:"Transform (by ref)"=MakeTransform(Location=vector * float(A=@n3, B=@n9), Rotation=MakeRotFromX!(X:"Vector (by ref)"=@n10) -> (ReturnValue:Rotator))
    bSweep:bool=false
    bTeleport:bool=false
   ) -> (then, SweepHitResult:Hit Result Structure) {is=title K2Node_CallFunction}
   > Add(
    TargetArray:Array of Directional Light Object References=Get lights() -> (lights:Array of Directional Light Object References)
    NewItem:"Directional Light Object Reference (by ref)"=@n5
   )
   Outputs!(out:Exec=@n1.then_1)
  ]}
 Downsample Billboard {nodes [
   Downsample Billboard() {is=title K2Node_FunctionEntry}
   > Branch(Condition="Equal (Enum)"(A=Get Impostor Type() -> (Impostor Type=NewEnumerator0), B=NewEnumerator2))
   > ClearRenderTarget2D(TextureRenderTarget=Get CombinedAlphasRT(), ClearColor=(R=0,G=0,B=0,A=1))
   n1 = Find(TargetMap=Get TargetMaps() -> (TargetMaps=NewEnumerator0), Key=NewEnumerator0)
   > Branch(Condition=@n1.ReturnValue)
   > Resample RT(Source RT=@n1, Dest RT=Get CombinedAlphasRT())
  ]}
 Draw Frame to Atlas {nodes [
   n1 = Draw Frame to Atlas!() -> (then, TextureRenderTarget:Texture Render Target 2D Object Reference) {is=title K2Node_FunctionEntry}
   n2 = "Equal (Enum)"(A:EImpostorBakeTypes Enum=Get CurrentMap() -> (CurrentMap=NewEnumerator0), B:EImpostorBakeTypes Enum=NewEnumerator6)
   > SetVectorParameterValue(
    self=Get SampleFrameMID()
    ParameterName=BackgroundColor
    Value=Make Linear Color!(R:float=0.5, G:float=0.5, B:float=Select(Option 0:float=0.5, Option 1:float=1, Index=@n2) -> (ReturnValue:float), A:float=0.5) -> (LinearColor:Linear Color Structure) {is=title K2Node_MakeStruct}
   )
   > n3 = BeginDrawCanvasToRenderTarget!(execute, WorldContextObject:Object Reference=Get World(), TextureRenderTarget:Texture Render Target 2D Object Reference=@n1) -> (then, Canvas:Canvas Object Reference, Size:Vector 2D Structure, Context:Draw To Render Target Context Structure)
   n4 = BreakVector2D(InVec=@n3.Size)
   n5 = "To Float (Integer)"(InInt=Get FramesXYInternal() -> (FramesXYInternal=0))
   n6 = MakeVector2D(X=float / float(A=@n4, B=@n5), Y=float / float(A=@n4.Y, B=@n5))
   n7 = Convert a 1D Index to a 2D Index!(Index1D:int=Get curframe() -> (curframe=0), XSize:int=Get FramesXYInternal() -> (FramesXYInternal=0)) -> (ReturnValue:Int Point Structure)
   n8 = Get CurrentMap() -> (CurrentMap=NewEnumerator0)
   > Draw Material!(
    execute
    self:Canvas Object Reference=@n3
    RenderMaterial:Material Interface Object Reference=Select(
     Option 0:Material Interface Object Reference=Get SampleFrameMID()
     Option 1:Material Interface Object Reference=Get SampleFrame_DFAlphaMID()
     Index="Equal (Enum)"(A=@n8, B=NewEnumerator0)
    ) -> (ReturnValue:Material Interface Object Reference)
    ScreenPosition:Vector 2D Structure=vector2d * vector2d(
     A:Vector 2D Structure=@n6
     B:Vector 2D Structure="To Vector2D (IntPoint)"!(InIntPoint:Int Point Structure=@n7) -> (ReturnValue:Vector 2D Structure)
    ) -> (ReturnValue:Vector 2D Structure) {is=title K2Node_CommutativeAssociativeBinaryOperator}
    ScreenSize:Vector 2D Structure=@n6
    CoordinatePosition:Vector 2D Structure
    CoordinateSize:Vector 2D Structure=(X=1,Y=1)
    Rotation:float=0
    PivotPoint:Vector 2D Structure=(X=0.5,Y=0.5)
   ) -> (then)
   > EndDrawCanvasToRenderTarget!(execute, WorldContextObject:Object Reference=Get World(), Context:"Draw To Render Target Context Structure (by ref)"=@n3.Context) -> (then)
   > Branch(Condition="Equal (Enum)"(A=Get CurrentMap() -> (CurrentMap=NewEnumerator0), B=NewEnumerator0))
   > DrawMaterialToRenderTarget(WorldContextObject=Get World(), TextureRenderTarget=Get CombinedAlphasRT(), Material=Get AddAlphasMID())
  ]}
 EventGraph {nodes [
   n1 = Split(SourceString=Get Object Path String(Object=Get Static Mesh()), InStr=., SearchDir=FromEnd)
   n2 = PrintString(InString=@n1.RightS)
   n3 = GetAssetTools(execute=@n11.then)
   n4 = Get Processed Texture Path()
   n5 = CreateAsset(
    execute=[@n3.then, @n2.then]
    self=@n3
    AssetName=Append(A=DA_ImpostorPreset_, B=@n1.RightS)
    PackagePath=Append(A=Split(SourceString=@n4, InStr=/, SearchDir=FromEnd), B=/)
    AssetClass=/ImpostorBaker/BP/DataAsset/ImpostorPresets.ImpostorPresets_C
   )
   > n6 = Cast To ImpostorPresets!(execute, Object=@n5) -> (then, CastFailed, AsImpostor Presets:Impostor Presets Object Reference) {is=title K2Node_DynamicCast}
   > Set SettingsPreset(SettingsPreset:Impostor Presets Object Reference=@n6) -> (Output_Get:Impostor Presets Object Reference)
   > Apply Settings Preset()
   n7 = "On Property Changed (Details_BakeSettings)"!() -> (OutputDelegate, then, PropertyName:Name=None) {is=title K2Node_ComponentBoundEvent}
   > n8 = Set ChangedProperty(ChangedProperty:Name=@n7.PropertyName) -> (Output_Get:Name=None)
   > n9 = Switch on Name!(execute:Exec, Selection:Name=@n8) -> (Default:Exec, Static Mesh:Exec, CaptureUsingGBuffer:Exec) {is=title K2Node_SwitchName, pinNames [Static Mesh, CaptureUsingGBuffer]}
   Set Capture Mesh and Size Info!(execute=[@n9.CaptureUsingGBuffer, @n9.Static Mesh], self, ReAdjustCamera:bool=true) -> (then) {is=title K2Node_CallFunction}
   n10 = Get Color Maps to Render() -> (Color Maps to Render=NewEnumerator0)
   Set Visibility(
    execute=@n9.Default
    Visibility=Collapsed or Visible!(Index:bool=Contains Item(TargetArray=@n10, ItemToFind=NewEnumerator10)) -> (ReturnValue:ESlateVisibility Enum=Visible) {is=title K2Node_MacroInstance}
    self=Get Details_CustomLighting() -> (Details_CustomLighting:Details View Object Reference)
   )
   > Setup RT and Save List()
   > Update Material Settings Details View!(execute:Exec) -> (Completed:Exec) {is=title K2Node_MacroInstance}
   "On Clicked (Button_Render)"() {is=title K2Node_ComponentBoundEvent}
   > RenderFrames()
   "On Clicked (Button_NewDataAsset)"() {is=title K2Node_ComponentBoundEvent}
   > n11 = Process Path Name()
   "On Clicked (Button_ReplaceLOD)"() {is=title K2Node_ComponentBoundEvent}
   > n12 = Save Dynamic Mesh To Staticmesh(Overwrite Last LOD=true)
   SetVisibility!(
    execute=[@n26.then, @n12.then, @n13.then]
    self:Widget Object Reference=Get VerticalBox_LODDialogue()
    InVisibility:ESlateVisibility Enum=Collapsed
   ) -> (then)
   > Save Complete()
   "On Clicked (Button_AddNewLOD)"() {is=title K2Node_ComponentBoundEvent}
   > n13 = Save Dynamic Mesh To Staticmesh()
   "On Property Changed (Details_AssetSaving)"!() -> (OutputDelegate, then, PropertyName:Name=None) {is=title K2Node_ComponentBoundEvent}
   > Branch(Condition=StartsWith(SourceString=Get New Asset Path() -> (New Asset Path:String), InPrefix=../))
   > Set Use Relative Path(Use Relative Path:bool=true) -> (Output_Get:bool)
   n14 = "On Property Changed (Details_MaterialSettings)"!() -> (OutputDelegate, then, PropertyName:Name=None) {is=title K2Node_ComponentBoundEvent}
   > n15 = Switch on Name!(execute, Selection=@n14.PropertyName) -> (Default, Iterative Parallax Mode:Exec) {pinNames [Iterative Parallax Mode]}
   Create And Setup MIDs!(In:Exec=[@n16.then, @n15.Default]) -> (Out:Exec) {is=title K2Node_MacroInstance}
   n16 = Initialize(execute=@n15.Iterative Parallax Mode)
   n17 = "On Text Committed (Text_SwapDistance)"!() -> (OutputDelegate, then, Text:"Text (by ref)", CommitMethod:ETextCommit Enum=Default) {is=title K2Node_ComponentBoundEvent}
   > Set SwapDistance(SwapDistance:double=String To Float(InString="To String (Text)"(InText=@n17.Text))) -> (Output_Get:double)
   "On Property Changed (Details_BakeAssetMaterialOverrides)"!() -> (OutputDelegate, then, PropertyName:Name=None) {is=title K2Node_ComponentBoundEvent}
   > Initialize()
   n18 = "On Selection Changed (Combobox_PreviewMode)"!() -> (OutputDelegate, then, SelectedItem:String, SelectionType:ESelectInfo Enum) {is=title K2Node_ComponentBoundEvent}
   > SetVisibility(
    self=[
     Get HorizontalBox_SwapDist() -> (HorizontalBox_SwapDist:Horizontal Box Object Reference)
     Get HorizontalBox_CurDist() -> (HorizontalBox_CurDist:Horizontal Box Object Reference)
    ]
    InVisibility=Select(
     Option 0:ESlateVisibility Enum=Collapsed
     Option 1:ESlateVisibility Enum=SelfHitTestInvisible
     Index="Equal Exactly (String)"(A=@n18.SelectedItem, B=Swap by Distance)
    ) -> (ReturnValue:ESlateVisibility Enum)
   )
   > Update Preview Mode and Location!(execute:Exec, Selection:String=@n18.SelectedItem) {is=title K2Node_MacroInstance}
   n19 = "On Selection Changed (Combobox_ViewMode)"!() -> (OutputDelegate, then, SelectedItem:String, SelectionType:ESelectInfo Enum) {is=title K2Node_ComponentBoundEvent}
   n20 = String To Name(InString=@n19.SelectedItem)
   > Change View Modes!(execute, self:Impostor Scene Object Reference=Get Impostor Scene(), ViewMode:Name=@n20) -> (then)
   n21 = Get ImpostorViewport(self=Get ImpostorViewport_Widget())
   n22 = "Equal (Name)"(A=@n20, B=Lit)
   > SetShowFlag!(execute, self:Viewport Object Reference=@n21, InShowFlagName:String=ColorGrading, InValue:bool=@n22) -> (then)
   > SetShowFlag(self=@n21, InShowFlagName=Tonemapper, InValue=@n22)
   > n23 = Branch(Condition="Equal (Name)"(A=@n20, B=CustomLighting))
   > Setup Octahedron Layout()
   > Display Custom Lighting()
   Restore Lighting State(execute=@n23.else)
   n24 = "On Property Changed (Details_Advanced)"!() -> (OutputDelegate, then, PropertyName:Name=None) {is=title K2Node_ComponentBoundEvent}
   > Set ChangedProperty(ChangedProperty=@n24.PropertyName) -> (Output_Get=None)
   > n25 = Switch on Name!(execute, Selection=Get ChangedProperty() -> (ChangedProperty:Name=None)) -> (
    Default
    Show Debug Variables:Exec
    Show Render Targets:Exec
    Manual sRGB_RetainerBox:Exec
    Enable Asset Material Overrides:Exec
    Show Floor in Preview Scene:Exec
    Impostor Casts Shadows:Exec
    Shader Debug Mode:Exec
   ) {pinNames [
     Show Debug Variables
     Show Render Targets
     Manual sRGB_RetainerBox
     Enable Asset Material Overrides
     Show Floor in Preview Scene
     Impostor Casts Shadows
     Shader Debug Mode
    ]}
   > Setup RT and Save List()
   Set MIC Switches and ShadingModel(
    execute=@n25.Shader Debug Mode
    In MIC=Get MaterialPreviewProxy() -> (MaterialPreviewProxy:Material Instance Constant Object Reference)
    Creating Final Assets=false
   )
   Set Visibility(
    execute=@n25.Show Debug Variables
    Visibility=Collapsed or Visible(Index=Get Show Debug Variables() -> (Show Debug Variables:bool)) -> (ReturnValue=Visible)
    self=Get Details_Debug() -> (Details_Debug:Details View Object Reference)
   )
   SetCastShadow!(
    execute=@n25.Impostor Casts Shadows
    self:Primitive Component Object Reference=Get DynamicMeshComponent()
    NewCastShadow:bool=Get Impostor Casts Shadows() -> (Impostor Casts Shadows:bool)
   ) -> (then)
   Set Visibility(
    execute=@n25.Enable Asset Material Overrides
    Visibility=Collapsed or Visible(Index=Get Enable Asset Material Overrides()) -> (ReturnValue=Visible)
    self=Get Details_BakeAssetMaterialOverrides() -> (Details_BakeAssetMaterialOverrides:Details View Object Reference)
   )
   SetRetainRendering!(
    execute=@n25.Manual sRGB_RetainerBox
    self:Retainer Box Object Reference=Get RetainerBox_35!(self:Impostor Viewport Widget Object Reference=Get ImpostorViewport_Widget()) -> (RetainerBox_35:Retainer Box Object Reference)
    bInRetainRendering:bool=Get Manual sRGB_RetainerBox() -> (Manual sRGB_RetainerBox:bool)
   ) -> (then)
   SetVisibility!(
    execute=@n25.Show Floor in Preview Scene
    self:Scene Component Object Reference=Get Preview Floor!(self:Impostor Scene Object Reference=Get Impostor Scene()) -> (Preview Floor:Static Mesh Component Object Reference)
    bNewVisibility:bool=Get Show Floor in Preview Scene() -> (Show Floor in Preview Scene:bool)
    bPropagateToChildren:bool=false
   ) -> (then)
   Set Visibility(
    execute=@n25.Show Render Targets
    Visibility=Collapsed or Visible(Index=Get Show Render Targets() -> (Show Render Targets:bool)) -> (ReturnValue=Visible)
    self=Get Details_RenderTargets() -> (Details_RenderTargets:Details View Object Reference)
   )
   n26 = "On Clicked (Button_Skip)"() {is=title K2Node_ComponentBoundEvent}
   "On Clicked (Button_CreateAssets)"() {is=title K2Node_ComponentBoundEvent}
   > Create Static Assets() {is=title K2Node_CallFunction}
   "On Clicked (Button_DataAssetCategory)"() {is=title K2Node_ComponentBoundEvent}
   n27 = Get VerticalBox_SettingsPreset() -> (VerticalBox_SettingsPreset:Vertical Box Object Reference)
   n28 = "Equal (Enum)"(
    A:ESlateVisibility Enum=GetVisibility!(self:Widget Object Reference=@n27) -> (ReturnValue:ESlateVisibility Enum=Visible)
    B:ESlateVisibility Enum=Collapsed
   )
   > SetVisibility!(
    execute
    self:Widget Object Reference=@n27
    InVisibility:ESlateVisibility Enum=Select(Option 0=Collapsed, Option 1=SelfHitTestInvisible, Index=@n28)
   ) -> (then)
   > SetBrushFromMaterial(
    self=Get Image_SettingsPresetArrow() -> (Image_SettingsPresetArrow:Image Object Reference)
    Material=Select(
     Option 0:Material Interface Object Reference=/ImpostorBaker/Materials/Slate/CategoryArrow_Rotated
     Option 1:Material Interface Object Reference=/ImpostorBaker/Materials/Slate/CategoryArrow
     Index=@n28
    ) -> (ReturnValue:Material Interface Object Reference)
   )
   n29 = "On Clicked (Button_LoadSettings)"() {is=title K2Node_ComponentBoundEvent}
   K2Node_Composite_2 = Load Settings From Preset DataAsset!(exec:Exec=[@n29.then, @n32.then]) {is=title K2Node_Composite}
   "On Clicked (Button_SaveToDataAsset)"() {is=title K2Node_ComponentBoundEvent}
   > K2Node_Composite_4 = Save Settings to Preset DataAsset!(exec:Exec) {is=title K2Node_Composite}
   RenderFrames() {is=title K2Node_CustomEvent}
   > Initialize()
   > SetVisibility!(execute, self:Scene Component Object Reference=Get StaticMeshComponent(), bNewVisibility:bool=true, bPropagateToChildren:bool=false) -> (then)
   > Render Maps()
   n30 = Render Next Map() {is=title K2Node_CustomEvent}
   > Capture Imposter Grid() {is=title K2Node_CallFunction}
   > Increment Int(Value=Get CurrentMapIndex() -> (CurrentMapIndex=0))
   > Render Maps()
   n31 = Enable Retainerbox() {is=title K2Node_CustomEvent}
   > SetRetainRendering(self=Get RetainerBox_35(self=Get ImpostorViewport_Widget()), bInRetainRendering=Get Manual sRGB_RetainerBox())
   n32 = Apply Settings Preset() {is=title K2Node_CustomEvent}
   Render Maps() {is=title K2Node_CustomEvent}
   > Set Rendering(Rendering:bool=true) -> (Output_Get:bool)
   n33 = Get Maps to Render_Internal() -> (Maps to Render_Internal=NewEnumerator0)
   > n34 = Branch(Condition="integer >= integer"(A=Last Index(TargetArray:Array of EImpostorBakeTypes Enums=@n33) -> (ReturnValue=0), B=Get CurrentMapIndex() -> (CurrentMapIndex=0)))
   > Update Progress!(execute:Exec) -> (then:Exec) {is=title K2Node_MacroInstance}
   > K2Node_Composite_3 = Set Timer by Event!(execute, Delegate:Delegate=@n30) {is=title K2Node_Composite}
   n35 = Sequence(execute=@n34.else)
   > ChannelPackMasks() {is=title K2Node_CallFunction}
   > Custom Compositing() {is=title K2Node_CallFunction}
   > Downsample Billboard() {is=title K2Node_CallFunction}
   > GenerateMeshData()
   Set Rendered(execute=@n35.then_1, Rendered:bool=true) -> (Output_Get:bool)
   > Set Rendering(Rendering=false)
   > Set Progress(Progress:double=1) -> (Output_Get:double)
   > All Passes Rendered()
   > Delay!(execute, WorldContextObject:Object Reference, Duration=0.25) -> (then)
   > SetVisibility!(
    execute
    self:Widget Object Reference=Get VerticalBox_GeneratingImpostor() -> (VerticalBox_GeneratingImpostor:Vertical Box Object Reference)
    InVisibility:ESlateVisibility Enum=Collapsed
   ) -> (then)
   All Passes Rendered() {is=title K2Node_CustomEvent}
   > PrintString(InString="Render Complete!")
   > Delay(Duration=0.2)
   > Call Impostor Render Complete!(execute:Exec, self:EUW Impostor Editor Object Reference) -> (then:Exec) {is=title K2Node_CallDelegate}
   n36 = Remove an Alpha() {is=title K2Node_CustomEvent}
   > Set CompressionNoAlpha(
    CompressionNoAlpha=true
    self="Get (a copy)"(
     Array:Array of Texture 2D Object References=Get TexturesToRemoveAlpha()
     Dimension 1=Get CurrentRemovalIndex() -> (CurrentRemovalIndex:int=0)
    ) -> (Output:Texture 2D Object Reference)
   )
   > Increment Int(Value=Get CurrentRemovalIndex() -> (CurrentRemovalIndex=0))
   > PrintString(InString=Removed one alpha)
   > Remove Alphas Loop()
   Save Complete() {is=title K2Node_CustomEvent}
   > PrintString(InString=Saving Complete event hit.)
   > Call Impostor Save Complete!(execute:Exec, self:EUW Impostor Editor Object Reference) -> (then:Exec) {is=title K2Node_CallDelegate}
   Remove Alphas Loop() {is=title K2Node_CustomEvent}
   n37 = Get TexturesToRemoveAlpha()
   > n38 = Branch(Condition="integer <= integer"(
     A=Get CurrentRemovalIndex() -> (CurrentRemovalIndex=0)
     B=Last Index(TargetArray:Array of Texture 2D Object References=@n37) -> (ReturnValue=0)
    ))
   > Set Timer by Event(Delegate=@n36, Time=1, bLooping=false) {is=title K2Node_CallFunction}
   > PrintString(InString="Removing an Alpha, set Timer")
   n39 = Get CurrentRemovalIndex() -> (CurrentRemovalIndex=0)
   n40 = Last Index(TargetArray=Get TexturesToRemoveAlpha()) -> (ReturnValue=0)
   PrintString(
    execute=@n38.else
    InString=Append!(A="Removed last alpha index: ", B="To String (Integer)"(InInt=@n39), C:String="out of ", D:String="To String (Integer)"(InInt=@n40)) -> (ReturnValue)
   )
   > Call Impostor Render Save Complete!(execute:Exec, self:EUW Impostor Editor Object Reference) -> (then:Exec) {is=title K2Node_CallDelegate}
   Event PreConstruct()
   > SetObject(self=Get Details_BakeSettings() -> (Details_BakeSettings:Details View Object Reference), NewObject=Self-Reference())
   > SetObject(self=Get Details_CustomLighting(), NewObject=Self-Reference())
   > SetObject(self=Get Details_AssetSaving() -> (Details_AssetSaving:Details View Object Reference), NewObject=Self-Reference())
   > SetObject(self=Get Details_Advanced() -> (Details_Advanced:Details View Object Reference), NewObject=Self-Reference())
   > SetObject(
    self=[
     Get Details_MaterialSettings() -> (Details_MaterialSettings:Details View Object Reference)
     Get SinglePropertyView_80() -> (SinglePropertyView_80:Single Property View Object Reference)
    ]
    NewObject=Self-Reference()
   )
   > SetObject(self=Get Details_Debug(), NewObject=Self-Reference())
   > SetObject(self=Get Details_RenderTargets(), NewObject=Self-Reference())
   > SetObject(self=Get Details_BakeAssetMaterialOverrides(), NewObject=Self-Reference())
   Event Construct()
   > Get Settings Preset from INI!(execute:Exec) -> (OutputPin:Exec) {is=title K2Node_MacroInstance}
   n41 = Get ImpostorViewport_Widget()
   > Set World(World:Object Reference=GetViewportWorld!(self:Viewport Object Reference=Get ImpostorViewport(self=@n41)) -> (ReturnValue:World Object Reference)) -> (Output_Get:Object Reference)
   > Set Impostor Scene(Impostor Scene:Impostor Scene Object Reference=Get Impostor Scene!(self:Impostor Viewport Widget Object Reference=Get ImpostorViewport_Widget()) -> (Impostor Scene)) -> (Output_Get:Impostor Scene Object Reference)
   n42 = Get Impostor Scene!() -> (Impostor Scene)
   > Set SceneCaptureComponent2D(SceneCaptureComponent2D:Scene Capture Component 2D Object Reference=Get SceneCaptureComponent2D!(self:Impostor Scene Object Reference=@n42) -> (SceneCaptureComponent2D)) -> (Output_Get:Scene Capture Component 2D Object Reference)
   > Set StaticMeshComponent(StaticMeshComponent:Static Mesh Component Object Reference=Get Capture_StaticMesh!(self:Impostor Scene Object Reference=@n42) -> (Capture_StaticMesh:Static Mesh Component Object Reference)) -> (Output_Get:Static Mesh Component Object Reference)
   > Set DynamicMeshComponent(DynamicMeshComponent:Dynamic Mesh Component Object Reference=Get DynamicMeshComponent!(self:Impostor Scene Object Reference=@n42) -> (DynamicMeshComponent)) -> (Output_Get:Dynamic Mesh Component Object Reference)
   > Set Impostor_MeshCutout_Component(Impostor_MeshCutout_Component:Impostor Mesh Cutout Object Reference=Get Impostor_MeshCutout!(self:Impostor Scene Object Reference=@n42) -> (Impostor_MeshCutout:Impostor Mesh Cutout Object Reference)) -> (Output_Get:Impostor Mesh Cutout Object Reference)
   > Set Capture Mesh and Size Info(ReAdjustCamera=true)
   > Reset All Capture Modes!(execute:Exec) -> (then_1:Exec) {is=title K2Node_MacroInstance}
   > Populate View Modes!(In:Exec) -> (Out:Exec) {is=title K2Node_MacroInstance}
   n43 = Get ImpostorViewport(self=Get ImpostorViewport_Widget())
   > SetLightIntensity!(execute, self:Viewport Object Reference=@n43, LightIntensity:float=0) -> (then)
   > SetSkyIntensity!(execute, self:Viewport Object Reference=@n43, LightIntensity:float=0) -> (then)
   > SetEnableAdvancedFeatures!(execute, self:Viewport Object Reference=@n43, InEnableAdvancedFeatures:bool=true) -> (then)
   > SetShowFlag(self=@n43, InShowFlagName=Vignette, InValue=false)
   > ExecuteConsoleCommand(WorldContextObject=@n43, Command=r.RayTracing.Shadows 0)
   > Set Timer by Event(Delegate=@n31, Time=0.1, bLooping=false)
   Event Tick()
   > Branch(Condition=Get SwapByDistance() -> (SwapByDistance:bool))
   > Branch(Condition=NOT Boolean(A=Get Rendering() -> (Rendering:bool)))
   > K2Node_Composite_1 = Swap Distance Text Fields!(execute) -> (then:Exec) {is=title K2Node_Composite}
   n44 = "float < float"(A=Get CurrentCameraDistance() -> (CurrentCameraDistance:double), B=Get SwapDistance() -> (SwapDistance:double))
   > SetVisibility!(execute, self:Scene Component Object Reference=Get StaticMeshComponent(), bNewVisibility:bool=@n44, bPropagateToChildren:bool=false) -> (then)
   n45 = NOT Boolean(A=@n44)
   > SetVisibility(self=Get DynamicMeshComponent!() -> (DynamicMeshComponent), bNewVisibility=@n45, bPropagateToChildren=false)
   n46 = NOT Boolean(A=Get Rendered() -> (Rendered:bool))
   > SetVisibility!(
    execute
    self:Widget Object Reference=Get Text_ImpostorNotRenderd() -> (Text_ImpostorNotRenderd:Text Object Reference)
    InVisibility:ESlateVisibility Enum=Select(Option 0:ESlateVisibility Enum=Hidden, Option 1:ESlateVisibility Enum=SelfHitTestInvisible, Index=AND Boolean(A=@n45, B=@n46)) -> (ReturnValue:ESlateVisibility Enum)
   ) -> (then)
   Event Destruct()
   n47 = Get MaterialPreviewProxy()
   > Is Valid(InputObject=@n47)
   > DeleteLoadedAsset!(execute, AssetToDelete:Object Reference=@n47) -> (then, ReturnValue:bool)
   Set CurrentRemovalIndex(CurrentRemovalIndex:int=0) -> (Output_Get:int=0)
   > Remove Alphas Loop() {is=title K2Node_CallFunction}
   Remove Alphas Loop()
   "On Property Changed (Details_CustomLighting)"!() -> (OutputDelegate, then, PropertyName:Name=None) {is=title K2Node_ComponentBoundEvent}
   CustomEvent() {is=title K2Node_CustomEvent}
   Process Path Name()
   Update Preview Mode and Location(Selection=GetSelectedOption!(self:"Combo Box (String) Object Reference"=Get Combobox_PreviewMode() -> (Combobox_PreviewMode:Editor Utility Combo Box String Object Reference)) -> (ReturnValue:String))
   Get SettingsPreset() -> (SettingsPreset:Impostor Presets Object Reference)
  ]}
 Swap Distance Text Fields {
  path=EventGraph.K2Node_Composite_1.Swap Distance Text Fields
  nodes [
   Inputs()
   n1 = Get ImpostorViewport(self=Get ImpostorViewport_Widget())
   > n2 = Set CurrentCameraDistance(CurrentCameraDistance:double=Vector Length(A=GetViewLocation!(self:Viewport Object Reference=@n1) -> (ReturnValue:Vector))) -> (Output_Get:double)
   > "SetText (Text)"(
    self=Get CurrentDistanceDisplay() -> (CurrentDistanceDisplay:Text Object Reference)
    InText="To Text (Float)"(Value=@n2, MinimumIntegralDigits=1, MaximumIntegralDigits=324, MinimumFractionalDigits=0, MaximumFractionalDigits=3)
   )
   n3 = Get SwapDistance()
   n4 = Get CurrentCameraDistance()
   > Set Text(
    Text=Select(
     Option 0:Text="LOCTEXT(\"Displaying: Impostor\")"
     Option 1:Text="LOCTEXT(\"Displaying: Source Static Mesh\")"
     Index="float < float"(A=@n4, B=@n3)
    ) -> (ReturnValue:Text)
    self=Get Text_DisplayingWhat() -> (Text_DisplayingWhat:Text Object Reference)
   )
   > Outputs!(then:Exec)
  ]
 }
 Load Settings From Preset DataAsset {
  path=EventGraph.K2Node_Composite_2.Load Settings From Preset DataAsset
  nodes [
   Inputs!() -> (exec:Exec)
   > n1 = Sequence()
   > Is Valid(InputObject=Get SettingsPreset())
   > n2 = Sequence!(execute) -> (then_0, then_1, then_2:Exec, then_3:Exec, then_4:Exec, then_5:Exec, then_6:Exec)
   n3 = Get SettingsPreset()
   > Set Frames XY(Frames XY:int=Get Frames XY!(self:Impostor Presets Object Reference=@n3) -> (Frames XY:int=0)) -> (Output_Get:int=0)
   > Set Color Maps to Render(Color Maps to Render:Array of EImpostorColorMaps Enums=Get Color Maps to Render!(self:Impostor Presets Object Reference=@n3) -> (Color Maps to Render=NewEnumerator0)) -> (Output_Get:Array of EImpostorColorMaps Enums=NewEnumerator0)
   > Set Channel Packed Masks(Channel Packed Masks:Array of Impostor Channel Pack Settings Structures=Get Channel Packed Masks!(self:Impostor Presets Object Reference=@n3) -> (Channel Packed Masks)) -> (Output_Get:Array of Impostor Channel Pack Settings Structures)
   > Set Resolution(Resolution:int=Get Resolution!(self:Impostor Presets Object Reference=@n3) -> (Resolution:int=0)) -> (Output_Get:int=0)
   > Set Impostor Type(Impostor Type:EImposterLayoutType Enum=Get Impostor Type!(self:Impostor Presets Object Reference=@n3) -> (Impostor Type=NewEnumerator0)) -> (Output_Get:EImposterLayoutType Enum=NewEnumerator0)
   n4 = Get SettingsPreset()
   Set Custom Lighting Power(
    execute=@n2.then_4
    Custom Lighting Power:double=Get Custom Lighting Power!(self:Impostor Presets Object Reference=@n4) -> (Custom Lighting Power)
   ) -> (Output_Get:double)
   > Set Custom Lighting Opacity(Custom Lighting Opacity:double=Get Custom Lighting Opacity!(self:Impostor Presets Object Reference=@n4) -> (Custom Lighting Opacity)) -> (Output_Get:double)
   > Set Lighting Grid XY(Lighting Grid XY:int=Get Lighting Grid XY!(self:Impostor Presets Object Reference=@n4) -> (Lighting Grid XY=0)) -> (Output_Get:int=0)
   n5 = Get SettingsPreset()
   > Set Upward Bias(Upward Bias:double=Get Upward Bias!(self:Impostor Presets Object Reference=@n5) -> (Upward Bias:double)) -> (Output_Get:double)
   > Set DirLightBrightness(DirLightBrightness:double=Get DirLightBrightness!(self:Impostor Presets Object Reference=@n5) -> (DirLightBrightness)) -> (Output_Get:double)
   > Set Custom SkyLight Intensity(Custom SkyLight Intensity:double=Get Custom SkyLight Intensity!(self:Impostor Presets Object Reference=@n5) -> (Custom SkyLight Intensity)) -> (Output_Get:double)
   > Set Apply to BaseColor(Apply to BaseColor:bool=Get Apply to BaseColor!(self:Impostor Presets Object Reference=Get SettingsPreset()) -> (Apply to BaseColor)) -> (Output_Get:bool)
   n6 = Get SettingsPreset()
   Set Reference Plane Sides(
    execute=@n2.then_2
    Reference Plane Sides:double=Get Reference Plane Sides!(self:Impostor Presets Object Reference=@n6) -> (Reference Plane Sides)
   ) -> (Output_Get:double)
   > Set Reference Plane Top(Reference Plane Top:double=Get Reference Plane Top!(self:Impostor Presets Object Reference=@n6) -> (Reference Plane Top)) -> (Output_Get:double)
   > Set Center XY on Mesh Pivot(Center XY on Mesh Pivot:bool=Get Center XY on Mesh Pivot!(self:Impostor Presets Object Reference=@n6) -> (Center XY on Mesh Pivot:bool)) -> (Output_Get:bool)
   > Set Use Sprite Vertex Shader(Use Sprite Vertex Shader:bool=Get Use Sprite Vertex Shader!(self:Impostor Presets Object Reference=@n6) -> (Use Sprite Vertex Shader:bool)) -> (Output_Get:bool)
   n7 = Get SettingsPreset()
   Set Apply to BaseColor(execute=@n2.then_1, Apply to BaseColor=Get Apply to BaseColor(self=@n7))
   > Set Orthographic(Orthographic:bool=Get Orthographic!(self:Impostor Presets Object Reference=@n7) -> (Orthographic)) -> (Output_Get:bool)
   > Set Camera Distance(Camera Distance:double=Get Camera Distance!(self:Impostor Presets Object Reference=@n7) -> (Camera Distance)) -> (Output_Get:double)
   > Set CaptureUsingGBuffer(CaptureUsingGBuffer:bool=Get CaptureUsingGBuffer!(self:Impostor Presets Object Reference=@n7) -> (CaptureUsingGBuffer)) -> (Output_Get:bool)
   > Set Scene Capture Resolution(Scene Capture Resolution:int=Get Scene Capture Resolution!(self:Impostor Presets Object Reference=@n7) -> (Scene Capture Resolution=0)) -> (Output_Get:int=0)
   > Set Enable Asset Material Overrides(Enable Asset Material Overrides:bool=Get Enable Asset Material Overrides!(self:Impostor Presets Object Reference=@n7) -> (Enable Asset Material Overrides)) -> (Output_Get:bool)
   > Set Mesh Scale(Mesh Scale:Vector=Get Mesh Scale!(self:Impostor Presets Object Reference=@n7) -> (Mesh Scale:Vector)) -> (Output_Get:Vector)
   n8 = Get SettingsPreset()
   Set Full Sphere Material(
    execute=@n2.then_3
    Full Sphere Material:Material Interface Object Reference=Get Full Sphere Material!(self:Impostor Presets Object Reference=@n8) -> (Full Sphere Material)
   ) -> (Output_Get:Material Interface Object Reference)
   > Set Hemisphere Material(Hemisphere Material:Material Interface Object Reference=Get Hemisphere Material!(self:Impostor Presets Object Reference=@n8) -> (Hemisphere Material)) -> (Output_Get:Material Interface Object Reference)
   > Set Billboard Material(Billboard Material:Material Interface Object Reference=Get Billboard Material!(self:Impostor Presets Object Reference=@n8) -> (Billboard Material)) -> (Output_Get:Material Interface Object Reference)
   n9 = Get SettingsPreset()
   > Set Constant Specular(Constant Specular:double=Get Constant Specular!(self:Impostor Presets Object Reference=@n9) -> (Constant Specular)) -> (Output_Get:double)
   > Set Constant Roughness(Constant Roughness:double=Get Constant Roughness!(self:Impostor Presets Object Reference=@n9) -> (Constant Roughness)) -> (Output_Get:double)
   > Set Constant Opacity(Constant Opacity:double=Get Constant Opacity!(self:Impostor Presets Object Reference=@n9) -> (Constant Opacity)) -> (Output_Get:double)
   n10 = Get SettingsPreset()
   > Set Subsurface Color(Subsurface Color:Linear Color Structure=Get Subsurface Color!(self:Impostor Presets Object Reference=@n10) -> (Subsurface Color)) -> (Output_Get:Linear Color Structure)
   > Set Scatter Mask Min(Scatter Mask Min:double=Get Scatter Mask Min!(self:Impostor Presets Object Reference=@n10) -> (Scatter Mask Min)) -> (Output_Get:double)
   > Set Scatter Mask Length(Scatter Mask Length:double=Get Scatter Mask Length!(self:Impostor Presets Object Reference=@n10) -> (Scatter Mask Length)) -> (Output_Get:double)
   > Set Mask Offset(Mask Offset:double=Get MaskOffset!(self:Impostor Presets Object Reference=@n10) -> (MaskOffset:double)) -> (Output_Get:double)
   n11 = Get SettingsPreset()
   > Set Shading Model Override(Shading Model Override:EImpostorShadingModelOverride Enum=Get Shading Model Override!(self:Impostor Presets Object Reference=@n11) -> (Shading Model Override:EImpostorShadingModelOverride Enum=NewEnumerator0)) -> (Output_Get:EImpostorShadingModelOverride Enum=NewEnumerator0)
   > Set Parallax Mode(Parallax Mode:EImpostorParallaxMode Enum=Get Parallax Mode!(self:Impostor Presets Object Reference=@n11) -> (Parallax Mode:EImpostorParallaxMode Enum=NewEnumerator0)) -> (Output_Get:EImpostorParallaxMode Enum=NewEnumerator0)
   n12 = Get SettingsPreset()
   Set Override Material(
    execute=@n2.then_6
    Override Material:Material Interface Object Reference=Get Override Material!(self:Impostor Presets Object Reference=@n12) -> (Override Material)
   ) -> (Output_Get:Material Interface Object Reference)
   > Set Pseudo Volume Texture(Pseudo Volume Texture:Texture Object Reference=Get Pseudo Volume Texture!(self:Impostor Presets Object Reference=@n12) -> (Pseudo Volume Texture)) -> (Output_Get:Texture Object Reference)
   > Set PsuedoVolume XY Frames(PsuedoVolume XY Frames:Int Point Structure=Get PsuedoVolume XY Frames!(self:Impostor Presets Object Reference=@n12) -> (PsuedoVolume XY Frames)) -> (Output_Get:Int Point Structure)
   > Set Volume Texture(Volume Texture:Volume Texture Object Reference=Get Volume Texture!(self:Impostor Presets Object Reference=@n12) -> (Volume Texture)) -> (Output_Get:Volume Texture Object Reference)
   > Set Custom Floats(Custom Floats:"Map of Names to Float (double-precision)s"=Get Custom Floats!(self:Impostor Presets Object Reference=@n12) -> (Custom Floats)) -> (Output_Get:"Map of Names to Float (double-precision)s")
   > Set Custom Vectors(Custom Vectors:Map of Names to Linear Color Structures=Get Custom Vectors!(self:Impostor Presets Object Reference=@n12) -> (Custom Vectors)) -> (Output_Get:Map of Names to Linear Color Structures)
   > Set Custom Textures(Custom Textures:Map of Names to Texture 2D Object References=Get Custom Textures!(self:Impostor Presets Object Reference=@n12) -> (Custom Textures)) -> (Output_Get:Map of Names to Texture 2D Object References)
   n13 = Get SettingsPreset()
   Set New Asset Path(execute=@n2.then_5, New Asset Path:String=Get New Asset Path!(self:Impostor Presets Object Reference=@n13) -> (New Asset Path)) -> (Output_Get:String)
   > Set New Asset MIC Prefix(New Asset MIC Prefix:String=Get New Asset MIC Prefix!(self:Impostor Presets Object Reference=@n13) -> (New Asset MIC Prefix:String)) -> (Output_Get:String)
   > Set New Asset Texture Prefix(New Asset Texture Prefix:String=Get New Asset Texture Prefix!(self:Impostor Presets Object Reference=@n13) -> (New Asset Texture Prefix:String)) -> (Output_Get:String)
   > Set Use Relative Path(Use Relative Path=Get Use Relative Path!(self:Impostor Presets Object Reference=@n13) -> (Use Relative Path:bool))
   n14 = Get Color Maps to Render!() -> (Color Maps to Render=NewEnumerator0)
   Set Visibility(
    execute=@n1.then_1
    Visibility=Collapsed or Visible(Index=Contains Item(TargetArray=@n14, ItemToFind=NewEnumerator10)) -> (ReturnValue=Visible)
    self=Get Details_CustomLighting()
   )
   > Set Visibility(
    Visibility=Collapsed or Visible(Index=Get Enable Asset Material Overrides!() -> (Enable Asset Material Overrides)) -> (ReturnValue=Visible)
    self=Get Details_BakeAssetMaterialOverrides()
   )
   > Update Material Settings Details View()
   > Initialize()
   Outputs!()
  ]
 }
 Set Timer by Event {
  path=EventGraph.K2Node_Composite_3.Set Timer by Event
  nodes [
   n1 = Inputs!() -> (execute:Exec, Delegate:Delegate)
   > Set Timer by Event(Delegate=@n1, Time=0.1, bLooping=false)
   Outputs()
  ]
 }
 Save Settings to Preset DataAsset {
  path=EventGraph.K2Node_Composite_4.Save Settings to Preset DataAsset
  nodes [
   Inputs!() -> (exec:Exec)
   > Is Valid(InputObject=Get SettingsPreset())
   > n1 = Sequence()
   > Set Frames XY!(execute, Frames XY=Get Frames XY!() -> (Frames XY=0), self:Impostor Presets Object Reference=Get SettingsPreset()) -> (then, Output_Get=0)
   > Set Color Maps to Render!(
    execute
    Color Maps to Render=Get Color Maps to Render() -> (Color Maps to Render=NewEnumerator0)
    self:Impostor Presets Object Reference=Get SettingsPreset()
   ) -> (then, Output_Get=NewEnumerator0)
   > Set Channel Packed Masks!(
    execute
    Channel Packed Masks=Get Channel Packed Masks!() -> (Channel Packed Masks)
    self:Impostor Presets Object Reference=Get SettingsPreset()
   ) -> (then, Output_Get)
   > Set Impostor Type!(execute, Impostor Type=Get Impostor Type!() -> (Impostor Type=NewEnumerator0), self:Impostor Presets Object Reference=Get SettingsPreset()) -> (then, Output_Get=NewEnumerator0)
   > Set Resolution!(execute, Resolution=Get Resolution!() -> (Resolution=0), self:Impostor Presets Object Reference=Get SettingsPreset()) -> (then, Output_Get=0)
   Set Apply to BaseColor!(
    execute=@n1.then_1
    Apply to BaseColor=Get Apply to BaseColor!() -> (Apply to BaseColor)
    self:Impostor Presets Object Reference=Get SettingsPreset()
   ) -> (then, Output_Get)
   > Set Orthographic!(execute, Orthographic=Get Orthographic!() -> (Orthographic), self:Impostor Presets Object Reference=Get SettingsPreset()) -> (then, Output_Get)
   > Set Camera Distance!(execute, Camera Distance=Get Camera Distance!() -> (Camera Distance), self:Impostor Presets Object Reference=Get SettingsPreset()) -> (then, Output_Get)
   > Set Scene Capture Resolution!(
    execute
    Scene Capture Resolution=Get Scene Capture Resolution!() -> (Scene Capture Resolution=0)
    self:Impostor Presets Object Reference=Get SettingsPreset()
   ) -> (then, Output_Get=0)
   > Set CaptureUsingGBuffer!(
    execute
    CaptureUsingGBuffer=Get CaptureUsingGBuffer!() -> (CaptureUsingGBuffer)
    self:Impostor Presets Object Reference=Get SettingsPreset()
   ) -> (then, Output_Get)
   > Set Enable Asset Material Overrides!(execute, Enable Asset Material Overrides=Get Enable Asset Material Overrides(), self:Impostor Presets Object Reference=Get SettingsPreset()) -> (then, Output_Get)
   > Set Mesh Scale!(execute, Mesh Scale=Get Mesh Scale!() -> (Mesh Scale), self:Impostor Presets Object Reference=Get SettingsPreset()) -> (then, Output_Get)
   Set New Asset Path!(execute=@n1.then_5, New Asset Path=Get New Asset Path!() -> (New Asset Path), self:Impostor Presets Object Reference=Get SettingsPreset()) -> (then, Output_Get)
   > Set New Asset Texture Prefix!(
    execute
    New Asset Texture Prefix=Get New Asset Texture Prefix!() -> (New Asset Texture Prefix)
    self:Impostor Presets Object Reference=Get SettingsPreset()
   ) -> (then, Output_Get)
   > Set New Asset MIC Prefix!(
    execute
    New Asset MIC Prefix=Get New Asset MIC Prefix!() -> (New Asset MIC Prefix)
    self:Impostor Presets Object Reference=Get SettingsPreset()
   ) -> (then, Output_Get)
   > Set Use Relative Path!(execute, Use Relative Path=Get Use Relative Path!() -> (Use Relative Path), self:Impostor Presets Object Reference=Get SettingsPreset()) -> (then, Output_Get)
   Set Override Material!(
    execute=@n1.then_6
    Override Material=Get Override Material!() -> (Override Material)
    self:Impostor Presets Object Reference=Get SettingsPreset()
   ) -> (then, Output_Get)
   > Set Pseudo Volume Texture!(
    execute
    Pseudo Volume Texture=Get Pseudo Volume Texture!() -> (Pseudo Volume Texture)
    self:Impostor Presets Object Reference=Get SettingsPreset()
   ) -> (then, Output_Get)
   > Set PsuedoVolume XY Frames!(
    execute
    PsuedoVolume XY Frames=Get PsuedoVolume XY Frames!() -> (PsuedoVolume XY Frames)
    self:Impostor Presets Object Reference=Get SettingsPreset()
   ) -> (then, Output_Get)
   > Set Volume Texture!(execute, Volume Texture=Get Volume Texture!() -> (Volume Texture), self:Impostor Presets Object Reference=Get SettingsPreset()) -> (then, Output_Get)
   > Set Custom Floats!(execute, Custom Floats=Get Custom Floats!() -> (Custom Floats), self:Impostor Presets Object Reference=Get SettingsPreset()) -> (then, Output_Get)
   > Set Custom Vectors!(execute, Custom Vectors=Get Custom Vectors!() -> (Custom Vectors), self:Impostor Presets Object Reference=Get SettingsPreset()) -> (then, Output_Get)
   > Set Custom Textures!(execute, Custom Textures=Get Custom Textures!() -> (Custom Textures), self:Impostor Presets Object Reference=Get SettingsPreset()) -> (then, Output_Get)
   Set Full Sphere Material!(
    execute=@n1.then_3
    Full Sphere Material=Get Full Sphere Material!() -> (Full Sphere Material)
    self:Impostor Presets Object Reference=Get SettingsPreset()
   ) -> (then, Output_Get)
   > Set Hemisphere Material!(
    execute
    Hemisphere Material=Get Hemisphere Material!() -> (Hemisphere Material)
    self:Impostor Presets Object Reference=Get SettingsPreset()
   ) -> (then, Output_Get)
   > Set Billboard Material!(execute, Billboard Material=Get Billboard Material!() -> (Billboard Material), self:Impostor Presets Object Reference=Get SettingsPreset()) -> (then, Output_Get)
   > Set Constant Specular!(execute, Constant Specular=Get Constant Specular!() -> (Constant Specular), self:Impostor Presets Object Reference=Get SettingsPreset()) -> (then, Output_Get)
   > Set Constant Roughness!(execute, Constant Roughness=Get Constant Roughness!() -> (Constant Roughness), self:Impostor Presets Object Reference=Get SettingsPreset()) -> (then, Output_Get)
   > Set Constant Opacity!(execute, Constant Opacity=Get Constant Opacity!() -> (Constant Opacity), self:Impostor Presets Object Reference=Get SettingsPreset()) -> (then, Output_Get)
   > Set Subsurface Color!(execute, Subsurface Color=Get Subsurface Color!() -> (Subsurface Color), self:Impostor Presets Object Reference=Get SettingsPreset()) -> (then, Output_Get)
   > Set Scatter Mask Min!(execute, Scatter Mask Min=Get Scatter Mask Min!() -> (Scatter Mask Min), self:Impostor Presets Object Reference=Get SettingsPreset()) -> (then, Output_Get)
   > Set Scatter Mask Length!(
    execute
    Scatter Mask Length=Get Scatter Mask Length!() -> (Scatter Mask Length)
    self:Impostor Presets Object Reference=Get SettingsPreset()
   ) -> (then, Output_Get)
   > Set MaskOffset!(execute, MaskOffset:double=Get Mask Offset(), self:Impostor Presets Object Reference=Get SettingsPreset()) -> (then, Output_Get:double)
   > Set Shading Model Override!(
    execute
    Shading Model Override=Get Shading Model Override!() -> (Shading Model Override=NewEnumerator0)
    self:Impostor Presets Object Reference=Get SettingsPreset()
   ) -> (then, Output_Get=NewEnumerator0)
   > Set Parallax Mode!(execute, Parallax Mode=Get Parallax Mode!() -> (Parallax Mode=NewEnumerator0), self:Impostor Presets Object Reference=Get SettingsPreset()) -> (then, Output_Get=NewEnumerator0)
   Set Reference Plane Sides!(
    execute=@n1.then_2
    Reference Plane Sides=Get Reference Plane Sides!() -> (Reference Plane Sides)
    self:Impostor Presets Object Reference=Get SettingsPreset()
   ) -> (then, Output_Get)
   > Set Reference Plane Top!(
    execute
    Reference Plane Top=Get Reference Plane Top!() -> (Reference Plane Top)
    self:Impostor Presets Object Reference=Get SettingsPreset()
   ) -> (then, Output_Get)
   > Set Center XY on Mesh Pivot!(
    execute
    Center XY on Mesh Pivot=Get Center XY on Mesh Pivot!() -> (Center XY on Mesh Pivot)
    self:Impostor Presets Object Reference=Get SettingsPreset()
   ) -> (then, Output_Get)
   > Set Use Sprite Vertex Shader!(
    execute
    Use Sprite Vertex Shader=Get Use Sprite Vertex Shader!() -> (Use Sprite Vertex Shader)
    self:Impostor Presets Object Reference=Get SettingsPreset()
   ) -> (then, Output_Get)
   Set Custom Lighting Power!(
    execute=@n1.then_4
    Custom Lighting Power=Get Custom Lighting Power!() -> (Custom Lighting Power)
    self:Impostor Presets Object Reference=Get SettingsPreset()
   ) -> (then, Output_Get)
   > Set Custom Lighting Opacity!(
    execute
    Custom Lighting Opacity=Get Custom Lighting Opacity!() -> (Custom Lighting Opacity)
    self:Impostor Presets Object Reference=Get SettingsPreset()
   ) -> (then, Output_Get)
   > Set Lighting Grid XY!(execute, Lighting Grid XY=Get Lighting Grid XY!() -> (Lighting Grid XY=0), self:Impostor Presets Object Reference=Get SettingsPreset()) -> (then, Output_Get=0)
   > Set Upward Bias!(execute, Upward Bias=Get Upward Bias!() -> (Upward Bias), self:Impostor Presets Object Reference=Get SettingsPreset()) -> (then, Output_Get)
   > Set DirLightBrightness!(execute, DirLightBrightness=Get DirLightBrightness!() -> (DirLightBrightness), self:Impostor Presets Object Reference=Get SettingsPreset()) -> (then, Output_Get)
   > Set Custom SkyLight Intensity!(
    execute
    Custom SkyLight Intensity=Get Custom SkyLight Intensity!() -> (Custom SkyLight Intensity)
    self:Impostor Presets Object Reference=Get SettingsPreset()
   ) -> (then, Output_Get)
   > Set Apply to BaseColor(Apply to BaseColor=Get Apply to BaseColor(), self=Get SettingsPreset())
   Outputs()
  ]
 }
 GenerateMeshData {nodes [
   GenerateMeshData() {is=title K2Node_FunctionEntry}
   > Initialize!(
    execute
    self:Impostor Mesh Cutout Object Reference=Get Impostor_MeshCutout_Component() -> (Impostor_MeshCutout_Component:Impostor Mesh Cutout Object Reference)
    DynamicMeshComponent:Dynamic Mesh Component Object Reference=Get DynamicMeshComponent()
    CombinedAlphas RT:Texture Render Target 2D Object Reference=Get CombinedAlphasRT()
    Impostor Type:EImposterLayoutType Enum=Get Impostor Type() -> (Impostor Type=NewEnumerator0)
    Object Radius:double=Get Object Radius()
    Offset:Vector=Get Offset Vector()
    FramesXY:int=Get FramesXYInternal() -> (FramesXYInternal=0)
    Capture Vectors:Array of Vectors=Get ViewCaptureVectors()
   ) -> (then)
   > n1 = GenerateMeshData!(execute, self:Impostor Mesh Cutout Object Reference=Get Impostor_MeshCutout_Component()) -> (then, Verts:Array of Vectors, Tris:Array of Integers, UVs:Array of Vector 2D Structures) {is=title K2Node_CallFunction}
   > Set Verts(Verts:Array of Vectors=@n1) -> (Output_Get:Array of Vectors)
   > Set Tris(Tris:Array of Integers=@n1.Tris) -> (Output_Get:Array of Integers)
   > Set UVs(UVs:Array of Vector 2D Structures=@n1.UVs) -> (Output_Get:Array of Vector 2D Structures)
  ]}
 GetSettingsPresetFromINI {nodes [
   Inputs!() -> (execute:Exec)
   > n1 = Branch(Condition="Equal Exactly (String)"(A=Get DefaultSettingsAssetPath() -> (DefaultSettingsAssetPath:String)))
   Outputs!(OutputPin:Exec=[@n6.then, @n5.Is Not Valid, @n1.then])
   n2 = LoadAsset(execute=@n1.else, AssetPath=Get DefaultSettingsAssetPath())
   > n3 = Cast To ImpostorPresets(Object=@n2)
   > n4 = Set SettingsPreset(SettingsPreset=@n3)
   n5 = Is Valid(exec=[@n4.then, @n3.CastFailed], InputObject=@n4)
   > n6 = Apply Settings Preset() {is=title K2Node_CallFunction}
  ]}
 ImpostorGridVector {nodes [
   n1 = vector2d - float!(
    A:Vector 2D Structure=vector2d * float!(A:Vector 2D Structure=Inputs!() -> (0 to 1:Vector 2D Structure), B:double=2) -> (ReturnValue:Vector 2D Structure)
    B:double=1
   ) -> (ReturnValue:Vector 2D Structure)
   n2 = Get ImpostorViewport(self=Get ImpostorViewport_Widget())
   n3 = HemiOctahedronToUnitVector!(Hemi Octahedron:Vector 2D Structure=@n1, __WorldContext:Object Reference=@n2) -> (Unit Vector:Vector)
   Outputs!(ReturnValue:Vector=Select!(
     NewEnumerator0:Vector=OctahedronToVector!(Octahedron:Vector 2D Structure=@n1, __WorldContext:Object Reference=@n2) -> (Vector:Vector)
     NewEnumerator1:Vector=@n3
     NewEnumerator2:Vector=Normalize!(A:Vector=@n3, Tolerance:float=0.0001) -> (ReturnValue:Vector) {is=title K2Node_CallFunction}
     Index:EImposterLayoutType Enum=Get Impostor Type() -> (Impostor Type=NewEnumerator0)
    ) -> (ReturnValue:Vector))
  ]}
 ImpostorRenderComplete {nodes [ImpostorRenderComplete() {is=title K2Node_FunctionEntry}]}
 ImpostorRenderSaveComplete {nodes [ImpostorRenderSaveComplete() {is=title K2Node_FunctionEntry}]}
 ImpostorSaveComplete {nodes [ImpostorSaveComplete() {is=title K2Node_FunctionEntry}]}
 Initialize {nodes [
   Initialize() {is=title K2Node_FunctionEntry}
   > Setup RT and Save List()
   > Set Capture Mesh and Size Info(ReAdjustCamera=false)
   > Setup Octahedron Layout()
   > Clear RenderTargets() {is=title K2Node_CallFunction}
   > Set TextureTarget(
    TextureTarget="Get (a copy)"(Array:Array of Texture Render Target 2D Object References=Get SceneCaptureMipChain()) -> (Output:Texture Render Target 2D Object Reference)
    self=Get SceneCaptureComponent2D!() -> (SceneCaptureComponent2D)
   )
   n1 = Get SceneCaptureComponent2D()
   > ClearShowOnlyComponents(self:Scene Capture Component Object Reference=@n1)
   > ShowOnlyComponent!(execute, self:Scene Capture Component Object Reference=@n1, InComponent:Primitive Component Object Reference=Get StaticMeshComponent()) -> (then)
   > n2 = Branch(Condition=Get CaptureUsingGBuffer())
   > n3 = ShowOnlyComponent(
    self=@n1
    InComponent=Get ViewMode_TranslucentSphere!(self:Impostor Scene Object Reference=Get Impostor Scene()) -> (ViewMode_TranslucentSphere:Static Mesh Component Object Reference)
   )
   Set ProjectionType!(
    execute=[@n3.then, @n2.else]
    ProjectionType:ECameraProjectionMode Enum=Select!(Option 0:ECameraProjectionMode Enum=Perspective, Option 1:ECameraProjectionMode Enum=Orthographic, Index:bool=Get Orthographic()) -> (ReturnValue:ECameraProjectionMode Enum)
    self:Scene Capture Component 2D Object Reference=Get SceneCaptureComponent2D()
   ) -> (then, Output_Get:ECameraProjectionMode Enum=Perspective)
   > Set OrthoWidth!(
    execute
    OrthoWidth:float=float * float(A=Get Object Radius(), B=2)
    self:Scene Capture Component 2D Object Reference=Get SceneCaptureComponent2D()
   ) -> (then, Output_Get:float)
   n4 = float / float(A=Get Object Radius(), B=Get Camera Distance())
   > Set FOVAngle!(
    execute
    FOVAngle:float=float * float(A="Atan (Degrees)"!(A:double=@n4) -> (ReturnValue:double), B=2)
    self:Scene Capture Component 2D Object Reference=Get SceneCaptureComponent2D()
   ) -> (then, Output_Get:float)
   > Create And Setup MIDs()
   > SetMaterial(self=Get DynamicMeshComponent(), ElementIndex=0, Material=Get ImpostorPreviewMID())
   > Set CurrentMapIndex(CurrentMapIndex:int=0) -> (Output_Get:int=0)
   > Reset All Capture Modes()
   > SetVisibility!(execute, self:Scene Component Object Reference=Get StaticMeshComponent(), bNewVisibility:bool=true, bPropagateToChildren:bool=false) -> (then)
   > SetVisibility(self=Get DynamicMeshComponent(), bNewVisibility=false, bPropagateToChildren=false)
   SetMeshData!(
    execute
    self:Impostor Mesh Cutout Object Reference=Get Impostor_MeshCutout_Component()
    Verts:Array of Vectors=Get Verts() -> (Verts:Array of Vectors)
    Tris:Array of Integers=Get Tris() -> (Tris:Array of Integers)
    UVs:Array of Vector 2D Structures=Get UVs() -> (UVs:Array of Vector 2D Structures)
   ) -> (then)
  ]}
 IsPathValidSoftObj {nodes [
   n1 = Inputs!() -> (PathString:String)
   n2 = Resolve Soft Reference!(Input:Soft Object Reference=To Soft Object Reference(SoftObjectPath=MakeSoftObjectPath!(PathString:String=@n1) -> (ReturnValue:Soft Object Path Structure))) -> (Output:Object Reference)
   Outputs!(ReturnValue:bool=IsValid(Object=@n2), Object:Object Reference=@n2)
  ]}
 ObjectPathAddOuter {nodes [
   n1 = Split(SourceString=Inputs!() -> (Path:String), InStr=/, SearchDir=FromEnd)
   Outputs!(OutPath:String=Append!(A=@n1, B=/, C=@n1.RightS, D=., E:String=@n1.RightS) -> (ReturnValue))
  ]}
 PopulateViewModes {nodes [
   Inputs!() -> (In:Exec)
   n1 = Get Combobox_ViewMode() -> (Combobox_ViewMode:Editor Utility Combo Box String Object Reference)
   > ClearOptions(self:"Combo Box (String) Object Reference"=@n1)
   > AddOption!(execute, self:"Combo Box (String) Object Reference"=@n1, Option:String=Lit) -> (then)
   > n2 = ForEach EImpostorBakeTypes!(execute:Exec, SkipHidden:bool) -> (LoopBody:Exec, EnumValue:EImpostorBakeTypes Enum, then:Exec) {is=title K2Node_ForEachElementInEnum}
   SetSelectedOption!(execute=@n2.then, self:"Combo Box (String) Object Reference"=Get Combobox_ViewMode(), Option:String=Lit) -> (then)
   > Outputs!(Out:Exec)
   n3 = Branch(
    execute=@n2.LoopBody
    Condition=OR Boolean!(
     A="Equal (Enum)"(A:EImpostorBakeTypes Enum=@n2, B:EImpostorBakeTypes Enum=NewEnumerator9)
     B="Equal (Enum)"(A=@n2, B=NewEnumerator8)
     C:bool="Equal (Enum)"(A=@n2, B=NewEnumerator18)
    ) -> (ReturnValue)
   )
   AddOption(execute=@n3.else, self=Get Combobox_ViewMode(), Option=Enum to String(Enumerator=@n2))
  ]}
 Process Path Name {nodes [
   n1 = Split(SourceString=Get Object Path String(Object=Get Static Mesh()), InStr=/, SearchDir=FromEnd)
   n2 = Get New Asset Path()
   n3 = Get Use Relative Path()
   n4 = Append!(
    A=@n1
    B=Select(Option 0:String=Select(Option 0=/, Index=@n3), Option 1:String=/, Index=StartsWith(SourceString=@n2, InPrefix=../)) -> (ReturnValue:String)
   ) -> (ReturnValue)
   n5 = PrintString(InString=@n4)
   n6 = Get New Asset Path()
   n7 = Get New Asset Path()
   n8 = Select(
    Option 0=Get New Asset Path()
    Option 1=CollapseRelativeDirectories!(InPath:String=Append(A=@n4, B=@n6)) -> (OutPath:String, ReturnValue:bool)
    Index=OR Boolean!(A=Get Use Relative Path(), B=StartsWith(SourceString=@n7, InPrefix=../)) -> (ReturnValue)
   )
   n9 = Get New Asset Texture Prefix()
   n10 = Append(A=_, B=Split(SourceString=@n1.RightS, InStr=., SearchDir=FromEnd))
   Set Processed Texture Path(execute=[@n5.then, @n14.execute], Processed Texture Path:String=Append(A=Append(A=@n8, B=@n9), B=@n10)) -> (Output_Get:String)
   n11 = Get New Asset MIC Prefix()
   > n12 = Set Processed MIC Path(Processed MIC Path:String=Append(A=Append(A=@n8, B=@n11), B=@n10)) -> (Output_Get:String)
   Outputs!(then:Exec=[@n13.then, @n12.then])
   n13 = PrintString(InString=Hello)
   n14 = Inputs!() -> (execute:Exec)
  ]}
 Render SceneCapture Mip Chain {nodes [
   Inputs()
   n1 = Get SceneCaptureMipChain()
   n2 = Get CurrentMap() -> (CurrentMap=NewEnumerator0)
   > n3 = For Loop(
    FirstIndex=1
    LastIndex=SelectInt(
     A=Last Index(TargetArray:Array of Texture Render Target 2D Object References=@n1) -> (ReturnValue=0)
     B=Get SceneCaptureMipToComposite() -> (SceneCaptureMipToComposite=0)
     bPickA="Equal (Enum)"(A=@n2, B=NewEnumerator0)
    )
   )
   Outputs(then=@n3.Completed)
   n4 = Get SceneCaptureMipChain()
   n5 = "Get (a copy)"(Array=@n4, Dimension 1=@n3)
   > ClearRenderTarget2D(TextureRenderTarget=@n5, ClearColor=(R=0,G=0,B=0,A=1))
   > Resample RT(Source RT="Get (a copy)"(Array=@n4, Dimension 1=int - int(A=@n3, B=1) {is=title K2Node_CallFunction}), Dest RT=@n5)
  ]}
 Resample RT {nodes [
   n1 = Resample RT!() -> (then, Source RT:Texture Render Target 2D Object Reference, Dest RT:Texture Render Target 2D Object Reference) {is=title K2Node_FunctionEntry}
   > SetTextureParameterValue(self=Get Resample MID() -> (Resample MID:Material Instance Dynamic Object Reference), ParameterName=RT, Value=@n1)
   > ClearRenderTarget2D(WorldContextObject=Get World(), TextureRenderTarget=@n1.Dest RT, ClearColor=(R=0,G=0,B=0,A=1))
   > DrawMaterialToRenderTarget(WorldContextObject=Get World(), TextureRenderTarget=@n1.Dest RT, Material=Get Resample MID())
  ]}
 Reset All Capture Modes {nodes [
   Inputs()
   > n1 = Sequence!(execute) -> (then_0, then_1, then_2)
   > SetScalarParameterValue(
    WorldContextObject=Get Impostor Scene()
    Collection=&MPC_ImposterCapture
    ParameterName=EnableCapture
    ParameterValue=0
   ) {is=title K2Node_CallMaterialParameterCollectionFunction}
   > n2 = For Each Loop(Array:Array of EImpostorBakeTypes Enums=Make Array!("[0]"=Get CaptureTypes() -> (CaptureTypes:EImpostorBakeTypes Enum=NewEnumerator0)) -> (Array)) -> (Array Element:EImpostorBakeTypes Enum)
   > n3 = Branch(Condition=OR Boolean(A="Equal (Enum)"(A=@n2, B=NewEnumerator9), B="Equal (Enum)"(A=@n2, B=NewEnumerator18)))
   SetScalarParameterValue(
    execute=@n3.else
    WorldContextObject=Get World()
    Collection=&MPC_ImposterCapture
    ParameterName=String To Name(InString=Enum to String(Enumerator=@n2))
    ParameterValue=0
   )
   SetSelectedOption(execute=@n1.then_1, self=Get Combobox_ViewMode(), Option=Lit)
   > Change View Modes(self=Get Impostor Scene(), ViewMode=Lit)
   > Restore Lighting State()
   Outputs!(then_1:Exec=@n1.then_2)
  ]}
 Restore Lighting State {nodes [
   Inputs()
   > n1 = For Each Loop(Array:Array of Directional Light Object References=Get lights()) -> (Array Element:Directional Light Object Reference)
   Clear(execute=@n1.Completed, TargetArray:Array of Directional Light Object References=Get lights())
   > SetVisibility(self=Get DirectionalLight(self=Get Impostor Scene()), bNewVisibility=true, bPropagateToChildren=false)
   > SetVisibility(self=Get Decal_CaptureCustomLighting(self=Get Impostor Scene()), bNewVisibility=false, bPropagateToChildren=false)
   > SetIntensity(self:Sky Light Component Object Reference=Get SkyLight(self=Get Impostor Scene()), NewIntensity=1)
   Destroy Actor(execute=@n1.LoopBody, self=@n1)
   Outputs!()
  ]}
 Save DynamicMesh to Staticmesh {nodes [
   Reroute Node()
   n1 = GetNumLODs(self=Get Static Mesh())
   > n2 = GetSectionMaterialListFromStaticMesh!(
    execute
    FromStaticMeshAsset:Static Mesh Object Reference=Get Static Mesh()
    RequestedLOD:Geometry Script Mesh Read LOD Structure=Make Geometry Script Mesh Read LOD!(LODType:EGeometryScriptLODType Enum=MaxAvailable, LODIndex:int=int - int(A=@n1) {is=title K2Node_PromotableOperator}) -> (GeometryScriptMeshReadLOD:Geometry Script Mesh Read LOD Structure) {is=title K2Node_MakeStruct}
   ) -> (
    Failure:Exec
    Success:Exec
    MaterialList:Array of Material Interface Object References
    MaterialIndex:Array of Integers
    MaterialSlotNames:Array of Names
   )
   > Branch(Condition="Equal (Integer)"(A=Length(TargetArray:Array of Integers=@n2.MaterialIndex) {is=title K2Node_CallArrayFunction}, B=1) {is=title K2Node_PromotableOperator})
   n3 = Inputs!() -> (execute, Overwrite Last LOD:bool=~)
   > n4 = Branch(Condition=@n3)
   n5 = Get Static Mesh()
   n6 = Get Created MIC()
   n7 = Get Static Mesh()
   > n8 = SetMaterial!(
    execute
    self:Static Mesh Object Reference=@n7
    MaterialIndex:int=Last Index(TargetArray:Array of Static Material Structures=Get StaticMaterials!(self:Static Mesh Object Reference=@n5) -> (StaticMaterials:Array of Static Material Structures)) -> (ReturnValue=0)
    NewMaterial:Material Interface Object Reference=@n6
   ) -> (then)
   n9 = GetDynamicMesh!(execute=[@n16.then, @n8.then], self:Base Dynamic Mesh Component Object Reference=Get DynamicMeshComponent()) -> (then, ReturnValue:Dynamic Mesh Object Reference)
   > n10 = Set Dynamic Mesh(Dynamic Mesh:Dynamic Mesh Object Reference=@n9) -> (Output_Get:Dynamic Mesh Object Reference)
   n11 = Get Static Mesh()
   > ClearMaterialIDs!(
    execute
    TargetMesh:Dynamic Mesh Object Reference=@n10
    ClearValue:int=Last Index(TargetArray=Get StaticMaterials(self=@n11)) -> (ReturnValue=0)
   ) -> (then, ReturnValue:Dynamic Mesh Object Reference)
   n12 = GetNumLODs(self=Get Static Mesh())
   n13 = int - int(A=@n12)
   > n14 = CopyMeshToStaticMesh!(
    execute
    FromDynamicMesh:Dynamic Mesh Object Reference=Get Dynamic Mesh() -> (Dynamic Mesh:Dynamic Mesh Object Reference)
    ToStaticMeshAsset:Static Mesh Object Reference=Get Static Mesh()
    Options:Geometry Script Copy Mesh To Asset Options Structure=Make Geometry Script Copy Mesh to Asset Options!(
     bEnableRecomputeNormals:bool=true
     bEnableRecomputeTangents:bool=true
     bEnableRemoveDegenerates:bool=False
     BoneHierarchyMismatchHandling:EGeometryScriptBoneHierarchyMismatchHandling Enum=DoNothing
     bRemapBoneIndicesToMatchAsset:bool=False
     bUseOriginalVertexOrder:bool=False
     bUseBuildScale:bool=True
     bReplaceMaterials:bool=False
     NewMaterialUsage:ECopyMeshToAssetUseNewMaterialsOptions Enum=ReplaceAssetMaterials
     NewMaterials:Array of Material Interface Object References
     NewMaterialSlotNames:Array of Names
     bCleanAssignedMaterials:bool=False
     GenerateLightmapUVs:EGeometryScriptGenerateLightmapUVOptions Enum=MatchTargetLODSetting
     bApplyNaniteSettings:bool=False
     NaniteSettings:Geometry Script Nanite Options Structure=(bEnabled=True,FallbackPercentTriangles=100,FallbackRelativeError=0)
     NewNaniteSettings:Mesh Nanite Settings Structure=(bEnabled=False,bExplicitTangents=False,bLerpUVs=True,bSeparable=False,bVoxelNDF=True,bVoxelOpacity=False,ShapePreservation=None,PositionPrecision=-2147483648,NormalPrecision=-1,TangentPrecision=-1,BoneWeightPrecision=-1,TargetMinimumResidencyInKB=0,KeepPercentTriangles=1,TrimRelativeError=0,GenerateFallback=PlatformDefault,FallbackTarget=Auto,FallbackPercentTriangles=1,FallbackRelativeError=1,MaxEdgeLengthFactor=0,NumRays=64,VoxelLevel=0,RayBackUp=0,DisplacementUVChannel=0,DisplacementMaps=,NaniteAssemblyData=(Parts=,Nodes=))
     bEmitTransaction:bool=True
     bDeferMeshPostEditChange:bool=False
    ) -> (GeometryScriptCopyMeshToAssetOptions:Geometry Script Copy Mesh To Asset Options Structure) {is=title K2Node_MakeStruct}
    TargetLOD:Geometry Script Mesh Write LOD Structure=Make Geometry Script Mesh Write LOD!(bWriteHiResSource:bool=False, LODIndex:int=Select(Option 0:int=@n12, Option 1:int=@n13, Index=@n3) -> (ReturnValue:int)) -> (GeometryScriptMeshWriteLOD:Geometry Script Mesh Write LOD Structure) {is=title K2Node_MakeStruct}
    bUseSectionMaterials:bool=true
   ) -> (Failure:Exec, Success:Exec, ReturnValue:Dynamic Mesh Object Reference)
   > n15 = PrintString(InString=Impostor Generator: Failed to Save DynamicMesh to StaticMesh.)
   Outputs!(then:Exec=[@n14.Success, @n15.then])
   n16 = AddMaterial!(execute=@n4.else, self:Static Mesh Object Reference=@n7, Material:Material Interface Object Reference=@n6) -> (then, ReturnValue:Name=None)
   "Get (a ref)"(Array:Array of Integers=@n2.MaterialIndex, Dimension 1=0) -> (Output:"Integer (by ref)")
  ]}
 ScalarMapToBakeType {nodes [Outputs!(ReturnValue:EImpostorBakeTypes Enum=Select!(
     NewEnumerator1:EImpostorBakeTypes Enum=NewEnumerator1
     NewEnumerator2:EImpostorBakeTypes Enum=NewEnumerator2
     NewEnumerator3:EImpostorBakeTypes Enum=NewEnumerator3
     NewEnumerator14:EImpostorBakeTypes Enum=NewEnumerator14
     NewEnumerator5:EImpostorBakeTypes Enum=NewEnumerator5
     NewEnumerator12:EImpostorBakeTypes Enum=NewEnumerator12
     NewEnumerator16:EImpostorBakeTypes Enum=NewEnumerator16
     NewEnumerator17:EImpostorBakeTypes Enum=NewEnumerator17
     NewEnumerator13:EImpostorBakeTypes Enum=NewEnumerator13
     NewEnumerator8:EImpostorBakeTypes Enum=NewEnumerator8
     NewEnumerator18:EImpostorBakeTypes Enum=NewEnumerator18
     Index:EImpostorScalarMaps Enum=Inputs!() -> (Index:EImpostorScalarMaps Enum=NewEnumerator1)
    ) -> (ReturnValue:EImpostorBakeTypes Enum))]}
 Set Capture Mesh and Size Info {nodes [
   Set Capture Mesh and Size Info!() -> (then, ReAdjustCamera:bool=~) {is=title K2Node_FunctionEntry}
   > SetStaticMesh(self=Get StaticMeshComponent(), NewMesh=Get Static Mesh())
   n1 = Get StaticMeshComponent()
   > Set World Location(self=@n1, NewLocation="0, 0, 0", bSweep=false, bTeleport=false)
   n2 = GetComponentBounds!(Component:Scene Component Object Reference=@n1) -> (Origin:Vector, BoxExtent:Vector, SphereRadius:float)
   n3 = BreakVector(InVec=@n2)
   > n4 = Set Offset Vector(Offset Vector:Vector=Select!(
     Option 0:Vector=@n2
     Option 1:Vector=MakeVector_NetQuantize!(X:double=0, Y:double=0, Z:double=@n3.Z) -> (ReturnValue:Vector Net Quantize Structure)
     Index:bool=Get Center XY on Mesh Pivot()
    ) -> (ReturnValue:Vector)) -> (Output_Get:Vector)
   n5 = Get StaticMeshComponent()
   > Set World Location(self=[@n5, Get DynamicMeshComponent()], NewLocation=vector * vector(A=@n4, B:float=-1) -> (ReturnValue=~), bSweep=false, bTeleport=false)
   n6 = Get Mesh Scale()
   n7 = BreakVector(InVec=@n6)
   n8 = "Max (Float)"!(A=@n7, B=@n7.Y, C:double=@n7.Z) -> (ReturnValue)
   > Set Object Radius(Object Radius:double=float / float(A=@n2.SphereRadius, B=@n8) {is=title K2Node_PromotableOperator}) -> (Output_Get:double)
   > SetWorldScale3D!(execute, self:Scene Component Object Reference=@n5, NewScale:Vector=vector / vector(A:Vector=@n6, B=@n8) -> (ReturnValue:Vector)) -> (then)
   > Set Zoom Sensitivity!(
    execute
    Zoom Sensitivity:double=float / float(A=Get Object Radius(), B=100)
    self:Impostor Viewport Widget Object Reference=Get ImpostorViewport_Widget()
   ) -> (then, Output_Get:double)
   n9 = float * float(A=Get Object Radius(), B=4) {is=title K2Node_PromotableOperator}
   > Set SwapDistance(SwapDistance=@n9)
   > Set Text(
    Text="To Text (Float)"(Value=@n9, MinimumIntegralDigits=1, MaximumIntegralDigits=324, MinimumFractionalDigits=0, MaximumFractionalDigits=3)
    self:Text Box Object Reference=Get Text_SwapDistance() -> (Text_SwapDistance:Editor Utility Editable Text Box Object Reference)
   )
   n10 = Get Object Radius()
   n11 = Get Preview Floor(self=Get Impostor Scene())
   > Set World Location(self=@n11, NewLocation=MakeVector_NetQuantize(X=0, Y=0, Z=float * float(A=@n10, B=-1)), bSweep=false, bTeleport=false)
   n12 = Get Object Radius()
   > SetWorldScale3D(self=@n11, NewScale="To Vector (Float)"!(InDouble:double=float / float(A=@n12, B=256)) -> (ReturnValue:Vector))
   > Set AnyNonOpaque(AnyNonOpaque:bool=false) -> (Output_Get:bool)
   > n13 = Sequence!(execute) -> (then_0, then_1, then_2, then_3:Exec)
   > n14 = For Each Loop(Array:Array of Static Material Structures=Get StaticMaterials(self=Get Static Mesh())) -> (Array Element:Static Material Structure)
   n15 = GetBlendMode!(self:Material Interface Object Reference=Break Static Material!(StaticMaterial:"Static Material Structure (by ref)"=@n14) -> (
     MaterialInterface:Material Interface Object Reference
     MaterialSlotName:Name
     UVChannelData:Mesh UVChannel Info Structure
     OverlayMaterialInterface:Material Interface Object Reference
    ) {is=title K2Node_BreakStruct}) -> (ReturnValue:EBlendMode Enum=BLEND_Opaque)
   > Branch(Condition=OR Boolean!(
     A="Equal (Enum)"(A:EBlendMode Enum=@n15, B:EBlendMode Enum=BLEND_Translucent)
     B="Equal (Enum)"(A=@n15, B=BLEND_Additive)
     C:bool="Equal (Enum)"(A=@n15, B=BLEND_Modulate)
    ) -> (ReturnValue))
   > Set AnyNonOpaque(AnyNonOpaque=true)
   n16 = Branch(execute=@n13.then_1, Condition=AND Boolean(A=Get AnyNonOpaque() -> (AnyNonOpaque:bool), B=Get CaptureUsingGBuffer()))
   n17 = Get Text_GBuffer_Doesnt_Support_Translucency() -> (Text_GBuffer_Doesnt_Support_Translucency:Text Object Reference)
   > SetVisibility!(execute, self:Widget Object Reference=@n17, InVisibility:ESlateVisibility Enum=HitTestInvisible) -> (then)
   SetVisibility(execute=@n16.else, self=@n17, InVisibility=Collapsed)
   Clear Dynamic Mesh(execute=@n13.then_3, self:Impostor Scene Object Reference=Get Impostor Scene())
   > Set Rendered(Rendered=false)
   Branch(execute=@n13.then_2, Condition=Get ReAdjustCamera() -> (ReAdjustCamera:bool))
   n18 = Get ImpostorViewport_Widget()
   n19 = Normalize(A=GetViewLocation(self=Get ImpostorViewport(self=@n18)), Tolerance=0.0001)
   n20 = Vector Length(A=@n19)
   n21 = float * float(A=Get Object Radius(), B=1.5)
   > SetViewLocation!(
    execute
    self:Viewport Object Reference=Get ImpostorViewport(self=Get ImpostorViewport_Widget())
    Location:Vector=vector * vector(A=@n19, B:double="Max (Float)"!(A=@n20, B=@n21) -> (ReturnValue)) -> (ReturnValue=~)
   ) -> (then)
  ]}
 Set MIC Parameters {nodes [
   Inputs!() -> (execute:Exec)
   n1 = Get Created MIC()
   > Set MIC Scalar Parameter Editor Only!(execute, Material:Material Instance Constant Object Reference=@n1, ParamName:String=Specular, Value:float=Get Constant Specular()) -> (then, ReturnValue:bool) {is=title K2Node_CallFunction}
   > Set MIC Scalar Parameter Editor Only(Material=@n1, ParamName=Roughness, Value=Get Constant Roughness())
   > Set MIC Scalar Parameter Editor Only(Material=@n1, ParamName=Opacity, Value=Get Constant Opacity())
   > Set MIC Scalar Parameter Editor Only(Material=@n1, ParamName=Scatter Mask Min, Value=Get Scatter Mask Min())
   > Set MIC Scalar Parameter Editor Only(Material=@n1, ParamName=Scatter Mask Length, Value=Get Scatter Mask Length())
   > Set MIC Scalar Parameter Editor Only(Material=@n1, ParamName=MaskOffset, Value=Get Mask Offset())
   > Set MIC Vector Parameter Editor Only!(
    execute
    Material:Material Instance Constant Object Reference=@n1
    ParamName:String=SubsurfaceColor
    Value:Linear Color Structure=Get Subsurface Color()
   ) -> (then, ReturnValue:bool) {is=title K2Node_CallFunction}
   n2 = Get Created MIC()
   > Set MIC Scalar Parameter Editor Only(Material=@n2, ParamName=FramesXY, Value="To Float (Integer)"(InInt=Get FramesXYInternal() -> (FramesXYInternal=0)))
   > Set MIC Scalar Parameter Editor Only(
    Material=@n2
    ParamName=Default Mesh Size
    Value=float * float(A=Get Object Radius(), B=2) {is=title K2Node_CommutativeAssociativeBinaryOperator}
   )
   > Set MIC Vector Parameter Editor Only(Material=@n2, ParamName=Pivot Offset, Value="To LinearColor (Vector)"(InVec=Get Offset Vector()))
   > Branch(Condition="Equal (Enum)"(A:EImposterLayoutType Enum=Get Impostor Type() -> (Impostor Type=NewEnumerator0), B:EImposterLayoutType Enum=NewEnumerator2))
   n3 = Get Created MIC()
   > Set MIC Scalar Parameter Editor Only(Material=@n3, ParamName=Reference Plane Sides, Value=Get Reference Plane Sides())
   > Set MIC Scalar Parameter Editor Only(Material=@n3, ParamName=Reference Plane Top, Value=Get Reference Plane Top())
   Outputs!()
  ]}
 Set MIC Switches and ShadingModel {nodes [
   Set MIC Switches and ShadingModel!() -> (then, In MIC:Material Instance Constant Object Reference, Creating Final Assets:bool=~) {is=title K2Node_FunctionEntry}
   > n1 = Sequence()
   > Is Valid(InputObject=Get In MIC() -> (In MIC:Material Instance Constant Object Reference))
   > n2 = For Each Loop(Array:Array of Strings=Get Scalar Switch Params() -> (Scalar Switch Params:Array of Strings)) -> (Array Element:String)
   > Set Current Param(Current Param:String=@n2) -> (Output_Get:String)
   > Set Found Current Param(Found Current Param:bool=false) -> (Output_Get:bool)
   > n3 = For Each Loop(Array:Array of Impostor Channel Pack Settings Structures=Get Channel Packed Masks()) -> (Array Element:Impostor Channel Pack Settings Structure)
   n4 = Branch(execute=@n3.Completed, Condition=Get Found Current Param() -> (Found Current Param:bool))
   n5 = Get Current Param() -> (Current Param:String)
   SetMaterialInstanceStaticSwitchParameterValue!(
    execute=@n4.else
    Instance:Material Instance Constant Object Reference=Get In MIC()
    ParameterName:Name=String To Name(InString=Append!(A=Use, B=@n5, C:String=Texture) -> (ReturnValue))
    Value:bool=false
    Association:EMaterialParameterAssociation Enum=GlobalParameter
    bUpdateMaterialInstance:bool=true
   ) -> (then, ReturnValue:bool)
   > Set Component Mask Via Switches!(In:Exec, MIC:Material Instance Constant Object Reference=Get In MIC(), Prefix:String=Get Current Param(), Idx:int) -> (Out:Exec) {is=title K2Node_MacroInstance}
   n6 = Break Impostor Channel Pack Settings(ImpostorChannelPackSettings=@n3)
   n7 = For Each Loop(
    Exec=@n3.LoopBody
    Array:Array of EImpostorBakeTypes Enums=Make Array!(
     "[0]"=Scalar Map to Bake Type(Index=@n6) -> (ReturnValue=NewEnumerator0)
     "[1]":EImpostorBakeTypes Enum=Scalar Map to Bake Type(Index=@n6.G_11_#2) -> (ReturnValue=NewEnumerator0)
     "[2]":EImpostorBakeTypes Enum=Scalar Map to Bake Type(Index=@n6.B_12_#3) -> (ReturnValue=NewEnumerator0)
     "[3]":EImpostorBakeTypes Enum=Scalar Map to Bake Type(Index=@n6.A_13_#4) -> (ReturnValue=NewEnumerator0)
    ) -> (Array)
   ) -> (Array Element:EImpostorBakeTypes Enum)
   n8 = Enum to String(Enumerator=@n7)
   n9 = Get Current Param()
   > n10 = Set Found Current Param(Found Current Param=OR Boolean!(A=Get Found Current Param(), B="Equal Exactly (String)"(A=@n9, B=@n8)) -> (ReturnValue))
   > Branch(Condition=@n10)
   > SetMaterialInstanceStaticSwitchParameterValue(
    Instance=Get In MIC()
    ParameterName=String To Name(InString=Append(A=Use, B=@n8, C=Texture))
    Value=true
    Association=GlobalParameter
    bUpdateMaterialInstance=true
   )
   > Set Component Mask Via Switches(MIC=Get In MIC(), Prefix=@n8, Idx=@n7.Array Index)
   n11 = Branch(
    execute=@n1.then_2
    Condition="Equal (Enum)"(
     A:EImpostorShadingModelOverride Enum=Get Shading Model Override() -> (Shading Model Override=NewEnumerator0)
     B:EImpostorShadingModelOverride Enum=NewEnumerator0
    )
   )
   K2Node_Composite_0 = Shading Model Selection!(Index:EImpostorShadingModelOverride Enum=Get Shading Model Override() -> (Shading Model Override=NewEnumerator0)) -> (ReturnValue:EMaterialShadingModel Enum=MSM_Unlit) {is=title K2Node_Composite}
   Set MIC Shading Model Editor Only!(
    execute=@n11.else
    Material:Material Instance Constant Object Reference=Get In MIC()
    ShadingModel:EMaterialShadingModel Enum=@K2Node_Composite_0
   ) -> (then, ReturnValue:bool) {is=title K2Node_CallFunction}
   n12 = For Each Loop(Exec=@n1.then_1, Array=Get Color Map Switch Params() -> (Color Map Switch Params:Array of EImpostorBakeTypes Enums=NewEnumerator0))
   n13 = Enum to String(Enumerator=@n12)
   n14 = Get Maps to Save() -> (Maps to Save=NewEnumerator0)
   > SetMaterialInstanceStaticSwitchParameterValue(
    Instance=Get In MIC()
    ParameterName=String To Name(InString=Append(A=Use, B=@n13, C=Texture))
    Value="integer >= integer"(A=Find Item!(TargetArray:Array of EImpostorBakeTypes Enums=@n14, ItemToFind:"EImpostorBakeTypes Enum (by ref)"=@n12) -> (ReturnValue))
    Association=GlobalParameter
    bUpdateMaterialInstance=true
   )
   SetMaterialInstanceStaticSwitchParameterValue(
    execute=@n1.then_3
    Instance=Get In MIC()
    ParameterName=Multi Iteration Parallax
    Value="Equal (Enum)"(A:EImpostorParallaxMode Enum=Get Parallax Mode() -> (Parallax Mode=NewEnumerator0), B:EImpostorParallaxMode Enum=NewEnumerator2)
    Association=GlobalParameter
    bUpdateMaterialInstance=true
   )
   > SetMaterialInstanceStaticSwitchParameterValue(
    Instance=Get In MIC()
    ParameterName=BumpOffset Parallax
    Value="Equal (Enum)"(A=Get Parallax Mode() -> (Parallax Mode=NewEnumerator0), B=NewEnumerator1)
    Association=GlobalParameter
    bUpdateMaterialInstance=true
   )
   n15 = Get Creating Final Assets() -> (Creating Final Assets:bool)
   > SetMaterialInstanceStaticSwitchParameterValue(
    Instance=Get In MIC()
    ParameterName=Show Frame Debug Info
    Value=AND Boolean(A=Get Shader Debug Mode() -> (Shader Debug Mode:bool), B=NOT Boolean(A=@n15))
    Association=GlobalParameter
    bUpdateMaterialInstance=true
   )
   > SetMaterialInstanceStaticSwitchParameterValue(
    Instance=Get In MIC()
    ParameterName=Use Sprite Vertex Shader
    Value=Get Use Sprite Vertex Shader()
    Association=GlobalParameter
    bUpdateMaterialInstance=true
   )
   > UpdateMaterialInstance!(execute, Instance:Material Instance Constant Object Reference=Get In MIC()) -> (then)
  ]}
 Shading Model Selection {
  path=Set MIC Switches and ShadingModel.K2Node_Composite_0.Shading Model Selection
  nodes [Outputs!(ReturnValue:EMaterialShadingModel Enum=Select!(
     NewEnumerator0:EMaterialShadingModel Enum=MSM_DefaultLit
     NewEnumerator1:EMaterialShadingModel Enum=MSM_Unlit
     NewEnumerator2:EMaterialShadingModel Enum=MSM_DefaultLit
     NewEnumerator3:EMaterialShadingModel Enum=MSM_Subsurface
     NewEnumerator4:EMaterialShadingModel Enum=MSM_PreintegratedSkin
     NewEnumerator12:EMaterialShadingModel Enum=MSM_ClearCoat
     NewEnumerator5:EMaterialShadingModel Enum=MSM_SubsurfaceProfile
     NewEnumerator6:EMaterialShadingModel Enum=MSM_TwoSidedFoliage
     NewEnumerator7:EMaterialShadingModel Enum=MSM_Hair
     NewEnumerator8:EMaterialShadingModel Enum=MSM_Unlit
     NewEnumerator9:EMaterialShadingModel Enum=MSM_Eye
     NewEnumerator11:EMaterialShadingModel Enum=MSM_SingleLayerWater
     NewEnumerator10:EMaterialShadingModel Enum=MSM_ThinTranslucent
     Index:EImpostorShadingModelOverride Enum=Inputs!() -> (Index:EImpostorShadingModelOverride Enum=NewEnumerator0)
    ) -> (ReturnValue:EMaterialShadingModel Enum))]
 }
 SetComponentMaskViaSwitches {nodes [
   n1 = Inputs!() -> (In:Exec, MIC:Material Instance Constant Object Reference, Prefix:String, Idx:int)
   > n2 = For Each Loop(Array:Array of Strings=Make Array("[0]":String=R, "[1]":String=G, "[2]":String=B, "[3]":String=A) -> (Array:Array of Strings)) -> (Array Element:String)
   Outputs!(Out:Exec=@n2.Completed)
   > SetMaterialInstanceStaticSwitchParameterValue(
    Instance=@n1
    ParameterName=String To Name(InString=Append(A=@n1.Prefix, B=_Channel_, C=@n2))
    Value="Equal (Integer)"(A=@n2.Array Index, B=@n1.Idx)
    Association=GlobalParameter
    bUpdateMaterialInstance=true
   )
  ]}
 Setup Octahedron Layout {nodes [
   Setup Octahedron Layout() {is=title K2Node_FunctionEntry}
   > Clear(TargetArray:Array of Vectors=Get ViewCaptureVectors())
   > Clear(TargetArray=Get LightVectors())
   n1 = Get Impostor Type() -> (Impostor Type=NewEnumerator0)
   > Set FramesXYInternal(FramesXYInternal:int=Select!(
     Option 0:int=Get Frames XY() -> (Frames XY=0)
     Option 1:int=3
     Index:bool="Equal (Enum)"(A:EImposterLayoutType Enum=@n1, B:EImposterLayoutType Enum=NewEnumerator2)
    ) -> (ReturnValue:int)) -> (Output_Get:int=0)
   > n2 = Sequence!(execute) -> (then_0, then_1)
   > n3 = For Loop(LastIndex=int - int(A=Get FramesXYInternal() -> (FramesXYInternal=0), B=1) {is=title K2Node_CallFunction})
   > Set CurrentY(CurrentY:int=@n3) -> (Output_Get:int=0)
   > n4 = For Loop(LastIndex=int - int(A=Get FramesXYInternal() -> (FramesXYInternal=0), B=1))
   > Set CurrentX(CurrentX:int=@n4) -> (Output_Get:int=0)
   n5 = "To Float (Integer)"(InInt=Get CurrentX() -> (CurrentX:int=0))
   n6 = float - float(A="To Float (Integer)"(InInt=Get FramesXYInternal() -> (FramesXYInternal=0)), B=1) {is=title K2Node_CallFunction}
   n7 = "To Float (Integer)"(InInt=Get CurrentY() -> (CurrentY:int=0))
   n8 = Impostor Grid Vector!(0 to 1:Vector 2D Structure=MakeVector2D(X=float / float(A=@n5, B=@n6) {is=title K2Node_CallFunction}, Y=float / float(A=@n7, B=@n6))) -> (ReturnValue:Vector=~) {is=title K2Node_MacroInstance}
   > Add(TargetArray:Array of Vectors=Get ViewCaptureVectors(), NewItem:"Vector (by ref)"=@n8)
   > Branch(Condition=Get Show Debug View Vectors() -> (Show Debug View Vectors:bool))
   n9 = float * float(A=float * float(A=Get Object Radius(), B=1.1), B=1.1)
   > DrawDebugPoint!(
    execute
    WorldContextObject:Object Reference
    Position:Vector=vector + vector(A:Vector=vector * float(A=@n8, B=@n9), B:Vector=Get Offset Vector()) -> (ReturnValue:Vector) {is=title K2Node_CommutativeAssociativeBinaryOperator}
    Size:float=0
    PointColor:Linear Color Structure
    Duration:float=0
    DepthPriority:EDrawDebugSceneDepthPriorityGroup Enum=World
   ) -> (then)
   Branch(
    execute=@n2.then_1
    Condition=Contains Item(TargetArray=Get Color Maps to Render() -> (Color Maps to Render=NewEnumerator0), ItemToFind=NewEnumerator10)
   )
   > n10 = For Loop(LastIndex=int - int(A=Get Lighting Grid XY() -> (Lighting Grid XY=0), B=1))
   > Set CurrentY(CurrentY=@n10) -> (Output_Get=0)
   > n11 = For Loop(LastIndex=int - int(A=Get Lighting Grid XY() -> (Lighting Grid XY=0), B=1))
   > Set CurrentX(CurrentX=@n11) -> (Output_Get=0)
   n12 = "To Float (Integer)"(InInt=Get CurrentX() -> (CurrentX=0))
   n13 = float - float(A="To Float (Integer)"(InInt=Get Lighting Grid XY() -> (Lighting Grid XY=0)), B=1)
   n14 = "To Float (Integer)"(InInt=Get CurrentY() -> (CurrentY=0))
   n15 = Impostor Grid Vector(0 to 1=MakeVector2D(X=float / float(A=@n12, B=@n13), Y=float / float(A=@n14, B=@n13))) -> (ReturnValue=~)
   n16 = Get Upward Bias()
   > Add(
    TargetArray=Get LightVectors()
    NewItem=Normalize(A="Lerp (Vector)"!(A:Vector=@n15, B:Vector="0,0,1", Alpha:float=@n16) -> (ReturnValue:Vector), Tolerance=0.0001)
   )
  ]}
 Setup RT and Save List {nodes [
   Setup RT and Save List() {is=title K2Node_FunctionEntry}
   > Set Atlas_Res_Per_Frame(Atlas_Res_Per_Frame:int=int / int(A=Get Resolution() -> (Resolution=0), B=Get Frames XY() -> (Frames XY=0))) -> (Output_Get:int=0)
   > Set Scene Capture Resolution!(
    execute
    Scene Capture Resolution="Max (Integer)"(A=Get Scene Capture Resolution() -> (Scene Capture Resolution=0), B=Get Atlas_Res_Per_Frame() -> (Atlas_Res_Per_Frame:int=0))
   ) -> (then, Output_Get=0)
   n1 = Log!(A:double="To Float (Integer)"(InInt=Get Resolution() -> (Resolution=0)), Base:double=2) -> (ReturnValue:double)
   n2 = "To Float (Integer)"(InInt="Clamp (Integer)"(Value=Floor(A=@n1) {is=title K2Node_CallFunction}, Min=9, Max=14) -> (ReturnValue=0))
   > Set Resolution!(execute, Resolution=Truncate(A=Power!(Base:double=2, Exp:double=@n2) -> (ReturnValue:double) {is=title K2Node_CallFunction})) -> (then, Output_Get=0)
   n3 = "To Float (Integer)"(InInt=Get Scene Capture Resolution() -> (Scene Capture Resolution=0))
   n4 = "Clamp (Integer)"(Value=Floor(A=Log(A=@n3, Base=2)), Min=7, Max=11) -> (ReturnValue=0)
   n5 = "To Float (Integer)"(InInt=@n4)
   > Set Scene Capture Resolution(Scene Capture Resolution=Truncate(A=Power(Base=2, Exp=@n5))) -> (Output_Get=0)
   > Set SceneCaptureMips(SceneCaptureMips:int=@n4) -> (Output_Get:int=0)
   n6 = int - int(A=@n4, B=4)
   > Set CutoutMipTarget(CutoutMipTarget:int=@n6) -> (Output_Get:int=0)
   > Set DFMipTarget(DFMipTarget:int=int + int(A=@n6, B=2) -> (ReturnValue=0) {is=title K2Node_CommutativeAssociativeBinaryOperator}) -> (Output_Get:int=0)
   n7 = Log(A="To Float (Integer)"(InInt=Get Atlas_Res_Per_Frame() -> (Atlas_Res_Per_Frame=0)), Base=2)
   n8 = Log(A="To Float (Integer)"(InInt=Get Scene Capture Resolution() -> (Scene Capture Resolution=0)), Base=2)
   > Set SceneCaptureMipToComposite(SceneCaptureMipToComposite:int=int - int(A=Floor(A=@n8), B=Floor(A=@n7)) {is=title K2Node_PromotableOperator}) -> (Output_Get:int=0)
   > Combine Render Types!(Exec:Exec) -> (Completed:Exec) {is=title K2Node_MacroInstance}
   > n9 = Sequence!(execute) -> (then_0, then_1, then_2:Exec, then_3:Exec, then_4:Exec)
   > n10 = For Each Loop(Array:Array of EImpostorBakeTypes Enums=Get Maps to Render_Internal() -> (Maps to Render_Internal=NewEnumerator0)) -> (Array Element:EImpostorBakeTypes Enum)
   n11 = Keys(
    execute=@n10.Completed
    TargetMap:Map of EImpostorBakeTypes Enums to Texture Render Target 2D Object References=Get TargetMaps() -> (TargetMaps=NewEnumerator0)
   ) -> (Keys:Array of EImpostorBakeTypes Enums)
   > n12 = For Each Loop(Array=@n11)
   n13 = Get Maps to Render_Internal() -> (Maps to Render_Internal=NewEnumerator0)
   > n14 = Branch(Condition="integer >= integer"!(A:int=Find Item(TargetArray=@n13, ItemToFind=@n12), B:int=0) -> (ReturnValue:bool) {is=title K2Node_CallFunction})
   Remove!(
    execute=@n14.else
    TargetMap:Map of EImpostorBakeTypes Enums to Texture Render Target 2D Object References=Get TargetMaps() -> (TargetMaps=NewEnumerator0)
    Key:"EImpostorBakeTypes Enum (by ref)"=@n12
   ) -> (then, ReturnValue:bool)
   Set CurMapType(execute=@n10.LoopBody, CurMapType:EImpostorBakeTypes Enum=@n10) -> (Output_Get:EImpostorBakeTypes Enum=NewEnumerator0)
   n15 = Find(TargetMap=Get TargetMaps() -> (TargetMaps=NewEnumerator0), Key=Get CurMapType() -> (CurMapType:EImpostorBakeTypes Enum=NewEnumerator0))
   > n16 = Branch(Condition=@n15.ReturnValue)
   > n17 = Set TempRT(TempRT:Texture Render Target 2D Object Reference=@n15) -> (Output_Get:Texture Render Target 2D Object Reference)
   K2Node_Composite_2 = Select RT Format Based on Capture Type!() -> (ReturnValue:ETextureRenderTargetFormat Enum=RTF_R8) {is=title K2Node_Composite}
   n18 = Get Resolution() -> (Resolution=0)
   Check And Create RT!(
    In:Exec=[@n17.then, @n19.then]
    RT:Texture Render Target 2D Object Reference=Get TempRT() -> (TempRT:Texture Render Target 2D Object Reference)
    Width:int=@n18
    Height:int=@n18
    Format:ETextureRenderTargetFormat Enum=@K2Node_Composite_2
   ) -> (Out:Exec) {is=title K2Node_MacroInstance}
   > Add!(
    execute
    TargetMap:Map of EImpostorBakeTypes Enums to Texture Render Target 2D Object References=Get TargetMaps() -> (TargetMaps=NewEnumerator0)
    Key:"EImpostorBakeTypes Enum (by ref)"=Get CurMapType() -> (CurMapType=NewEnumerator0)
    Value:"Texture Render Target 2D Object Reference (by ref)"=Get TempRT()
   ) -> (then) {is=title K2Node_CallFunction}
   n19 = Set TempRT(execute=@n16.else)
   n20 = For Each Loop(Exec=@n9.then_1, Array:Array of Impostor Channel Pack Settings Structures=Get Channel Packed Masks()) -> (Array Element:Impostor Channel Pack Settings Structure)
   n21 = Keys(
    execute=@n20.Completed
    TargetMap:Map of Impostor Channel Pack Settings Structures to Texture Render Target 2D Object References=Get ChannelPackedMaps()
   ) -> (Keys:Array of Impostor Channel Pack Settings Structures)
   > n22 = For Each Loop(Array=@n21)
   n23 = Get Channel Packed Masks()
   > n24 = Branch(Condition="integer >= integer"(
     A=Find Item(TargetArray:Array of Impostor Channel Pack Settings Structures=@n23, ItemToFind:"Impostor Channel Pack Settings Structure (by ref)"=@n22)
     B=0
    ))
   Remove(
    execute=@n24.else
    TargetMap:Map of Impostor Channel Pack Settings Structures to Texture Render Target 2D Object References=Get ChannelPackedMaps()
    Key:"Impostor Channel Pack Settings Structure (by ref)"=@n22
   )
   n25 = Find(
    TargetMap:Map of Impostor Channel Pack Settings Structures to Texture Render Target 2D Object References=Get ChannelPackedMaps()
    Key:"Impostor Channel Pack Settings Structure (by ref)"=@n20
   )
   n26 = Branch(execute=@n20.LoopBody, Condition=@n25.ReturnValue)
   > n27 = Set TempRT(TempRT=@n25)
   n28 = Get Resolution() -> (Resolution=0)
   Check And Create RT(In=[@n29.then, @n27.then], RT=Get TempRT(), Width=@n28, Height=@n28, Format=RTF_RGBA8)
   > Add(
    TargetMap:Map of Impostor Channel Pack Settings Structures to Texture Render Target 2D Object References=Get ChannelPackedMaps()
    Key:"Impostor Channel Pack Settings Structure (by ref)"=@n20
    Value=Get TempRT()
   )
   n29 = Set TempRT(execute=@n26.else)
   n30 = For Loop(execute=@n9.then_2, LastIndex=int - int(A=Get SceneCaptureMips() -> (SceneCaptureMips:int=0)))
   > n31 = Set MipIdx(MipIdx:int=@n30) -> (Output_Get:int=0)
   n32 = Get SceneCaptureMipChain()
   > n33 = Branch(Condition="integer <= integer"!(A:int=@n31, B:int=Last Index(TargetArray:Array of Texture Render Target 2D Object References=@n32) -> (ReturnValue=0)) -> (ReturnValue:bool) {is=title K2Node_CallFunction})
   > n34 = Set TempRT(TempRT="Get (a copy)"(Array=Get SceneCaptureMipChain(), Dimension 1=Get MipIdx() -> (MipIdx:int=0)))
   n35 = "To Float (Integer)"(InInt=int + int(A=Get MipIdx() -> (MipIdx=0), B=0) -> (ReturnValue=0))
   n36 = int / int!(A:int=Get Scene Capture Resolution() -> (Scene Capture Resolution=0), B:int=Truncate(A=Power(Base=2, Exp=@n35))) -> (ReturnValue:int=0) {is=title K2Node_CallFunction}
   Check And Create RT(In=[@n34.then, @n37.then], RT=Get TempRT(), Width=@n36, Height=@n36, Format=RTF_RGBA16f)
   > Set Array Elem!(
    execute
    TargetArray:Array of Texture Render Target 2D Object References=Get SceneCaptureMipChain()
    Index:int=Get MipIdx() -> (MipIdx=0)
    Item:"Texture Render Target 2D Object Reference (by ref)"=Get TempRT()
    bSizeToFit:bool=true
   ) -> (then)
   n37 = Set TempRT(execute=@n33.else)
   n38 = MakeLiteralInt!(Value:int=16) -> (ReturnValue:int=0)
   n39 = Select!(
    NewEnumerator0:"Integer (by ref)"=@n38
    NewEnumerator1:"Integer (by ref)"=@n38
    NewEnumerator2:"Integer (by ref)"=MakeLiteralInt(Value=48) -> (ReturnValue=0)
    Index:EImposterLayoutType Enum=Get Impostor Type() -> (Impostor Type=NewEnumerator0)
   ) -> (ReturnValue:"Integer (by ref)")
   Check And Create RT(In=@n9.then_3, RT=Get CombinedAlphasRT(), Width=@n39, Height=@n39, Format=RTF_RGBA8)
   n40 = Get Resolution() -> (Resolution=0)
   > Check And Create RT(RT=Get ScratchRT(), Width=@n40, Height=@n40, Format=RTF_RGBA16f)
   n41 = Get SceneCaptureMipChain()
   n42 = "Min (Integer)"(A=Last Index(TargetArray=@n41) -> (ReturnValue=0), B=Get CutoutMipTarget() -> (CutoutMipTarget=0)) -> (ReturnValue=0)
   > Set CheckTargetMipSize(CheckTargetMipSize:int=Get SizeX(self="Get (a copy)"(Array=@n41, Dimension 1=@n42)) -> (SizeX=0)) -> (Output_Get:int=0)
   Set Maps to Save(execute=@n9.then_4, Maps to Save:Array of EImpostorBakeTypes Enums=Get Maps to Render_Internal() -> (Maps to Render_Internal=NewEnumerator0)) -> (Output_Get:Array of EImpostorBakeTypes Enums=NewEnumerator0)
   > n43 = Sequence!(execute) -> (then_0, then_1, then_2)
   > Branch(Condition=Get Apply to BaseColor())
   > Branch(Condition=Contains Item(
     TargetArray:Array of EImpostorBakeTypes Enums=Get Maps to Render_Internal() -> (Maps to Render_Internal=NewEnumerator0)
     ItemToFind:"EImpostorBakeTypes Enum (by ref)"=NewEnumerator0
    ))
   > Remove Item(TargetArray=Get Maps to Save() -> (Maps to Save=NewEnumerator0), Item=NewEnumerator10)
   n44 = For Each Loop(Exec=@n43.then_2, Array=Get Channel Packed Masks())
   n45 = Break Impostor Channel Pack Settings(ImpostorChannelPackSettings=@n44)
   n46 = Get Maps to Save() -> (Maps to Save=NewEnumerator0)
   > Remove Item(TargetArray=@n46, Item=Scalar Map to Bake Type(Index=@n45) -> (ReturnValue=NewEnumerator0))
   > Remove Item(TargetArray=@n46, Item=Scalar Map to Bake Type(Index=@n45.G_11_#2) -> (ReturnValue=NewEnumerator0))
   n47 = Get Maps to Save() -> (Maps to Save=NewEnumerator0)
   > Remove Item(TargetArray=@n47, Item=Scalar Map to Bake Type(Index=@n45.B_12_#3) -> (ReturnValue=NewEnumerator0))
   > Remove Item(TargetArray=@n47, Item=Scalar Map to Bake Type(Index=@n45.A_13_#4) -> (ReturnValue=NewEnumerator0))
   Branch(
    execute=@n43.then_1
    Condition=Contains Item(TargetArray=Get Maps to Render_Internal() -> (Maps to Render_Internal=NewEnumerator0), ItemToFind=NewEnumerator6)
   )
   > Remove Item(TargetArray=Get Maps to Save() -> (Maps to Save=NewEnumerator0), Item=NewEnumerator8)
  ]}
 Select RT Format Based on Capture Type {
  path=Setup RT and Save List.K2Node_Composite_2.Select RT Format Based on Capture Type
  nodes [
   Inputs!()
   n1 = "Equal (Enum)"(A:EImpostorBakeTypes Enum=Get CurMapType() -> (CurMapType=NewEnumerator0), B:EImpostorBakeTypes Enum=NewEnumerator8)
   n2 = "Equal (Enum)"(A=Get CurMapType() -> (CurMapType=NewEnumerator0), B=NewEnumerator0)
   n3 = "Equal (Enum)"(A=Get CurMapType() -> (CurMapType=NewEnumerator0), B=NewEnumerator4)
   n4 = "Equal (Enum)"(A=Get CurMapType() -> (CurMapType=NewEnumerator0), B=NewEnumerator6)
   n5 = "Equal (Enum)"(A=Get CurMapType() -> (CurMapType=NewEnumerator0), B=NewEnumerator15)
   n6 = "Equal (Enum)"(A=Get CurMapType() -> (CurMapType=NewEnumerator0), B=NewEnumerator7)
   n7 = "Equal (Enum)"(A=Get CurMapType() -> (CurMapType=NewEnumerator0), B=NewEnumerator9)
   n8 = "Equal (Enum)"(A=Get CurMapType() -> (CurMapType=NewEnumerator0), B=NewEnumerator10)
   Outputs!(ReturnValue:ETextureRenderTargetFormat Enum=Select!(
     Option 0:ETextureRenderTargetFormat Enum=RTF_R8
     Option 1:ETextureRenderTargetFormat Enum=RTF_RGBA16f
     Index:bool=OR Boolean!(A=@n2, B=@n3, C:bool=@n4, D:bool=@n5, E:bool=@n6, F:bool=@n7, G:bool=@n8, H:bool=@n1) -> (ReturnValue)
    ) -> (ReturnValue:ETextureRenderTargetFormat Enum))
  ]
 }
 Update Material Settings DetailsView {nodes [
   Inputs!() -> (execute:Exec)
   n1 = Get MaterialSettings_To_Show() -> (MaterialSettings_To_Show:Array of Names)
   > Clear(TargetArray:Array of Names=@n1)
   > Add(
    TargetArray:Array of Names=@n1
    NewItem:"Name (by ref)"=Select!(
     NewEnumerator0:"Name (by ref)"=MakeLiteralName(Value=Full Sphere Material)
     NewEnumerator1:"Name (by ref)"=MakeLiteralName(Value=Hemisphere Material)
     NewEnumerator2:"Name (by ref)"=MakeLiteralName(Value=Billboard Material)
     Index:EImposterLayoutType Enum=Get Impostor Type() -> (Impostor Type=NewEnumerator0)
    ) -> (ReturnValue:"Name (by ref)")
   ) {is=title K2Node_CallArrayFunction}
   n2 = Get MaterialSettings_To_Show()
   > Append Array!(
    execute
    TargetArray:Array of Names=@n2
    SourceArray:Array of Names=Make Array!(
     "[0]":Name=Constant Specular
     "[1]":Name=Constant Roughness
     "[2]":Name=Constant Opacity
     "[3]":Name=Mask Offset
     "[4]":Name=Subsurface Color
     "[5]":Name=Scatter Mask Min
     "[6]":Name=Scatter Mask Length
    ) -> (Array:Array of Names)
   ) -> (then)
   > Append Array(TargetArray=@n2, SourceArray=Make Array!("[0]"=Shading Model Override, "[1]"=Parallax Mode) -> (Array))
   > n3 = Branch(Condition="Equal (Enum)"(A:EImposterLayoutType Enum=Get Impostor Type() -> (Impostor Type=NewEnumerator0), B:EImposterLayoutType Enum=NewEnumerator2))
   > n4 = Append Array(
    TargetArray=Get MaterialSettings_To_Show()
    SourceArray=Make Array!("[0]"=Center XY on Mesh Pivot, "[1]"=Reference Plane Top, "[2]":Name=Reference Plane Sides, "[3]":Name=Use Sprite Vertex Shader) -> (Array)
   )
   n5 = For Each Loop(Exec=[@n4.then, @n3.else], Array=Get Channel Packed Masks())
   SetObject(execute=@n5.Completed, self=Get Details_MaterialSettings())
   > SetEditorProperty(Object=Get Details_MaterialSettings(), PropertyName=PropertiesToShow, PropertyValue:Array of Names=Get MaterialSettings_To_Show())
   > SetObject(self=Get Details_MaterialSettings(), NewObject=Self-Reference())
   > Outputs!(Completed:Exec)
   n6 = Break Impostor Channel Pack Settings(ImpostorChannelPackSettings=@n5)
   n7 = For Each Loop(
    Exec=@n5.LoopBody
    Array:Array of EImpostorScalarMaps Enums=Make Array(
     "[0]":EImpostorScalarMaps Enum=@n6
     "[1]":EImpostorScalarMaps Enum=@n6.G_11_#2
     "[2]":EImpostorScalarMaps Enum=@n6.B_12_#3
     "[3]":EImpostorScalarMaps Enum=@n6.A_13_#4
    ) -> (Array:Array of EImpostorScalarMaps Enums)
   ) -> (Array Element:EImpostorScalarMaps Enum)
   > n8 = Branch(Condition="Equal (Enum)"(A:EImpostorScalarMaps Enum=@n7, B:EImpostorScalarMaps Enum=NewEnumerator2))
   > Remove Item(TargetArray:Array of Names=Get MaterialSettings_To_Show(), Item:"Name (by ref)"=Constant Specular)
   n9 = Branch(execute=@n8.else, Condition="Equal (Enum)"(A=@n7, B=NewEnumerator3))
   > Remove Item(TargetArray=Get MaterialSettings_To_Show(), Item=Constant Roughness)
   Branch(execute=@n9.else, Condition="Equal (Enum)"(A=@n7, B=NewEnumerator5))
   > Remove Item(TargetArray=Get MaterialSettings_To_Show(), Item=Constant Opacity)
  ]}
 Update Preview Mode And Location {nodes [
   n1 = Inputs!() -> (execute, Selection:String)
   > n2 = Switch on String!(execute:Exec, Selection:String=@n1) -> (Default:Exec, Impostor + Original Side by Side:Exec, Swap by Distance:Exec, Impostor Only:Exec) {pinNames [Impostor + Original Side by Side, Swap by Distance, Impostor Only]}
   Set Text(
    execute=@n2.Impostor + Original Side by Side
    Text="LOCTEXT(\"Displaying: Source Mesh and Impostor\")"
    self:Text Object Reference=Get Text_DisplayingWhat()
   )
   > SetVisibility!(execute, self:Scene Component Object Reference=Get StaticMeshComponent(), bNewVisibility:bool=true, bPropagateToChildren:bool=false) -> (then)
   > SetVisibility(self=Get DynamicMeshComponent(), bNewVisibility=true, bPropagateToChildren=false)
   > Set SwapByDistance(SwapByDistance:bool=false) -> (Output_Get:bool)
   n3 = float * float(A=Get Object Radius(), B=1) {is=title K2Node_PromotableOperator}
   n4 = Get Offset Vector()
   > Set World Location(
    self=Get StaticMeshComponent()
    NewLocation=vector + vector(A=vector * vector(A=@n4, B:float=-1) -> (ReturnValue=~), B=MakeVector(X=@n3, Y=0)) {is=title K2Node_PromotableOperator}
    bSweep=false
    bTeleport=false
   )
   n5 = float * float(A=Get Object Radius(), B=-1)
   n6 = vector * vector(A=Get Offset Vector(), B=-1) -> (ReturnValue=~)
   > Set World Location(self=Get DynamicMeshComponent(), NewLocation=vector + vector(A=MakeVector(X=@n5, Y=0), B=@n6), bSweep=false, bTeleport=false)
   Set Text(execute=@n2.Impostor Only, Text="LOCTEXT(\"Displaying: Impostor\")", self=Get Text_DisplayingWhat())
   > SetVisibility(self=Get StaticMeshComponent(), bNewVisibility=false, bPropagateToChildren=false)
   > SetVisibility(self=Get DynamicMeshComponent(), bNewVisibility=true, bPropagateToChildren=false)
   > n7 = Set SwapByDistance(SwapByDistance=false)
   Set World Location(
    execute=[@n7.then, @n8.then]
    self=Get StaticMeshComponent()
    NewLocation=vector * vector(A=Get Offset Vector(), B=-1) -> (ReturnValue=~)
    bSweep=false
    bTeleport=false
   )
   > Set World Location(self=Get DynamicMeshComponent(), NewLocation=@n6, bSweep=false, bTeleport=false)
   n8 = Set SwapByDistance(execute=@n2.Swap by Distance, SwapByDistance=true)
   Outputs!()
  ]}
 Update Progress {nodes [
   Inputs!() -> (execute)
   > SetVisibility!(execute, self:Widget Object Reference=Get VerticalBox_GeneratingImpostor(), InVisibility:ESlateVisibility Enum=SelfHitTestInvisible) -> (then)
   n1 = Get Maps to Render_Internal() -> (Maps to Render_Internal=NewEnumerator0)
   n2 = "To Float (Integer)"(InInt=Get CurrentMapIndex() -> (CurrentMapIndex=0))
   > Set Progress(Progress=float / float(A=float + float(A=@n2), B:int=Length(TargetArray:Array of EImpostorBakeTypes Enums=@n1)) {is=title K2Node_PromotableOperator})
   n3 = Get CurrentMapIndex() -> (CurrentMapIndex=0)
   n4 = "To String (Integer)"(InInt=int + int(A=@n3) {is=title K2Node_PromotableOperator})
   n5 = "To String (Integer)"(InInt=Length(TargetArray=Get Maps to Render_Internal() -> (Maps to Render_Internal=NewEnumerator0)))
   n6 = Enum to String(Enumerator="Get (a ref)"(Array:Array of EImpostorBakeTypes Enums=Get Maps to Render_Internal() -> (Maps to Render_Internal=NewEnumerator0), Dimension 1=@n3) -> (Output:"EImpostorBakeTypes Enum (by ref)"))
   > Set Text(
    Text="To Text (String)"(InString=Append!(A="Capturing ", B=@n6, C=" (", D:String=@n4, E:String=/, F:String=@n5, G:String=")") -> (ReturnValue))
    self=Get Progress_CurrentMap() -> (Progress_CurrentMap:Text Object Reference)
   )
   > Outputs!(then:Exec)
  ]}
 UpdateCombinedCustomLighting {nodes [
   UpdateCombinedCustomLighting() {is=title K2Node_FunctionEntry}
   n1 = "integer >= integer"(
    A=Find Item(
     TargetArray:Array of EImpostorBakeTypes Enums=Get Maps to Render_Internal() -> (Maps to Render_Internal=NewEnumerator0)
     ItemToFind:"EImpostorBakeTypes Enum (by ref)"=NewEnumerator0
    )
    B=0
   )
   n2 = "integer >= integer"(A=Find Item(TargetArray=Get Maps to Render_Internal() -> (Maps to Render_Internal=NewEnumerator0), ItemToFind=NewEnumerator10), B=0)
   > Branch(Condition=AND Boolean(A=Get Apply to BaseColor(), B=AND Boolean(A=@n1, B=@n2)))
   > ClearRenderTarget2D(
    TextureRenderTarget=Find(
     TargetMap:Map of EImpostorBakeTypes Enums to Texture Render Target 2D Object References=Get TargetMaps() -> (TargetMaps=NewEnumerator0)
     Key:"EImpostorBakeTypes Enum (by ref)"=NewEnumerator0
    )
    ClearColor=(R=0,G=0,B=0,A=1)
   )
   > DrawMaterialToRenderTarget(
    TextureRenderTarget=Find(TargetMap=Get TargetMaps() -> (TargetMaps=NewEnumerator0), Key=NewEnumerator0)
    Material=Get BaseColorCustomLightingMID()
   )
  ]}
 UpdatePreviewMaterialProxy {nodes [
   UpdatePreviewMaterialProxy() {is=title K2Node_FunctionEntry}
   > n1 = Sequence!(execute) -> (then_0, then_1, then_2, then_3:Exec)
   > n2 = Set PreviewMaterial(PreviewMaterial:Material Interface Object Reference=Select(
     NewEnumerator0:Material Interface Object Reference=Get Full Sphere Material()
     NewEnumerator1:Material Interface Object Reference=Get Hemisphere Material()
     NewEnumerator2:Material Interface Object Reference=Get Billboard Material()
     Index=Get Impostor Type() -> (Impostor Type=NewEnumerator0)
    ) -> (ReturnValue:Material Interface Object Reference)) -> (Output_Get:Material Interface Object Reference)
   n3 = Is Path Valid Soft Obj(PathString=Object Path Add Outer(Path=/ImpostorBaker/Materials/Internal/ImpostorMaterialPreviewProxy)) -> (ReturnValue=~)
   > n4 = Branch(Condition=@n3)
   > n5 = Cast To MaterialInterface!(execute, Object=@n3.Object) -> (then, CastFailed, AsMaterial Interface:Material Interface Object Reference) {is=title K2Node_DynamicCast}
   > Set CreatedMaterialAsset(CreatedMaterialAsset:Material Interface Object Reference=@n5) -> (Output_Get:Material Interface Object Reference)
   > n6 = Cast To MaterialInstanceConstant(Object=Get CreatedMaterialAsset() -> (CreatedMaterialAsset:Material Interface Object Reference))
   > n7 = Set MaterialPreviewProxy(MaterialPreviewProxy:Material Instance Constant Object Reference=@n6) -> (Output_Get:Material Instance Constant Object Reference)
   > SetMaterialInstanceParent(Instance=@n7, NewParent=Get PreviewMaterial() -> (PreviewMaterial:Material Interface Object Reference))
   PrintString(execute=@n5.CastFailed, InString=Impostor Baker: Failed to create preview MIC.)
   n8 = Create MIC Editor Only(execute=@n4.else, Material=@n2, Name=&ImpostorMaterialPreviewProxy)
   > Set CreatedMaterialAsset(CreatedMaterialAsset=@n8)
   > n9 = Cast To MaterialInstanceConstant(Object=Get CreatedMaterialAsset())
   > Set MaterialPreviewProxy(MaterialPreviewProxy=@n9)
   Set MIC Switches and ShadingModel(execute=@n1.then_1, In MIC=Get MaterialPreviewProxy(), Creating Final Assets=false) {is=title K2Node_CallFunction}
   n10 = Get MaterialPreviewProxy()
   n11 = Get Impostor Type() -> (Impostor Type=NewEnumerator0)
   n12 = Get Full Sphere Material()
   n13 = Get Billboard Material()
   n14 = Get Hemisphere Material()
   Set Material Preview Parent(
    execute=@n1.then_3
    Material Preview Parent:Material Interface Object Reference=Select!(
     Option 0:Material Interface Object Reference=Select!(
      NewEnumerator0:Material Interface Object Reference=@n12
      NewEnumerator1:Material Interface Object Reference=@n14
      NewEnumerator2:Material Interface Object Reference=@n13
      Index:EImposterLayoutType Enum=@n11
     ) -> (ReturnValue)
     Option 1:Material Interface Object Reference=Get MaterialPreviewProxy()
     Index:bool=IsValid(Object=@n10)
    ) -> (ReturnValue)
   ) -> (Output_Get:Material Interface Object Reference)
  ]}
 Wireframe Display Mode {nodes [
   n1 = Inputs!() -> (execute, A:Name)
   > n2 = Branch(Condition="Equal (Name)"(A=@n1, B=Wireframe))
   > SetVisibility!(
    execute
    self:Scene Component Object Reference=Get Preview Floor(self=Get Impostor Scene())
    bNewVisibility:bool=false
    bPropagateToChildren:bool=false
   ) -> (then)
   > SetMaterial!(
    execute
    self:Primitive Component Object Reference=Get DynamicMeshComponent()
    ElementIndex:int=0
    Material:Material Interface Object Reference
   ) -> (then)
   n3 = Get StaticMeshComponent()
   > n4 = For Loop(LastIndex=int - int(A=GetNumMaterials!(self:Primitive Component Object Reference=@n3) -> (ReturnValue:int=0)))
   > SetMaterial(self=Get StaticMeshComponent(), ElementIndex=@n4)
   SetVisibility(
    execute=@n2.else
    self=Get Preview Floor(self=Get Impostor Scene())
    bNewVisibility=Get Show Floor in Preview Scene()
    bPropagateToChildren=false
   )
   > SetMaterial(self=Get DynamicMeshComponent(), ElementIndex=0, Material=Get ImpostorPreviewMID())
   n5 = Get StaticMeshComponent()
   > n6 = For Loop(LastIndex=int - int(A=GetNumMaterials(self=@n5) -> (ReturnValue=0)))
   > SetMaterial(self=Get StaticMeshComponent(), ElementIndex=@n6)
   Outputs!()
  ]}
 setAsset {nodes [
   n1 = setAsset!() -> (then, Object:Object Reference) {is=title K2Node_FunctionEntry}
   > n2 = Cast To StaticMesh(Object=@n1)
   > Set Static Mesh(Static Mesh:Static Mesh Object Reference=@n2) -> (Output_Get:Static Mesh Object Reference)
   > Set Capture Mesh and Size Info(ReAdjustCamera=true) {is=title K2Node_CallFunction}
  ]}
]
variables [
 AddAlphasMID {type=MaterialInstanceDynamic}
 Apply to BaseColor {type=boolean}
 asdf {type=EImpostorBakeTypes, default=BaseColor}
 AssetMeshMID {type=MaterialInstanceDynamic}
 Atlas_Res_Per_Frame {type=integer}
 BakingDepth-to-ScalarMaps {type=boolean, default=true}
 BaseColorCustomLightingMID {type=MaterialInstanceDynamic}
 Billboard Material {type=MaterialInterface, default=/ImpostorBaker/Materials/MI_Billboard_01_Inst}
 Camera Distance {type=number, default=1000}
 CaptureTypes {type=EImpostorBakeTypes, default=BaseColor}
 CaptureUsingGBuffer {type=boolean, default=true}
 Center XY on Mesh Pivot {type=boolean}
 ChangedProperty {type=string, default=None}
 Channel Packed Masks {type="array<ImpostorChannelPackSettings>", default []}
 ChannelPackedMaps {default {}}
 ChannelPackMasksMID {type=MaterialInstanceDynamic}
 CheckTargetMipSize {type=integer}
 Color Maps to Render {type="array<EImpostorColorMaps>", default [BaseColor,Normal]}
 CombinedAlphasRT {type=TextureRenderTarget2D, default=null}
 CombineNormalsDepthMID {type=MaterialInstanceDynamic}
 Constant Opacity {type=number, default=0.5}
 Constant Roughness {type=number, default=0.5}
 Constant Specular {type=number, default=0.5}
 Created MIC {type=MaterialInstanceConstant, default=null}
 curframe {type=integer}
 CurrentCameraDistance {type=number}
 CurrentMap {type=EImpostorBakeTypes, default=BaseColor}
 CurrentMapIndex {type=integer}
 CurrentRemovalIndex {type=integer}
 CurrentTargetRT {type=TextureRenderTarget2D, default=null}
 CurrentX {type=integer}
 CurrentY {type=integer}
 Custom Floats {type="map<number>", default {}}
 Custom Lighting Opacity {type=number, default=1}
 Custom Lighting Power {type=number, default=1}
 Custom SkyLight Intensity {type=number, default=1}
 Custom Textures {type="map<Texture2D>", default {}}
 Custom Vectors {type="map<LinearColor>", default {}}
 CutoutMipTarget {type=integer}
 DefaultSettingsAssetPath {type=string, default=""}
 DFMipTarget {type=integer}
 DirLightBrightness {type=number, default=3.14}
 Dynamic Mesh {type=DynamicMesh, default=null}
 DynamicMeshComponent {type=DynamicMeshComponent, default=null}
 DynamicMeshTris {type="array<integer>"}
 Enable Asset Material Overrides {type=boolean}
 ExistingImpostorFoundBehavior {type=EImpostorOverwriteLOD, default=Present Dialogue}
 Frames XY {type=integer, default=16}
 FramesXYInternal {type=integer, default=16}
 Full Sphere Material {type=MaterialInterface, default=/ImpostorBaker/Materials/MI_Impostor_FullSphere}
 Hemisphere Material {type=MaterialInterface, default=/ImpostorBaker/Materials/MI_Impostor_Hemisphere}
 Impostor Casts Shadows {type=boolean, default=true}
 Impostor Scene {type=ImpostorScene_C, default=null}
 Impostor Type {type=EImposterLayoutType, default=Upper Hemisphere Only}
 Impostor_MeshCutout_Component {type=Impostor_MeshCutout_C, default=null}
 ImpostorPreviewMID {type=MaterialInstanceDynamic}
 ImpostorRenderComplete {default=()}
 ImpostorRenderSaveComplete {default=()}
 ImpostorSaveComplete {default=()}
 Lighting Grid XY {type=integer, default=4}
 lights {type="array<DirectionalLight>", default []}
 LightVectors {type="array<Vector>"}
 Manual sRGB_RetainerBox {type=boolean, default=true}
 Maps to Render_Internal {type="array<EImpostorBakeTypes>", default [BaseColor,Depth,Normal]}
 Maps to Save {type="array<EImpostorBakeTypes>", default []}
 Mask Offset {type=number}
 Material Preview Parent {type=MaterialInterface, default=&MI_Impostor_Hemisphere}
 MaterialPreviewProxy {type=MaterialInstanceConstant, default=null}
 MaterialSettings_To_Show {type="array<string>", default [Full Sphere Material, Constant Specular]}
 Mesh Scale {type=Vector, default=<1 1 1>}
 New Asset MIC Prefix {type=string, default=MI_Impostor}
 New Asset Path {type=string, default=/ImpostorLODs/}
 New Asset Texture Prefix {type=string, default=T_Impostor}
 NumTexWithoutAlpha {type=integer}
 Object Radius {type=number}
 Offset Vector {type=Vector}
 Orthographic {type=boolean, default=true}
 Override Material {type=MaterialInterface, default=null}
 Parallax Mode {type=EImpostorParallaxMode, default=Single Sample Parallax}
 Processed MIC Path {type=string, default=""}
 Processed Texture Path {type=string, default=""}
 Progress {type=number}
 Pseudo Volume Texture {type=Texture, default=null}
 PsuedoVolume XY Frames {type=IntPoint, default=<16 16>}
 Reference Plane Sides {type=number, default=0.5}
 Reference Plane Top {type=number, default=0.5}
 Rendered {type=boolean}
 Rendering {type=boolean}
 Resample MID {type=MaterialInstanceDynamic}
 Resolution {type=integer, default=4096}
 SampleFrame_DFAlphaMID {type=MaterialInstanceDynamic}
 SampleFrameMID {type=MaterialInstanceDynamic}
 Scatter Mask Length {type=number, default=0.5}
 Scatter Mask Min {type=number, default=0.35}
 Scene Capture Resolution {type=integer, default=512}
 SceneCaptureComponent2D {type=SceneCaptureComponent2D, default=null}
 SceneCaptureMipChain {type="array<TextureRenderTarget2D>", default []}
 SceneCaptureMips {type=integer}
 SceneCaptureMipToComposite {type=integer}
 ScratchRT {type=TextureRenderTarget2D, default=null}
 SettingsPreset {type=ImpostorPresets_C, default=null}
 Shader Debug Mode {type=boolean}
 Shading Model Override {type=EImpostorShadingModelOverride, default=Do Not Override}
 Show Debug Variables {type=boolean}
 Show Debug View Vectors {type=boolean}
 Show Floor in Preview Scene {type=boolean, default=true}
 Show Render Targets {type=boolean}
 Static Mesh {type=StaticMesh, default=/ImpostorBaker/TestContent/ColorCalibCopy/S_ColorCalibrator_ImpostorBaker}
 StaticMeshComponent {type=StaticMeshComponent, default=null}
 Subsurface Color {type=LinearColor, default=<rgba 1 1 1 1.5>}
 SwapByDistance {type=boolean, default=true}
 SwapDistance {type=number, default=500}
 TargetMaps {default {}}
 TexturesToRemoveAlpha {type="array<Texture2D>", default []}
 TotalMaps {type=number}
 Tris {type="array<integer>"}
 Upward Bias {type=number, default=0.75}
 Use Relative Path {type=boolean, default=true}
 Use Sprite Vertex Shader {type=boolean}
 UVs {type="array<Vector2D>", default []}
 Verts {type="array<Vector>"}
 ViewCaptureVectors {type="array<Vector>"}
 ViewModes_To_Show {type="array<string>", default [Full Sphere Material, Constant Specular]}
 Volume Texture {type=VolumeTexture, default=null}
 World {type=Object, default=null}
]
widgets {
 info {rootWidgetClass=&Overlay, widgetCount=89}
 widgets [
  Overlay_0 {widgetClassPath=&Overlay, widget=Overlay_0, properties {visibility=Visible}}
  HorizontalBox_168 {
   widgetClassPath=&HorizontalBox
   parent=Overlay_0
   widget=HorizontalBox_168
   slot {class=OverlaySlot, properties {horizontalAlignment=HAlign_Fill, verticalAlignment=VAlign_Fill}}
  }
  VerticalBox_950 {
   widgetClassPath=&VerticalBox
   parent=HorizontalBox_168
   widget=VerticalBox_950
   slot {class=HorizontalBoxSlot, properties {horizontalAlignment=HAlign_Left}}
  }
  EditorUtilityScrollBox_1526 {
   widgetClassPath=&EditorUtilityScrollBox
   parent=VerticalBox_950
   widget=EditorUtilityScrollBox_1526
   slot {class=VerticalBoxSlot, properties {verticalAlignment=VAlign_Top, size {sizeRule=Fill, value=1}}}
  }
  SizeBox_342 {
   widgetClassPath=&SizeBox
   parent=EditorUtilityScrollBox_1526
   widget=SizeBox_342
   properties {bOverride_WidthOverride=true, widthOverride=350}
   slot {class=ScrollBoxSlot, properties {verticalAlignment=VAlign_Top, size {sizeRule=Fill, value=1}}}
  }
  VerticalBox_63 {
   widgetClassPath=&VerticalBox
   parent=SizeBox_342
   widget=VerticalBox_63
   slot {class=SizeBoxSlot, properties {verticalAlignment=VAlign_Top}}
  }
  Button_DataAssetCategory {
   widgetClassPath=&EditorUtilityButton
   parent=VerticalBox_63
   widget=Button_DataAssetCategory
   properties {widgetStyle {
     clickedSlateSound {resourceObject=null}
     disabled {
      bIsDynamicallyLoaded=false
      drawAs=RoundedBox
      imageType=NoImage
      mirroring=NoMirror
      resourceName=None
      resourceObject=null
      tiling=NoTile
      imageSize=<0 0>
      margin=<ltrb 0 0 0 0>
      outlineSettings {
       bUseBrushTransparency=false
       roundingType=FixedRadius
       width=1
       color {colorUseRule=UseColor_ColorTable, specifiedColor=<rgba 0 0 0 0>}
       cornerRadii=<4 4 4 4>
      }
      tintColor {colorUseRule=UseColor_Specified, specifiedColor=<rgba 0.0395 0.0395 0.0395 1>}
      uVRegion {bIsValid=false, max=<0 0>, min=<0 0>}
     }
     disabledForeground {colorUseRule=UseColor_ColorTable, specifiedColor=<rgba 0 0 0 0>}
     hovered {
      bIsDynamicallyLoaded=false
      drawAs=RoundedBox
      imageType=NoImage
      mirroring=NoMirror
      resourceName=None
      resourceObject=null
      tiling=NoTile
      imageSize=<32 32>
      margin=<ltrb 0 0 0 0>
      outlineSettings {
       bUseBrushTransparency=false
       roundingType=FixedRadius
       width=1
       color {colorUseRule=UseColor_ColorTable, specifiedColor=<rgba 0 0 0 0>}
       cornerRadii=<4 4 4 4>
      }
      tintColor {colorUseRule=UseColor_Specified, specifiedColor=<rgba 0.0953 0.0953 0.0953 1>}
      uVRegion {bIsValid=false, max=<0 0>, min=<0 0>}
     }
     hoveredForeground {colorUseRule=UseColor_ColorTable, specifiedColor=<rgba 0 0 0 0>}
     hoveredSlateSound {resourceObject=null}
     normal {
      bIsDynamicallyLoaded=false
      drawAs=RoundedBox
      imageType=NoImage
      mirroring=NoMirror
      resourceName=None
      resourceObject=null
      tiling=NoTile
      imageSize=<32 32>
      margin=<ltrb 0 0 0 0>
      outlineSettings {
       bUseBrushTransparency=false
       roundingType=FixedRadius
       width=1
       color {colorUseRule=UseColor_ColorTable, specifiedColor=<rgba 0 0 0 0>}
       cornerRadii=<4 4 4 4>
      }
      tintColor {colorUseRule=UseColor_Specified, specifiedColor=<rgba 0.0395 0.0395 0.0395 1>}
      uVRegion {bIsValid=false, max=<0 0>, min=<0 0>}
     }
     normalForeground {colorUseRule=UseColor_ColorTable, specifiedColor=<rgba 0 0 0 0>}
     normalPadding=<ltrb 4 1.5 4 1.5>
     pressed {
      bIsDynamicallyLoaded=false
      drawAs=RoundedBox
      imageType=NoImage
      mirroring=NoMirror
      resourceName=None
      resourceObject=null
      tiling=NoTile
      imageSize=<32 32>
      margin=<ltrb 0 0 0 0>
      outlineSettings {
       bUseBrushTransparency=false
       roundingType=FixedRadius
       width=1
       color {colorUseRule=UseColor_ColorTable, specifiedColor=<rgba 0 0 0 0>}
       cornerRadii=<4 4 4 4>
      }
      tintColor {colorUseRule=UseColor_Specified, specifiedColor=<rgba 0.0284 0.0284 0.0284 1>}
      uVRegion {bIsValid=false, max=<0 0>, min=<0 0>}
     }
     pressedForeground {colorUseRule=UseColor_ColorTable, specifiedColor=<rgba 0 0 0 0>}
     pressedPadding=<ltrb 4 1.5 4 1.5>
     pressedSlateSound {resourceObject=null}
    }}
   slot {class=VerticalBoxSlot}
  }
  HorizontalBox_99 {
   widgetClassPath=&HorizontalBox
   parent=Button_DataAssetCategory
   widget=HorizontalBox_99
   slot {class=ButtonSlot, properties {horizontalAlignment=HAlign_Fill, verticalAlignment=VAlign_Fill, padding=<ltrb 0 0 0 0>}}
  }
  Image_SettingsPresetArrow {
   widgetClassPath=&Image
   bIsVariable=true
   parent=HorizontalBox_99
   widget=Image_SettingsPresetArrow
   properties {brush {
     bIsDynamicallyLoaded=false
     drawAs=Image
     imageType=NoImage
     mirroring=NoMirror
     resourceName=None
     resourceObject=&CategoryArrow
     tiling=NoTile
     imageSize=<12 12>
     margin=<ltrb 0 0 0 0>
     outlineSettings {
      bUseBrushTransparency=false
      roundingType=HalfHeightRadius
      width=0
      color {colorUseRule=UseColor_Specified, specifiedColor=<rgba 0 0 0 0>}
      cornerRadii=<0 0 0 0>
     }
     tintColor {colorUseRule=UseColor_Specified, specifiedColor=<rgba 0.375 0.375 0.375 1>}
     uVRegion {bIsValid=false, max=<0 0>, min=<0 0>}
    }}
   slot {class=HorizontalBoxSlot, properties {horizontalAlignment=HAlign_Center, verticalAlignment=VAlign_Center}}
  }
  TextBlock_13 {
   widgetClassPath=&TextBlock
   parent=HorizontalBox_99
   widget=TextBlock_13
   properties {
    text=Preset DataAsset
    colorAndOpacity {colorUseRule=UseColor_Specified, specifiedColor=<rgba 0.6667 0.6667 0.6667 1>}
    font {
     bForceMonospaced=false
     bMaterialIsStencil=false
     fontMaterial=null
     fontObject=&Roboto
     letterSpacing=0
     monospacedWidth=1
     size=8
     skewAmount=0
     typefaceFontName=Bold
     outlineSettings {
      bApplyOutlineToDropShadows=false
      bMiteredCorners=false
      bSeparateFillAlpha=false
      outlineMaterial=null
      outlineSize=0
      outlineColor=<rgba 0 0 0 1>
     }
    }
   }
   slot {class=HorizontalBoxSlot, properties {horizontalAlignment=HAlign_Left, verticalAlignment=VAlign_Center, padding=<ltrb 4 4 0 4>}}
  }
  VerticalBox_SettingsPreset {
   widgetClassPath=&VerticalBox
   bIsVariable=true
   parent=VerticalBox_63
   widget=VerticalBox_SettingsPreset
   properties {visibility=Collapsed}
   slot {class=VerticalBoxSlot}
  }
  HorizontalBox_69 {
   widgetClassPath=&HorizontalBox
   parent=VerticalBox_SettingsPreset
   widget=HorizontalBox_69
   slot {class=VerticalBoxSlot, properties {horizontalAlignment=HAlign_Center}}
  }
  Button_LoadSettings {
   widgetClassPath=&EditorUtilityButton
   parent=HorizontalBox_69
   widget=Button_LoadSettings
   properties {toolTipText=Loads and applies all settings from the specified Settings Preset Data Asset to this instance.}
   slot {class=HorizontalBoxSlot, properties {padding=<ltrb 10 10 10 10>}}
  }
  TextBlock_10 {
   widgetClassPath=&TextBlock
   parent=Button_LoadSettings
   widget=TextBlock_10
   properties {
    text=Load Settings
    font {
     bForceMonospaced=false
     bMaterialIsStencil=false
     fontMaterial=null
     fontObject=&Roboto
     letterSpacing=0
     monospacedWidth=1
     size=10
     skewAmount=0
     typefaceFontName=Bold
     outlineSettings {
      bApplyOutlineToDropShadows=false
      bMiteredCorners=false
      bSeparateFillAlpha=false
      outlineMaterial=null
      outlineSize=0
      outlineColor=<rgba 0 0 0 1>
     }
    }
   }
   slot {class=ButtonSlot, properties {padding=<ltrb 0 4 0 4>}}
  }
  Button_SaveToDataAsset {
   widgetClassPath=&EditorUtilityButton
   parent=HorizontalBox_69
   widget=Button_SaveToDataAsset
   properties {toolTipText=Saves the current settings from this instance to the specified Settings Preset Data Asset.}
   slot {class=HorizontalBoxSlot, properties {padding=<ltrb 0 10 0 10>}}
  }
  TextBlock_11 {
   widgetClassPath=&TextBlock
   parent=Button_SaveToDataAsset
   widget=TextBlock_11
   properties {
    text=Apply to DataAsset
    font {
     bForceMonospaced=false
     bMaterialIsStencil=false
     fontMaterial=null
     fontObject=&Roboto
     letterSpacing=0
     monospacedWidth=1
     size=10
     skewAmount=0
     typefaceFontName=Bold
     outlineSettings {
      bApplyOutlineToDropShadows=false
      bMiteredCorners=false
      bSeparateFillAlpha=false
      outlineMaterial=null
      outlineSize=0
      outlineColor=<rgba 0 0 0 1>
     }
    }
   }
   slot {class=ButtonSlot, properties {padding=<ltrb 0 4 0 4>}}
  }
  Button_NewDataAsset {
   widgetClassPath=&EditorUtilityButton
   parent=HorizontalBox_69
   widget=Button_NewDataAsset
   properties {toolTipText=Creates a new Settings Data Asset using the settings specified here. Default path will be copied from the specified Source Mesh.}
   slot {class=HorizontalBoxSlot, properties {padding=<ltrb 10 10 10 10>}}
  }
  TextBlock_12 {
   widgetClassPath=&TextBlock
   parent=Button_NewDataAsset
   widget=TextBlock_12
   properties {
    text=New DataAsset
    font {
     bForceMonospaced=false
     bMaterialIsStencil=false
     fontMaterial=null
     fontObject=&Roboto
     letterSpacing=0
     monospacedWidth=1
     size=10
     skewAmount=0
     typefaceFontName=Bold
     outlineSettings {
      bApplyOutlineToDropShadows=false
      bMiteredCorners=false
      bSeparateFillAlpha=false
      outlineMaterial=null
      outlineSize=0
      outlineColor=<rgba 0 0 0 1>
     }
    }
   }
   slot {class=ButtonSlot, properties {padding=<ltrb 0 4 0 4>}}
  }
  SinglePropertyView_80 {
   widgetClassPath=&SinglePropertyView
   parent=VerticalBox_SettingsPreset
   widget=SinglePropertyView_80
   properties {
    propertyName=SettingsPreset
    toolTipText="This is a Data Asset that contains all the properties for Impostor Generation. Allows saving and reusing presets.\r\n\r\nA default preset can be specified per-project by adding the following to DefaultEditor.ini:\r\n\r\n[&/EUW_Generate_Impostor.EUW_Generate_Impostor_C]\r\nDefaultSettingsAssetPath=/Game/YourPresetAssetName\r\n"
   }
   slot {class=VerticalBoxSlot, properties {padding=<ltrb 15 0 15 0>}}
  }
  Details_BakeSettings {
   widgetClassPath=&DetailsView
   parent=VerticalBox_63
   widget=Details_BakeSettings
   properties {categoriesToShow [Bake Settings]}
   slot {class=VerticalBoxSlot}
  }
  Details_MaterialSettings {
   widgetClassPath=&DetailsView
   parent=VerticalBox_63
   widget=Details_MaterialSettings
   properties {propertiesToShow [
     Hemisphere Material
     Shading Model Override
     Parallax Mode
     Constant Specular
     Constant Roughness
     Constant Opacity
     Subsurface Color
     Scatter Mask Min
     Scatter Mask Length
     Mask Offset
    ]}
   slot {class=VerticalBoxSlot}
  }
  Details_AssetSaving {
   widgetClassPath=&DetailsView
   parent=VerticalBox_63
   widget=Details_AssetSaving
   properties {categoriesToShow [Asset Saving]}
   slot {class=VerticalBoxSlot}
  }
  Details_CustomLighting {
   widgetClassPath=&DetailsView
   parent=VerticalBox_63
   widget=Details_CustomLighting
   properties {visibility=Collapsed, categoriesToShow [Custom Lighting]}
   slot {class=VerticalBoxSlot}
  }
  Details_Advanced {
   widgetClassPath=&DetailsView
   parent=VerticalBox_63
   widget=Details_Advanced
   properties {categoriesToShow [Advanced]}
   slot {class=VerticalBoxSlot}
  }
  Details_Debug {
   widgetClassPath=&DetailsView
   parent=VerticalBox_63
   widget=Details_Debug
   properties {visibility=Collapsed, categoriesToShow [Debug, Debug|MIDs]}
   slot {class=VerticalBoxSlot}
  }
  Details_RenderTargets {
   widgetClassPath=&DetailsView
   parent=VerticalBox_63
   widget=Details_RenderTargets
   properties {visibility=Collapsed, categoriesToShow [Render Targets]}
   slot {class=VerticalBoxSlot}
  }
  Details_BakeAssetMaterialOverrides {
   widgetClassPath=&DetailsView
   parent=VerticalBox_63
   widget=Details_BakeAssetMaterialOverrides
   properties {visibility=Collapsed, categoriesToShow [Asset Material Overrides]}
   slot {class=VerticalBoxSlot}
  }
  CanvasPanel_416 {
   widgetClassPath=&CanvasPanel
   parent=VerticalBox_950
   widget=CanvasPanel_416
   slot {class=VerticalBoxSlot, properties {verticalAlignment=VAlign_Top}}
  }
  VerticalBox_1242 {
   widgetClassPath=&VerticalBox
   parent=CanvasPanel_416
   widget=VerticalBox_1242
   slot {class=CanvasPanelSlot, properties {layoutData {alignment=<0 0>, anchors {maximum=<0 0>, minimum=<0 0>}, offsets=<ltrb 0 0 400 100>}}}
  }
  Button_Render {
   widgetClassPath=&EditorUtilityButton
   parent=VerticalBox_1242
   widget=Button_Render
   slot {class=VerticalBoxSlot, properties {padding=<ltrb 15 15 15 5>}}
  }
  TextBlock_87 {
   widgetClassPath=&TextBlock
   parent=Button_Render
   widget=TextBlock_87
   properties {
    text=Render Impostor
    font {
     bForceMonospaced=false
     bMaterialIsStencil=false
     fontMaterial=null
     fontObject=&Roboto
     letterSpacing=0
     monospacedWidth=1
     size=12
     skewAmount=0
     typefaceFontName=Bold
     outlineSettings {
      bApplyOutlineToDropShadows=false
      bMiteredCorners=false
      bSeparateFillAlpha=false
      outlineMaterial=null
      outlineSize=0
      outlineColor=<rgba 0 0 0 1>
     }
    }
   }
   slot {class=ButtonSlot, properties {padding=<ltrb 4 4 4 4>}}
  }
  Button_CreateAssets {
   widgetClassPath=&EditorUtilityButton
   parent=VerticalBox_1242
   widget=Button_CreateAssets
   slot {class=VerticalBoxSlot, properties {padding=<ltrb 15 5 15 5>}}
  }
  TextBlock {
   widgetClassPath=&TextBlock
   parent=Button_CreateAssets
   widget=TextBlock
   properties {
    text=Create Assets and Save to Mesh
    font {
     bForceMonospaced=false
     bMaterialIsStencil=false
     fontMaterial=null
     fontObject=&Roboto
     letterSpacing=0
     monospacedWidth=1
     size=12
     skewAmount=0
     typefaceFontName=Bold
     outlineSettings {
      bApplyOutlineToDropShadows=false
      bMiteredCorners=false
      bSeparateFillAlpha=false
      outlineMaterial=null
      outlineSize=0
      outlineColor=<rgba 0 0 0 1>
     }
    }
   }
   slot {class=ButtonSlot, properties {padding=<ltrb 4 4 4 4>}}
  }
  Overlay_46 {
   widgetClassPath=&Overlay
   parent=HorizontalBox_168
   widget=Overlay_46
   slot {class=HorizontalBoxSlot, properties {size {sizeRule=Fill, value=1}}}
  }
  ImpostorViewport_Widget {
   widgetClassPath=/ImpostorBaker/BP/Misc/ImpostorViewport_Widget
   bIsVariable=true
   parent=Overlay_46
   widget=ImpostorViewport_Widget
   slot {class=OverlaySlot, properties {horizontalAlignment=HAlign_Fill, verticalAlignment=VAlign_Fill}}
  }
  HorizontalBox_ViewOptions {
   widgetClassPath=&HorizontalBox
   parent=Overlay_46
   widget=HorizontalBox_ViewOptions
   slot {class=OverlaySlot, properties {padding=<ltrb 10 10 0 0>}}
  }
  HorizontalBox_246 {
   widgetClassPath=&HorizontalBox
   parent=HorizontalBox_ViewOptions
   widget=HorizontalBox_246
   slot {class=HorizontalBoxSlot, properties {padding=<ltrb 0 0 10 0>}}
  }
  Overlay_233 {widgetClassPath=&Overlay, parent=HorizontalBox_246, widget=Overlay_233, slot {class=HorizontalBoxSlot}}
  Image_183 {
   widgetClassPath=&Image
   parent=Overlay_233
   widget=Image_183
   properties {
    brush {
     bIsDynamicallyLoaded=false
     drawAs=Box
     imageType=NoImage
     mirroring=NoMirror
     resourceName=None
     resourceObject=/ImpostorBaker/Materials/Slate/Impostor_Button
     tiling=NoTile
     imageSize=<24 28>
     margin=<ltrb 0.5 0.5 0.5 0.5>
     outlineSettings {
      bUseBrushTransparency=true
      roundingType=FixedRadius
      width=8
      color {colorUseRule=UseColor_Foreground, specifiedColor=<rgba 2 2 0 1>}
      cornerRadii=<0.2 0.2 0.2 0.2>
     }
     tintColor {colorUseRule=UseColor_Specified, specifiedColor=<rgba 1 1 1 1>}
     uVRegion {bIsValid=false, max=<0 0>, min=<0 0>}
    }
    colorAndOpacity=<rgba 0.1 0.1 0.1 0.75>
   }
   slot {class=OverlaySlot, properties {horizontalAlignment=HAlign_Fill, verticalAlignment=VAlign_Fill}}
  }
  HorizontalBox_120 {
   widgetClassPath=&HorizontalBox
   parent=Overlay_233
   widget=HorizontalBox_120
   slot {class=OverlaySlot, properties {verticalAlignment=VAlign_Center, padding=<ltrb 10 0 10 0>}}
  }
  TextBlock_3 {
   widgetClassPath=&TextBlock
   parent=HorizontalBox_120
   widget=TextBlock_3
   properties {
    text=View Mode:
    colorAndOpacity {colorUseRule=UseColor_Specified, specifiedColor=<rgba 0.5312 0.5312 0.5312 1>}
    font {
     bForceMonospaced=false
     bMaterialIsStencil=false
     fontMaterial=null
     fontObject=&Roboto
     letterSpacing=0
     monospacedWidth=1
     size=10
     skewAmount=0
     typefaceFontName=Regular
     outlineSettings {
      bApplyOutlineToDropShadows=false
      bMiteredCorners=false
      bSeparateFillAlpha=false
      outlineMaterial=null
      outlineSize=0
      outlineColor=<rgba 0 0 0 1>
     }
    }
    shadowColorAndOpacity=<rgba 0 0 0 0.3>
   }
   slot {class=HorizontalBoxSlot, properties {verticalAlignment=VAlign_Center, padding=<ltrb 5 0 5 0>}}
  }
  Combobox_ViewMode {
   widgetClassPath=/Script/Blutility.EditorUtilityComboBoxString
   bIsVariable=true
   parent=HorizontalBox_120
   widget=Combobox_ViewMode
   properties {
    selectedOption=Lit
    contentPadding=<ltrb 4 0 4 0>
    defaultOptions [Lit,Basecolor,Specular,Roughness,Normal,Subsurface]
    font {
     bForceMonospaced=false
     bMaterialIsStencil=false
     fontMaterial=null
     fontObject=&Roboto
     letterSpacing=0
     monospacedWidth=1
     size=10
     skewAmount=0
     typefaceFontName=Regular
     outlineSettings {
      bApplyOutlineToDropShadows=false
      bMiteredCorners=false
      bSeparateFillAlpha=false
      outlineMaterial=null
      outlineSize=0
      outlineColor=<rgba 0 0 0 1>
     }
    }
    foregroundColor {colorUseRule=UseColor_Specified, specifiedColor=<rgba 0.5 0.5 0.5 1>}
    itemStyle {
     bUseParentRowBrush=false
     activeBrush {
      bIsDynamicallyLoaded=false
      drawAs=Image
      imageType=NoImage
      mirroring=NoMirror
      resourceName=None
      resourceObject=null
      tiling=NoTile
      imageSize=<0 0>
      margin=<ltrb 0 0 0 0>
      outlineSettings {
       bUseBrushTransparency=false
       roundingType=HalfHeightRadius
       width=0
       color {colorUseRule=UseColor_Specified, specifiedColor=<rgba 0 0 0 0>}
       cornerRadii=<0 0 0 0>
      }
      tintColor {colorUseRule=UseColor_Specified, specifiedColor=<rgba 0 0.162 0.7454 1>}
      uVRegion {bIsValid=false, max=<0 0>, min=<0 0>}
     }
     activeHighlightedBrush {
      bIsDynamicallyLoaded=false
      drawAs=Image
      imageType=NoImage
      mirroring=NoMirror
      resourceName=None
      resourceObject=null
      tiling=NoTile
      imageSize=<0 0>
      margin=<ltrb 0 0 0 0>
      outlineSettings {
       bUseBrushTransparency=false
       roundingType=HalfHeightRadius
       width=0
       color {colorUseRule=UseColor_Specified, specifiedColor=<rgba 0 0 0 0>}
       cornerRadii=<0 0 0 0>
      }
      tintColor {colorUseRule=UseColor_Specified, specifiedColor=<rgba 0.0252 0.0319 0.0423 1>}
      uVRegion {bIsValid=false, max=<0 0>, min=<0 0>}
     }
     activeHoveredBrush {
      bIsDynamicallyLoaded=false
      drawAs=Image
      imageType=NoImage
      mirroring=NoMirror
      resourceName=None
      resourceObject=null
      tiling=NoTile
      imageSize=<0 0>
      margin=<ltrb 0 0 0 0>
      outlineSettings {
       bUseBrushTransparency=false
       roundingType=HalfHeightRadius
       width=0
       color {colorUseRule=UseColor_Specified, specifiedColor=<rgba 0 0 0 0>}
       cornerRadii=<0 0 0 0>
      }
      tintColor {colorUseRule=UseColor_Specified, specifiedColor=<rgba 0 0.162 0.7454 1>}
      uVRegion {bIsValid=false, max=<0 0>, min=<0 0>}
     }
     dropIndicator_Above {
      bIsDynamicallyLoaded=false
      drawAs=Box
      imageType=FullColor
      mirroring=NoMirror
      resourceName=../../../Engine/Content/Slate/Common/DropZoneIndicator_Above.png
      resourceObject=null
      tiling=NoTile
      imageSize=<0 0>
      margin=<ltrb 0.625 0.625 0 0>
      outlineSettings {
       bUseBrushTransparency=false
       roundingType=HalfHeightRadius
       width=0
       color {colorUseRule=UseColor_Specified, specifiedColor=<rgba 0 0 0 0>}
       cornerRadii=<0 0 0 0>
      }
      tintColor {colorUseRule=UseColor_Specified, specifiedColor=<rgba 0.0091 0.3515 0.9647 1>}
      uVRegion {bIsValid=false, max=<0 0>, min=<0 0>}
     }
     dropIndicator_Below {
      bIsDynamicallyLoaded=false
      drawAs=Box
      imageType=FullColor
      mirroring=NoMirror
      resourceName=../../../Engine/Content/Slate/Common/DropZoneIndicator_Below.png
      resourceObject=null
      tiling=NoTile
      imageSize=<0 0>
      margin=<ltrb 0.625 0 0 0.625>
      outlineSettings {
       bUseBrushTransparency=false
       roundingType=HalfHeightRadius
       width=0
       color {colorUseRule=UseColor_Specified, specifiedColor=<rgba 0 0 0 0>}
       cornerRadii=<0 0 0 0>
      }
      tintColor {colorUseRule=UseColor_Specified, specifiedColor=<rgba 0.0091 0.3515 0.9647 1>}
      uVRegion {bIsValid=false, max=<0 0>, min=<0 0>}
     }
     dropIndicator_Onto {
      bIsDynamicallyLoaded=false
      drawAs=Box
      imageType=FullColor
      mirroring=NoMirror
      resourceName=../../../Engine/Content/Slate/Common/DropZoneIndicator_Onto.png
      resourceObject=null
      tiling=NoTile
      imageSize=<0 0>
      margin=<ltrb 0.25 0.25 0.25 0.25>
      outlineSettings {
       bUseBrushTransparency=false
       roundingType=HalfHeightRadius
       width=0
       color {colorUseRule=UseColor_Specified, specifiedColor=<rgba 0 0 0 0>}
       cornerRadii=<0 0 0 0>
      }
      tintColor {colorUseRule=UseColor_Specified, specifiedColor=<rgba 0.0091 0.3515 0.9647 1>}
      uVRegion {bIsValid=false, max=<0 0>, min=<0 0>}
     }
     evenRowBackgroundBrush {
      bIsDynamicallyLoaded=false
      drawAs=Image
      imageType=NoImage
      mirroring=NoMirror
      resourceName=None
      resourceObject=null
      tiling=NoTile
      imageSize=<0 0>
      margin=<ltrb 0 0 0 0>
      outlineSettings {
       bUseBrushTransparency=false
       roundingType=HalfHeightRadius
       width=0
       color {colorUseRule=UseColor_Specified, specifiedColor=<rgba 0 0 0 0>}
       cornerRadii=<0 0 0 0>
      }
      tintColor {colorUseRule=UseColor_Specified, specifiedColor=<rgba 0.0103 0.0103 0.0103 1>}
      uVRegion {bIsValid=false, max=<0 0>, min=<0 0>}
     }
     evenRowBackgroundHoveredBrush {
      bIsDynamicallyLoaded=false
      drawAs=Image
      imageType=NoImage
      mirroring=NoMirror
      resourceName=None
      resourceObject=null
      tiling=NoTile
      imageSize=<0 0>
      margin=<ltrb 0 0 0 0>
      outlineSettings {
       bUseBrushTransparency=false
       roundingType=HalfHeightRadius
       width=0
       color {colorUseRule=UseColor_Specified, specifiedColor=<rgba 0 0 0 0>}
       cornerRadii=<0 0 0 0>
      }
      tintColor {colorUseRule=UseColor_Specified, specifiedColor=<rgba 0.0176 0.0176 0.0176 1>}
      uVRegion {bIsValid=false, max=<0 0>, min=<0 0>}
     }
     inactiveBrush {
      bIsDynamicallyLoaded=false
      drawAs=Image
      imageType=NoImage
      mirroring=NoMirror
      resourceName=None
      resourceObject=null
      tiling=NoTile
      imageSize=<0 0>
      margin=<ltrb 0 0 0 0>
      outlineSettings {
       bUseBrushTransparency=false
       roundingType=HalfHeightRadius
       width=0
       color {colorUseRule=UseColor_Specified, specifiedColor=<rgba 0 0 0 0>}
       cornerRadii=<0 0 0 0>
      }
      tintColor {colorUseRule=UseColor_Specified, specifiedColor=<rgba 0.0513 0.0953 0.159 1>}
      uVRegion {bIsValid=false, max=<0 0>, min=<0 0>}
     }
     inactiveHighlightedBrush {
      bIsDynamicallyLoaded=false
      drawAs=Image
      imageType=NoImage
      mirroring=NoMirror
      resourceName=None
      resourceObject=null
      tiling=NoTile
      imageSize=<0 0>
      margin=<ltrb 0 0 0 0>
      outlineSettings {
       bUseBrushTransparency=false
       roundingType=HalfHeightRadius
       width=0
       color {colorUseRule=UseColor_Specified, specifiedColor=<rgba 0 0 0 0>}
       cornerRadii=<0 0 0 0>
      }
      tintColor {colorUseRule=UseColor_Specified, specifiedColor=<rgba 0.0252 0.0319 0.0423 1>}
      uVRegion {bIsValid=false, max=<0 0>, min=<0 0>}
     }
     inactiveHoveredBrush {
      bIsDynamicallyLoaded=false
      drawAs=Image
      imageType=NoImage
      mirroring=NoMirror
      resourceName=None
      resourceObject=null
      tiling=NoTile
      imageSize=<0 0>
      margin=<ltrb 0 0 0 0>
      outlineSettings {
       bUseBrushTransparency=false
       roundingType=HalfHeightRadius
       width=0
       color {colorUseRule=UseColor_Specified, specifiedColor=<rgba 0 0 0 0>}
       cornerRadii=<0 0 0 0>
      }
      tintColor {colorUseRule=UseColor_Specified, specifiedColor=<rgba 0.0513 0.0953 0.159 1>}
      uVRegion {bIsValid=false, max=<0 0>, min=<0 0>}
     }
     oddRowBackgroundBrush {
      bIsDynamicallyLoaded=false
      drawAs=Image
      imageType=NoImage
      mirroring=NoMirror
      resourceName=None
      resourceObject=null
      tiling=NoTile
      imageSize=<0 0>
      margin=<ltrb 0 0 0 0>
      outlineSettings {
       bUseBrushTransparency=false
       roundingType=HalfHeightRadius
       width=0
       color {colorUseRule=UseColor_Specified, specifiedColor=<rgba 0 0 0 0>}
       cornerRadii=<0 0 0 0>
      }
      tintColor {colorUseRule=UseColor_Specified, specifiedColor=<rgba 0.0103 0.0103 0.0103 1>}
      uVRegion {bIsValid=false, max=<0 0>, min=<0 0>}
     }
     oddRowBackgroundHoveredBrush {
      bIsDynamicallyLoaded=false
      drawAs=Image
      imageType=NoImage
      mirroring=NoMirror
      resourceName=None
      resourceObject=null
      tiling=NoTile
      imageSize=<0 0>
      margin=<ltrb 0 0 0 0>
      outlineSettings {
       bUseBrushTransparency=false
       roundingType=HalfHeightRadius
       width=0
       color {colorUseRule=UseColor_Specified, specifiedColor=<rgba 0 0 0 0>}
       cornerRadii=<0 0 0 0>
      }
      tintColor {colorUseRule=UseColor_Specified, specifiedColor=<rgba 0.0176 0.0176 0.0176 1>}
      uVRegion {bIsValid=false, max=<0 0>, min=<0 0>}
     }
     parentRowBackgroundBrush {
      bIsDynamicallyLoaded=false
      drawAs=Image
      imageType=NoImage
      mirroring=NoMirror
      resourceName=None
      resourceObject=null
      tiling=NoTile
      imageSize=<32 32>
      margin=<ltrb 0 0 0 0>
      outlineSettings {
       bUseBrushTransparency=false
       roundingType=HalfHeightRadius
       width=0
       color {colorUseRule=UseColor_Specified, specifiedColor=<rgba 0 0 0 0>}
       cornerRadii=<0 0 0 0>
      }
      tintColor {colorUseRule=UseColor_Specified, specifiedColor=<rgba 1 1 1 1>}
      uVRegion {bIsValid=false, max=<0 0>, min=<0 0>}
     }
     parentRowBackgroundHoveredBrush {
      bIsDynamicallyLoaded=false
      drawAs=Image
      imageType=NoImage
      mirroring=NoMirror
      resourceName=None
      resourceObject=null
      tiling=NoTile
      imageSize=<32 32>
      margin=<ltrb 0 0 0 0>
      outlineSettings {
       bUseBrushTransparency=false
       roundingType=HalfHeightRadius
       width=0
       color {colorUseRule=UseColor_Specified, specifiedColor=<rgba 0 0 0 0>}
       cornerRadii=<0 0 0 0>
      }
      tintColor {colorUseRule=UseColor_Specified, specifiedColor=<rgba 1 1 1 1>}
      uVRegion {bIsValid=false, max=<0 0>, min=<0 0>}
     }
     selectedTextColor {colorUseRule=UseColor_Specified, specifiedColor=<rgba 1 1 1 1>}
     selectorFocusedBrush {
      bIsDynamicallyLoaded=false
      drawAs=Border
      imageType=FullColor
      mirroring=NoMirror
      resourceName=../../../Engine/Content/Slate/Common/Selector.png
      resourceObject=null
      tiling=Both
      imageSize=<0 0>
      margin=<ltrb 0.25 0.25 0.25 0.25>
      outlineSettings {
       bUseBrushTransparency=false
       roundingType=HalfHeightRadius
       width=0
       color {colorUseRule=UseColor_Specified, specifiedColor=<rgba 0 0 0 0>}
       cornerRadii=<0 0 0 0>
      }
      tintColor {colorUseRule=UseColor_Specified, specifiedColor=<rgba 0.701 0.225 0.003 1>}
      uVRegion {bIsValid=false, max=<0 0>, min=<0 0>}
     }
     textColor {colorUseRule=UseColor_Specified, specifiedColor=<rgba 0.5271 0.5271 0.5271 1>}
    }
    widgetStyle {
     comboButtonStyle {
      downArrowAlign=VAlign_Center
      buttonStyle {
       clickedSlateSound {resourceObject=null}
       disabled {
        bIsDynamicallyLoaded=false
        drawAs=RoundedBox
        imageType=NoImage
        mirroring=NoMirror
        resourceName=None
        resourceObject=null
        tiling=NoTile
        imageSize=<0 0>
        margin=<ltrb 0 0 0 0>
        outlineSettings {
         bUseBrushTransparency=false
         roundingType=FixedRadius
         width=1
         color {colorUseRule=UseColor_ColorTable, specifiedColor=<rgba 0 0 0 0>}
         cornerRadii=<4 4 4 4>
        }
        tintColor {colorUseRule=UseColor_Specified, specifiedColor=<rgba 0.0395 0.0395 0.0395 1>}
        uVRegion {bIsValid=false, max=<0 0>, min=<0 0>}
       }
       disabledForeground {colorUseRule=UseColor_ColorTable, specifiedColor=<rgba 0 0 0 0>}
       hovered {
        bIsDynamicallyLoaded=false
        drawAs=RoundedBox
        imageType=NoImage
        mirroring=NoMirror
        resourceName=None
        resourceObject=null
        tiling=NoTile
        imageSize=<0 0>
        margin=<ltrb 0 0 0 0>
        outlineSettings {
         bUseBrushTransparency=false
         roundingType=FixedRadius
         width=1
         color {colorUseRule=UseColor_ColorTable, specifiedColor=<rgba 0 0 0 0>}
         cornerRadii=<4 4 4 4>
        }
        tintColor {colorUseRule=UseColor_Specified, specifiedColor=<rgba 0.0048 0.0048 0.0048 1>}
        uVRegion {bIsValid=false, max=<0 0>, min=<0 0>}
       }
       hoveredForeground {colorUseRule=UseColor_ColorTable, specifiedColor=<rgba 0 0 0 0>}
       hoveredSlateSound {resourceObject=null}
       normal {
        bIsDynamicallyLoaded=false
        drawAs=RoundedBox
        imageType=NoImage
        mirroring=NoMirror
        resourceName=None
        resourceObject=null
        tiling=NoTile
        imageSize=<0 0>
        margin=<ltrb 0 0 0 0>
        outlineSettings {
         bUseBrushTransparency=false
         roundingType=FixedRadius
         width=1.1
         color {colorUseRule=UseColor_ColorTable, specifiedColor=<rgba 0 0 0 0>}
         cornerRadii=<4 4 4 4>
        }
        tintColor {colorUseRule=UseColor_Specified, specifiedColor=<rgba 0.0048 0.0048 0.0048 0.25>}
        uVRegion {bIsValid=false, max=<0 0>, min=<0 0>}
       }
       normalForeground {colorUseRule=UseColor_ColorTable, specifiedColor=<rgba 0 0 0 0>}
       normalPadding=<ltrb 8 1 8 1>
       pressed {
        bIsDynamicallyLoaded=false
        drawAs=RoundedBox
        imageType=NoImage
        mirroring=NoMirror
        resourceName=None
        resourceObject=null
        tiling=NoTile
        imageSize=<0 0>
        margin=<ltrb 0 0 0 0>
        outlineSettings {
         bUseBrushTransparency=false
         roundingType=FixedRadius
         width=1
         color {colorUseRule=UseColor_ColorTable, specifiedColor=<rgba 0 0 0 0>}
         cornerRadii=<4 4 4 4>
        }
        tintColor {colorUseRule=UseColor_Specified, specifiedColor=<rgba 0.0048 0.0048 0.0048 1>}
        uVRegion {bIsValid=false, max=<0 0>, min=<0 0>}
       }
       pressedForeground {colorUseRule=UseColor_ColorTable, specifiedColor=<rgba 0 0 0 0>}
       pressedPadding=<ltrb 8 2 8 0>
       pressedSlateSound {resourceObject=null}
      }
      contentPadding=<ltrb 0 0 0 0>
      downArrowImage {
       bIsDynamicallyLoaded=false
       drawAs=Image
       imageType=Vector
       mirroring=NoMirror
       resourceName=../../../Engine/Content/Slate/Starship/CoreWidgets/ComboBox/wide-chevron-down.svg
       resourceObject=null
       tiling=NoTile
       imageSize=<16 16>
       margin=<ltrb 0 0 0 0>
       outlineSettings {
        bUseBrushTransparency=false
        roundingType=HalfHeightRadius
        width=0
        color {colorUseRule=UseColor_Specified, specifiedColor=<rgba 0 0 0 0>}
        cornerRadii=<0 0 0 0>
       }
       tintColor {colorUseRule=UseColor_Specified, specifiedColor=<rgba 1 1 1 1>}
       uVRegion {bIsValid=false, max=<0 0>, min=<0 0>}
      }
      downArrowPadding=<ltrb 2 2 2 2>
      menuBorderBrush {
       bIsDynamicallyLoaded=false
       drawAs=RoundedBox
       imageType=NoImage
       mirroring=NoMirror
       resourceName=None
       resourceObject=null
       tiling=NoTile
       imageSize=<0 0>
       margin=<ltrb 0 0 0 0>
       outlineSettings {
        bUseBrushTransparency=false
        roundingType=FixedRadius
        width=2
        color {colorUseRule=UseColor_ColorTable, specifiedColor=<rgba 0 0 0 0>}
        cornerRadii=<0 0 0 0>
       }
       tintColor {colorUseRule=UseColor_Specified, specifiedColor=<rgba 0.0395 0.0395 0.0395 1>}
       uVRegion {bIsValid=false, max=<0 0>, min=<0 0>}
      }
      menuBorderPadding=<ltrb 0 0 0 0>
      shadowColorAndOpacity=<rgba 0 0 0 1>
      shadowOffset=<0 0>
     }
     contentPadding=<ltrb 0 0 0 0>
     menuRowPadding=<ltrb 8 3 8 3>
     pressedSlateSound {resourceObject=null}
     selectionChangeSlateSound {resourceObject=null}
    }
   }
   slot {class=HorizontalBoxSlot, properties {horizontalAlignment=HAlign_Left, verticalAlignment=VAlign_Center}}
  }
  HorizontalBox {
   widgetClassPath=&HorizontalBox
   parent=HorizontalBox_ViewOptions
   widget=HorizontalBox
   slot {class=HorizontalBoxSlot, properties {padding=<ltrb 0 0 10 0>}}
  }
  Overlay {widgetClassPath=&Overlay, parent=HorizontalBox, widget=Overlay, slot {class=HorizontalBoxSlot}}
  Image {
   widgetClassPath=&Image
   parent=Overlay
   widget=Image
   properties {
    brush {
     bIsDynamicallyLoaded=false
     drawAs=Box
     imageType=NoImage
     mirroring=NoMirror
     resourceName=None
     resourceObject=&Impostor_Button
     tiling=NoTile
     imageSize=<24 28>
     margin=<ltrb 0.5 0.5 0.5 0.5>
     outlineSettings {
      bUseBrushTransparency=true
      roundingType=FixedRadius
      width=8
      color {colorUseRule=UseColor_Foreground, specifiedColor=<rgba 2 2 0 1>}
      cornerRadii=<0.2 0.2 0.2 0.2>
     }
     tintColor {colorUseRule=UseColor_Specified, specifiedColor=<rgba 1 1 1 1>}
     uVRegion {bIsValid=false, max=<0 0>, min=<0 0>}
    }
    colorAndOpacity=<rgba 0.1 0.1 0.1 0.75>
   }
   slot {class=OverlaySlot, properties {horizontalAlignment=HAlign_Fill, verticalAlignment=VAlign_Fill}}
  }
  HorizontalBox_1 {
   widgetClassPath=&HorizontalBox
   parent=Overlay
   widget=HorizontalBox_1
   slot {class=OverlaySlot, properties {verticalAlignment=VAlign_Center, padding=<ltrb 10 0 10 0>}}
  }
  TextBlock_4 {
   widgetClassPath=&TextBlock
   parent=HorizontalBox_1
   widget=TextBlock_4
   properties {
    text=Preview Mode:
    colorAndOpacity {colorUseRule=UseColor_Specified, specifiedColor=<rgba 0.5312 0.5312 0.5312 1>}
    font {
     bForceMonospaced=false
     bMaterialIsStencil=false
     fontMaterial=null
     fontObject=&Roboto
     letterSpacing=0
     monospacedWidth=1
     size=10
     skewAmount=0
     typefaceFontName=Regular
     outlineSettings {
      bApplyOutlineToDropShadows=false
      bMiteredCorners=false
      bSeparateFillAlpha=false
      outlineMaterial=null
      outlineSize=0
      outlineColor=<rgba 0 0 0 1>
     }
    }
    shadowColorAndOpacity=<rgba 0 0 0 0.3>
   }
   slot {class=HorizontalBoxSlot, properties {verticalAlignment=VAlign_Center, padding=<ltrb 5 0 5 0>}}
  }
  Combobox_PreviewMode {
   widgetClassPath=&EditorUtilityComboBoxString
   bIsVariable=true
   parent=HorizontalBox_1
   widget=Combobox_PreviewMode
   properties {
    selectedOption=Swap by Distance
    contentPadding=<ltrb 4 0 4 0>
    defaultOptions [Swap by Distance, Impostor + Original Side by Side, Impostor Only]
    font {
     bForceMonospaced=false
     bMaterialIsStencil=false
     fontMaterial=null
     fontObject=&Roboto
     letterSpacing=0
     monospacedWidth=1
     size=10
     skewAmount=0
     typefaceFontName=Regular
     outlineSettings {
      bApplyOutlineToDropShadows=false
      bMiteredCorners=false
      bSeparateFillAlpha=false
      outlineMaterial=null
      outlineSize=0
      outlineColor=<rgba 0 0 0 1>
     }
    }
    foregroundColor {colorUseRule=UseColor_Specified, specifiedColor=<rgba 0.5 0.5 0.5 1>}
    itemStyle {
     bUseParentRowBrush=false
     activeBrush {
      bIsDynamicallyLoaded=false
      drawAs=Image
      imageType=NoImage
      mirroring=NoMirror
      resourceName=None
      resourceObject=null
      tiling=NoTile
      imageSize=<0 0>
      margin=<ltrb 0 0 0 0>
      outlineSettings {
       bUseBrushTransparency=false
       roundingType=HalfHeightRadius
       width=0
       color {colorUseRule=UseColor_Specified, specifiedColor=<rgba 0 0 0 0>}
       cornerRadii=<0 0 0 0>
      }
      tintColor {colorUseRule=UseColor_Specified, specifiedColor=<rgba 0 0.162 0.7454 1>}
      uVRegion {bIsValid=false, max=<0 0>, min=<0 0>}
     }
     activeHighlightedBrush {
      bIsDynamicallyLoaded=false
      drawAs=Image
      imageType=NoImage
      mirroring=NoMirror
      resourceName=None
      resourceObject=null
      tiling=NoTile
      imageSize=<0 0>
      margin=<ltrb 0 0 0 0>
      outlineSettings {
       bUseBrushTransparency=false
       roundingType=HalfHeightRadius
       width=0
       color {colorUseRule=UseColor_Specified, specifiedColor=<rgba 0 0 0 0>}
       cornerRadii=<0 0 0 0>
      }
      tintColor {colorUseRule=UseColor_Specified, specifiedColor=<rgba 0.0252 0.0319 0.0423 1>}
      uVRegion {bIsValid=false, max=<0 0>, min=<0 0>}
     }
     activeHoveredBrush {
      bIsDynamicallyLoaded=false
      drawAs=Image
      imageType=NoImage
      mirroring=NoMirror
      resourceName=None
      resourceObject=null
      tiling=NoTile
      imageSize=<0 0>
      margin=<ltrb 0 0 0 0>
      outlineSettings {
       bUseBrushTransparency=false
       roundingType=HalfHeightRadius
       width=0
       color {colorUseRule=UseColor_Specified, specifiedColor=<rgba 0 0 0 0>}
       cornerRadii=<0 0 0 0>
      }
      tintColor {colorUseRule=UseColor_Specified, specifiedColor=<rgba 0 0.162 0.7454 1>}
      uVRegion {bIsValid=false, max=<0 0>, min=<0 0>}
     }
     dropIndicator_Above {
      bIsDynamicallyLoaded=false
      drawAs=Box
      imageType=FullColor
      mirroring=NoMirror
      resourceName=../../../Engine/Content/Slate/Common/DropZoneIndicator_Above.png
      resourceObject=null
      tiling=NoTile
      imageSize=<0 0>
      margin=<ltrb 0.625 0.625 0 0>
      outlineSettings {
       bUseBrushTransparency=false
       roundingType=HalfHeightRadius
       width=0
       color {colorUseRule=UseColor_Specified, specifiedColor=<rgba 0 0 0 0>}
       cornerRadii=<0 0 0 0>
      }
      tintColor {colorUseRule=UseColor_Specified, specifiedColor=<rgba 0.0091 0.3515 0.9647 1>}
      uVRegion {bIsValid=false, max=<0 0>, min=<0 0>}
     }
     dropIndicator_Below {
      bIsDynamicallyLoaded=false
      drawAs=Box
      imageType=FullColor
      mirroring=NoMirror
      resourceName=../../../Engine/Content/Slate/Common/DropZoneIndicator_Below.png
      resourceObject=null
      tiling=NoTile
      imageSize=<0 0>
      margin=<ltrb 0.625 0 0 0.625>
      outlineSettings {
       bUseBrushTransparency=false
       roundingType=HalfHeightRadius
       width=0
       color {colorUseRule=UseColor_Specified, specifiedColor=<rgba 0 0 0 0>}
       cornerRadii=<0 0 0 0>
      }
      tintColor {colorUseRule=UseColor_Specified, specifiedColor=<rgba 0.0091 0.3515 0.9647 1>}
      uVRegion {bIsValid=false, max=<0 0>, min=<0 0>}
     }
     dropIndicator_Onto {
      bIsDynamicallyLoaded=false
      drawAs=Box
      imageType=FullColor
      mirroring=NoMirror
      resourceName=../../../Engine/Content/Slate/Common/DropZoneIndicator_Onto.png
      resourceObject=null
      tiling=NoTile
      imageSize=<0 0>
      margin=<ltrb 0.25 0.25 0.25 0.25>
      outlineSettings {
       bUseBrushTransparency=false
       roundingType=HalfHeightRadius
       width=0
       color {colorUseRule=UseColor_Specified, specifiedColor=<rgba 0 0 0 0>}
       cornerRadii=<0 0 0 0>
      }
      tintColor {colorUseRule=UseColor_Specified, specifiedColor=<rgba 0.0091 0.3515 0.9647 1>}
      uVRegion {bIsValid=false, max=<0 0>, min=<0 0>}
     }
     evenRowBackgroundBrush {
      bIsDynamicallyLoaded=false
      drawAs=Image
      imageType=NoImage
      mirroring=NoMirror
      resourceName=None
      resourceObject=null
      tiling=NoTile
      imageSize=<0 0>
      margin=<ltrb 0 0 0 0>
      outlineSettings {
       bUseBrushTransparency=false
       roundingType=HalfHeightRadius
       width=0
       color {colorUseRule=UseColor_Specified, specifiedColor=<rgba 0 0 0 0>}
       cornerRadii=<0 0 0 0>
      }
      tintColor {colorUseRule=UseColor_Specified, specifiedColor=<rgba 0.0103 0.0103 0.0103 1>}
      uVRegion {bIsValid=false, max=<0 0>, min=<0 0>}
     }
     evenRowBackgroundHoveredBrush {
      bIsDynamicallyLoaded=false
      drawAs=Image
      imageType=NoImage
      mirroring=NoMirror
      resourceName=None
      resourceObject=null
      tiling=NoTile
      imageSize=<0 0>
      margin=<ltrb 0 0 0 0>
      outlineSettings {
       bUseBrushTransparency=false
       roundingType=HalfHeightRadius
       width=0
       color {colorUseRule=UseColor_Specified, specifiedColor=<rgba 0 0 0 0>}
       cornerRadii=<0 0 0 0>
      }
      tintColor {colorUseRule=UseColor_Specified, specifiedColor=<rgba 0.0176 0.0176 0.0176 1>}
      uVRegion {bIsValid=false, max=<0 0>, min=<0 0>}
     }
     inactiveBrush {
      bIsDynamicallyLoaded=false
      drawAs=Image
      imageType=NoImage
      mirroring=NoMirror
      resourceName=None
      resourceObject=null
      tiling=NoTile
      imageSize=<0 0>
      margin=<ltrb 0 0 0 0>
      outlineSettings {
       bUseBrushTransparency=false
       roundingType=HalfHeightRadius
       width=0
       color {colorUseRule=UseColor_Specified, specifiedColor=<rgba 0 0 0 0>}
       cornerRadii=<0 0 0 0>
      }
      tintColor {colorUseRule=UseColor_Specified, specifiedColor=<rgba 0.0513 0.0953 0.159 1>}
      uVRegion {bIsValid=false, max=<0 0>, min=<0 0>}
     }
     inactiveHighlightedBrush {
      bIsDynamicallyLoaded=false
      drawAs=Image
      imageType=NoImage
      mirroring=NoMirror
      resourceName=None
      resourceObject=null
      tiling=NoTile
      imageSize=<0 0>
      margin=<ltrb 0 0 0 0>
      outlineSettings {
       bUseBrushTransparency=false
       roundingType=HalfHeightRadius
       width=0
       color {colorUseRule=UseColor_Specified, specifiedColor=<rgba 0 0 0 0>}
       cornerRadii=<0 0 0 0>
      }
      tintColor {colorUseRule=UseColor_Specified, specifiedColor=<rgba 0.0252 0.0319 0.0423 1>}
      uVRegion {bIsValid=false, max=<0 0>, min=<0 0>}
     }
     inactiveHoveredBrush {
      bIsDynamicallyLoaded=false
      drawAs=Image
      imageType=NoImage
      mirroring=NoMirror
      resourceName=None
      resourceObject=null
      tiling=NoTile
      imageSize=<0 0>
      margin=<ltrb 0 0 0 0>
      outlineSettings {
       bUseBrushTransparency=false
       roundingType=HalfHeightRadius
       width=0
       color {colorUseRule=UseColor_Specified, specifiedColor=<rgba 0 0 0 0>}
       cornerRadii=<0 0 0 0>
      }
      tintColor {colorUseRule=UseColor_Specified, specifiedColor=<rgba 0.0513 0.0953 0.159 1>}
      uVRegion {bIsValid=false, max=<0 0>, min=<0 0>}
     }
     oddRowBackgroundBrush {
      bIsDynamicallyLoaded=false
      drawAs=Image
      imageType=NoImage
      mirroring=NoMirror
      resourceName=None
      resourceObject=null
      tiling=NoTile
      imageSize=<0 0>
      margin=<ltrb 0 0 0 0>
      outlineSettings {
       bUseBrushTransparency=false
       roundingType=HalfHeightRadius
       width=0
       color {colorUseRule=UseColor_Specified, specifiedColor=<rgba 0 0 0 0>}
       cornerRadii=<0 0 0 0>
      }
      tintColor {colorUseRule=UseColor_Specified, specifiedColor=<rgba 0.0103 0.0103 0.0103 1>}
      uVRegion {bIsValid=false, max=<0 0>, min=<0 0>}
     }
     oddRowBackgroundHoveredBrush {
      bIsDynamicallyLoaded=false
      drawAs=Image
      imageType=NoImage
      mirroring=NoMirror
      resourceName=None
      resourceObject=null
      tiling=NoTile
      imageSize=<0 0>
      margin=<ltrb 0 0 0 0>
      outlineSettings {
       bUseBrushTransparency=false
       roundingType=HalfHeightRadius
       width=0
       color {colorUseRule=UseColor_Specified, specifiedColor=<rgba 0 0 0 0>}
       cornerRadii=<0 0 0 0>
      }
      tintColor {colorUseRule=UseColor_Specified, specifiedColor=<rgba 0.0176 0.0176 0.0176 1>}
      uVRegion {bIsValid=false, max=<0 0>, min=<0 0>}
     }
     parentRowBackgroundBrush {
      bIsDynamicallyLoaded=false
      drawAs=Image
      imageType=NoImage
      mirroring=NoMirror
      resourceName=None
      resourceObject=null
      tiling=NoTile
      imageSize=<32 32>
      margin=<ltrb 0 0 0 0>
      outlineSettings {
       bUseBrushTransparency=false
       roundingType=HalfHeightRadius
       width=0
       color {colorUseRule=UseColor_Specified, specifiedColor=<rgba 0 0 0 0>}
       cornerRadii=<0 0 0 0>
      }
      tintColor {colorUseRule=UseColor_Specified, specifiedColor=<rgba 1 1 1 1>}
      uVRegion {bIsValid=false, max=<0 0>, min=<0 0>}
     }
     parentRowBackgroundHoveredBrush {
      bIsDynamicallyLoaded=false
      drawAs=Image
      imageType=NoImage
      mirroring=NoMirror
      resourceName=None
      resourceObject=null
      tiling=NoTile
      imageSize=<32 32>
      margin=<ltrb 0 0 0 0>
      outlineSettings {
       bUseBrushTransparency=false
       roundingType=HalfHeightRadius
       width=0
       color {colorUseRule=UseColor_Specified, specifiedColor=<rgba 0 0 0 0>}
       cornerRadii=<0 0 0 0>
      }
      tintColor {colorUseRule=UseColor_Specified, specifiedColor=<rgba 1 1 1 1>}
      uVRegion {bIsValid=false, max=<0 0>, min=<0 0>}
     }
     selectedTextColor {colorUseRule=UseColor_Specified, specifiedColor=<rgba 1 1 1 1>}
     selectorFocusedBrush {
      bIsDynamicallyLoaded=false
      drawAs=Border
      imageType=FullColor
      mirroring=NoMirror
      resourceName=../../../Engine/Content/Slate/Common/Selector.png
      resourceObject=null
      tiling=Both
      imageSize=<0 0>
      margin=<ltrb 0.25 0.25 0.25 0.25>
      outlineSettings {
       bUseBrushTransparency=false
       roundingType=HalfHeightRadius
       width=0
       color {colorUseRule=UseColor_Specified, specifiedColor=<rgba 0 0 0 0>}
       cornerRadii=<0 0 0 0>
      }
      tintColor {colorUseRule=UseColor_Specified, specifiedColor=<rgba 0.701 0.225 0.003 1>}
      uVRegion {bIsValid=false, max=<0 0>, min=<0 0>}
     }
     textColor {colorUseRule=UseColor_Specified, specifiedColor=<rgba 0.5271 0.5271 0.5271 1>}
    }
    widgetStyle {
     comboButtonStyle {
      downArrowAlign=VAlign_Center
      buttonStyle {
       clickedSlateSound {resourceObject=null}
       disabled {
        bIsDynamicallyLoaded=false
        drawAs=RoundedBox
        imageType=NoImage
        mirroring=NoMirror
        resourceName=None
        resourceObject=null
        tiling=NoTile
        imageSize=<0 0>
        margin=<ltrb 0 0 0 0>
        outlineSettings {
         bUseBrushTransparency=false
         roundingType=FixedRadius
         width=1
         color {colorUseRule=UseColor_ColorTable, specifiedColor=<rgba 0 0 0 0>}
         cornerRadii=<4 4 4 4>
        }
        tintColor {colorUseRule=UseColor_Specified, specifiedColor=<rgba 0.0395 0.0395 0.0395 1>}
        uVRegion {bIsValid=false, max=<0 0>, min=<0 0>}
       }
       disabledForeground {colorUseRule=UseColor_ColorTable, specifiedColor=<rgba 0 0 0 0>}
       hovered {
        bIsDynamicallyLoaded=false
        drawAs=RoundedBox
        imageType=NoImage
        mirroring=NoMirror
        resourceName=None
        resourceObject=null
        tiling=NoTile
        imageSize=<0 0>
        margin=<ltrb 0 0 0 0>
        outlineSettings {
         bUseBrushTransparency=false
         roundingType=FixedRadius
         width=1
         color {colorUseRule=UseColor_ColorTable, specifiedColor=<rgba 0 0 0 0>}
         cornerRadii=<4 4 4 4>
        }
        tintColor {colorUseRule=UseColor_Specified, specifiedColor=<rgba 0.0048 0.0048 0.0048 1>}
        uVRegion {bIsValid=false, max=<0 0>, min=<0 0>}
       }
       hoveredForeground {colorUseRule=UseColor_ColorTable, specifiedColor=<rgba 0 0 0 0>}
       hoveredSlateSound {resourceObject=null}
       normal {
        bIsDynamicallyLoaded=false
        drawAs=RoundedBox
        imageType=NoImage
        mirroring=NoMirror
        resourceName=None
        resourceObject=null
        tiling=NoTile
        imageSize=<0 0>
        margin=<ltrb 0 0 0 0>
        outlineSettings {
         bUseBrushTransparency=false
         roundingType=FixedRadius
         width=1.1
         color {colorUseRule=UseColor_ColorTable, specifiedColor=<rgba 0 0 0 0>}
         cornerRadii=<4 4 4 4>
        }
        tintColor {colorUseRule=UseColor_Specified, specifiedColor=<rgba 0.0048 0.0048 0.0048 0.25>}
        uVRegion {bIsValid=false, max=<0 0>, min=<0 0>}
       }
       normalForeground {colorUseRule=UseColor_ColorTable, specifiedColor=<rgba 0 0 0 0>}
       normalPadding=<ltrb 8 1 8 1>
       pressed {
        bIsDynamicallyLoaded=false
        drawAs=RoundedBox
        imageType=NoImage
        mirroring=NoMirror
        resourceName=None
        resourceObject=null
        tiling=NoTile
        imageSize=<0 0>
        margin=<ltrb 0 0 0 0>
        outlineSettings {
         bUseBrushTransparency=false
         roundingType=FixedRadius
         width=1
         color {colorUseRule=UseColor_ColorTable, specifiedColor=<rgba 0 0 0 0>}
         cornerRadii=<4 4 4 4>
        }
        tintColor {colorUseRule=UseColor_Specified, specifiedColor=<rgba 0.0048 0.0048 0.0048 1>}
        uVRegion {bIsValid=false, max=<0 0>, min=<0 0>}
       }
       pressedForeground {colorUseRule=UseColor_ColorTable, specifiedColor=<rgba 0 0 0 0>}
       pressedPadding=<ltrb 8 2 8 0>
       pressedSlateSound {resourceObject=null}
      }
      contentPadding=<ltrb 0 0 0 0>
      downArrowImage {
       bIsDynamicallyLoaded=false
       drawAs=Image
       imageType=Vector
       mirroring=NoMirror
       resourceName=../../../Engine/Content/Slate/Starship/CoreWidgets/ComboBox/wide-chevron-down.svg
       resourceObject=null
       tiling=NoTile
       imageSize=<16 16>
       margin=<ltrb 0 0 0 0>
       outlineSettings {
        bUseBrushTransparency=false
        roundingType=HalfHeightRadius
        width=0
        color {colorUseRule=UseColor_Specified, specifiedColor=<rgba 0 0 0 0>}
        cornerRadii=<0 0 0 0>
       }
       tintColor {colorUseRule=UseColor_Specified, specifiedColor=<rgba 1 1 1 1>}
       uVRegion {bIsValid=false, max=<0 0>, min=<0 0>}
      }
      downArrowPadding=<ltrb 2 2 2 2>
      menuBorderBrush {
       bIsDynamicallyLoaded=false
       drawAs=RoundedBox
       imageType=NoImage
       mirroring=NoMirror
       resourceName=None
       resourceObject=null
       tiling=NoTile
       imageSize=<0 0>
       margin=<ltrb 0 0 0 0>
       outlineSettings {
        bUseBrushTransparency=false
        roundingType=FixedRadius
        width=2
        color {colorUseRule=UseColor_ColorTable, specifiedColor=<rgba 0 0 0 0>}
        cornerRadii=<0 0 0 0>
       }
       tintColor {colorUseRule=UseColor_Specified, specifiedColor=<rgba 0.0395 0.0395 0.0395 1>}
       uVRegion {bIsValid=false, max=<0 0>, min=<0 0>}
      }
      menuBorderPadding=<ltrb 0 0 0 0>
      shadowColorAndOpacity=<rgba 0 0 0 1>
      shadowOffset=<0 0>
     }
     contentPadding=<ltrb 0 0 0 0>
     menuRowPadding=<ltrb 8 3 8 3>
     pressedSlateSound {resourceObject=null}
     selectionChangeSlateSound {resourceObject=null}
    }
   }
   slot {class=HorizontalBoxSlot, properties {horizontalAlignment=HAlign_Left, verticalAlignment=VAlign_Center}}
  }
  HorizontalBox_SwapDist {
   widgetClassPath=&HorizontalBox
   bIsVariable=true
   parent=HorizontalBox_ViewOptions
   widget=HorizontalBox_SwapDist
   slot {class=HorizontalBoxSlot, properties {padding=<ltrb 0 0 10 0>}}
  }
  Overlay_1 {widgetClassPath=&Overlay, parent=HorizontalBox_SwapDist, widget=Overlay_1, slot {class=HorizontalBoxSlot}}
  Image_1 {
   widgetClassPath=&Image
   parent=Overlay_1
   widget=Image_1
   properties {
    brush {
     bIsDynamicallyLoaded=false
     drawAs=Box
     imageType=NoImage
     mirroring=NoMirror
     resourceName=None
     resourceObject=&Impostor_Button
     tiling=NoTile
     imageSize=<24 28>
     margin=<ltrb 0.5 0.5 0.5 0.5>
     outlineSettings {
      bUseBrushTransparency=true
      roundingType=FixedRadius
      width=8
      color {colorUseRule=UseColor_Foreground, specifiedColor=<rgba 2 2 0 1>}
      cornerRadii=<0.2 0.2 0.2 0.2>
     }
     tintColor {colorUseRule=UseColor_Specified, specifiedColor=<rgba 1 1 1 1>}
     uVRegion {bIsValid=false, max=<0 0>, min=<0 0>}
    }
    colorAndOpacity=<rgba 0.1 0.1 0.1 0.75>
   }
   slot {class=OverlaySlot, properties {horizontalAlignment=HAlign_Fill, verticalAlignment=VAlign_Fill}}
  }
  HorizontalBox_3 {
   widgetClassPath=&HorizontalBox
   parent=Overlay_1
   widget=HorizontalBox_3
   slot {class=OverlaySlot, properties {verticalAlignment=VAlign_Center, padding=<ltrb 10 0 10 0>}}
  }
  TextBlock_5 {
   widgetClassPath=&TextBlock
   parent=HorizontalBox_3
   widget=TextBlock_5
   properties {
    text=Swap Distance:
    colorAndOpacity {colorUseRule=UseColor_Specified, specifiedColor=<rgba 0.5312 0.5312 0.5312 1>}
    font {
     bForceMonospaced=false
     bMaterialIsStencil=false
     fontMaterial=null
     fontObject=&Roboto
     letterSpacing=0
     monospacedWidth=1
     size=10
     skewAmount=0
     typefaceFontName=Regular
     outlineSettings {
      bApplyOutlineToDropShadows=false
      bMiteredCorners=false
      bSeparateFillAlpha=false
      outlineMaterial=null
      outlineSize=0
      outlineColor=<rgba 0 0 0 1>
     }
    }
    shadowColorAndOpacity=<rgba 0 0 0 0.3>
   }
   slot {class=HorizontalBoxSlot, properties {verticalAlignment=VAlign_Center, padding=<ltrb 5 0 5 0>}}
  }
  Text_SwapDistance {
   widgetClassPath=&EditorUtilityEditableTextBox
   bIsVariable=true
   parent=HorizontalBox_3
   widget=Text_SwapDistance
   properties {
    text=500
    widgetStyle {
     backgroundColor {colorUseRule=UseColor_Specified, specifiedColor=<rgba 1 1 1 1>}
     backgroundImageFocused {
      bIsDynamicallyLoaded=false
      drawAs=RoundedBox
      imageType=NoImage
      mirroring=NoMirror
      resourceName=None
      resourceObject=null
      tiling=NoTile
      imageSize=<0 0>
      margin=<ltrb 0 0 0 0>
      outlineSettings {
       bUseBrushTransparency=false
       roundingType=FixedRadius
       width=1
       color {colorUseRule=UseColor_ColorTable, specifiedColor=<rgba 0 0 0 0>}
       cornerRadii=<4 4 4 4>
      }
      tintColor {colorUseRule=UseColor_Specified, specifiedColor=<rgba 0.0048 0.0048 0.0048 1>}
      uVRegion {bIsValid=false, max=<0 0>, min=<0 0>}
     }
     backgroundImageHovered {
      bIsDynamicallyLoaded=false
      drawAs=RoundedBox
      imageType=NoImage
      mirroring=NoMirror
      resourceName=None
      resourceObject=null
      tiling=NoTile
      imageSize=<0 0>
      margin=<ltrb 0 0 0 0>
      outlineSettings {
       bUseBrushTransparency=false
       roundingType=FixedRadius
       width=1
       color {colorUseRule=UseColor_ColorTable, specifiedColor=<rgba 0 0 0 0>}
       cornerRadii=<4 4 4 4>
      }
      tintColor {colorUseRule=UseColor_Specified, specifiedColor=<rgba 0.0048 0.0048 0.0048 1>}
      uVRegion {bIsValid=false, max=<0 0>, min=<0 0>}
     }
     backgroundImageNormal {
      bIsDynamicallyLoaded=false
      drawAs=RoundedBox
      imageType=NoImage
      mirroring=NoMirror
      resourceName=None
      resourceObject=null
      tiling=NoTile
      imageSize=<0 0>
      margin=<ltrb 0 0 0 0>
      outlineSettings {
       bUseBrushTransparency=false
       roundingType=FixedRadius
       width=1
       color {colorUseRule=UseColor_ColorTable, specifiedColor=<rgba 0 0 0 0>}
       cornerRadii=<4 4 4 4>
      }
      tintColor {colorUseRule=UseColor_Specified, specifiedColor=<rgba 0.0048 0.0048 0.0048 0.25>}
      uVRegion {bIsValid=false, max=<0 0>, min=<0 0>}
     }
     backgroundImageReadOnly {
      bIsDynamicallyLoaded=false
      drawAs=RoundedBox
      imageType=NoImage
      mirroring=NoMirror
      resourceName=None
      resourceObject=null
      tiling=NoTile
      imageSize=<0 0>
      margin=<ltrb 0 0 0 0>
      outlineSettings {
       bUseBrushTransparency=false
       roundingType=FixedRadius
       width=1
       color {colorUseRule=UseColor_ColorTable, specifiedColor=<rgba 0 0 0 0>}
       cornerRadii=<4 4 4 4>
      }
      tintColor {colorUseRule=UseColor_Specified, specifiedColor=<rgba 0.0284 0.0284 0.0284 1>}
      uVRegion {bIsValid=false, max=<0 0>, min=<0 0>}
     }
     focusedForegroundColor {colorUseRule=UseColor_Specified, specifiedColor=<rgba 1 1 1 1>}
     foregroundColor {colorUseRule=UseColor_Specified, specifiedColor=<rgba 0.5271 0.5271 0.5271 1>}
     hScrollBarPadding=<ltrb 0 0 0 0>
     padding=<ltrb 8 3 8 4>
     readOnlyForegroundColor {colorUseRule=UseColor_Foreground, specifiedColor=<rgba 1 0 1 1>}
     scrollBarStyle {
      thickness=8
      draggedThumbImage {
       bIsDynamicallyLoaded=false
       drawAs=RoundedBox
       imageType=NoImage
       mirroring=NoMirror
       resourceName=None
       resourceObject=null
       tiling=NoTile
       imageSize=<0 0>
       margin=<ltrb 0 0 0 0>
       outlineSettings {
        bUseBrushTransparency=false
        roundingType=FixedRadius
        width=0
        color {colorUseRule=UseColor_Specified, specifiedColor=<rgba 0 0 0 0>}
        cornerRadii=<4 4 4 4>
       }
       tintColor {colorUseRule=UseColor_Specified, specifiedColor=<rgba 0.2159 0.2159 0.2159 1>}
       uVRegion {bIsValid=false, max=<0 0>, min=<0 0>}
      }
      horizontalBackgroundImage {
       bIsDynamicallyLoaded=false
       drawAs=NoDrawType
       imageType=NoImage
       mirroring=NoMirror
       resourceName=None
       resourceObject=null
       tiling=NoTile
       imageSize=<0 0>
       margin=<ltrb 0 0 0 0>
       outlineSettings {
        bUseBrushTransparency=false
        roundingType=HalfHeightRadius
        width=0
        color {colorUseRule=UseColor_Specified, specifiedColor=<rgba 0 0 0 0>}
        cornerRadii=<0 0 0 0>
       }
       tintColor {colorUseRule=UseColor_Specified, specifiedColor=<rgba 1 1 1 1>}
       uVRegion {bIsValid=false, max=<0 0>, min=<0 0>}
      }
      horizontalBottomSlotImage {
       bIsDynamicallyLoaded=false
       drawAs=NoDrawType
       imageType=NoImage
       mirroring=NoMirror
       resourceName=None
       resourceObject=null
       tiling=NoTile
       imageSize=<0 0>
       margin=<ltrb 0 0 0 0>
       outlineSettings {
        bUseBrushTransparency=false
        roundingType=HalfHeightRadius
        width=0
        color {colorUseRule=UseColor_Specified, specifiedColor=<rgba 0 0 0 0>}
        cornerRadii=<0 0 0 0>
       }
       tintColor {colorUseRule=UseColor_Specified, specifiedColor=<rgba 1 1 1 1>}
       uVRegion {bIsValid=false, max=<0 0>, min=<0 0>}
      }
      horizontalTopSlotImage {
       bIsDynamicallyLoaded=false
       drawAs=NoDrawType
       imageType=NoImage
       mirroring=NoMirror
       resourceName=None
       resourceObject=null
       tiling=NoTile
       imageSize=<0 0>
       margin=<ltrb 0 0 0 0>
       outlineSettings {
        bUseBrushTransparency=false
        roundingType=HalfHeightRadius
        width=0
        color {colorUseRule=UseColor_Specified, specifiedColor=<rgba 0 0 0 0>}
        cornerRadii=<0 0 0 0>
       }
       tintColor {colorUseRule=UseColor_Specified, specifiedColor=<rgba 1 1 1 1>}
       uVRegion {bIsValid=false, max=<0 0>, min=<0 0>}
      }
      hoveredThumbImage {
       bIsDynamicallyLoaded=false
       drawAs=RoundedBox
       imageType=NoImage
       mirroring=NoMirror
       resourceName=None
       resourceObject=null
       tiling=NoTile
       imageSize=<0 0>
       margin=<ltrb 0 0 0 0>
       outlineSettings {
        bUseBrushTransparency=false
        roundingType=FixedRadius
        width=0
        color {colorUseRule=UseColor_Specified, specifiedColor=<rgba 0 0 0 0>}
        cornerRadii=<4 4 4 4>
       }
       tintColor {colorUseRule=UseColor_Specified, specifiedColor=<rgba 0.2159 0.2159 0.2159 1>}
       uVRegion {bIsValid=false, max=<0 0>, min=<0 0>}
      }
      normalThumbImage {
       bIsDynamicallyLoaded=false
       drawAs=RoundedBox
       imageType=NoImage
       mirroring=NoMirror
       resourceName=None
       resourceObject=null
       tiling=NoTile
       imageSize=<0 0>
       margin=<ltrb 0 0 0 0>
       outlineSettings {
        bUseBrushTransparency=false
        roundingType=FixedRadius
        width=0
        color {colorUseRule=UseColor_Specified, specifiedColor=<rgba 0 0 0 0>}
        cornerRadii=<4 4 4 4>
       }
       tintColor {colorUseRule=UseColor_Specified, specifiedColor=<rgba 0.0953 0.0953 0.0953 1>}
       uVRegion {bIsValid=false, max=<0 0>, min=<0 0>}
      }
      verticalBackgroundImage {
       bIsDynamicallyLoaded=false
       drawAs=NoDrawType
       imageType=NoImage
       mirroring=NoMirror
       resourceName=None
       resourceObject=null
       tiling=NoTile
       imageSize=<0 0>
       margin=<ltrb 0 0 0 0>
       outlineSettings {
        bUseBrushTransparency=false
        roundingType=HalfHeightRadius
        width=0
        color {colorUseRule=UseColor_Specified, specifiedColor=<rgba 0 0 0 0>}
        cornerRadii=<0 0 0 0>
       }
       tintColor {colorUseRule=UseColor_Specified, specifiedColor=<rgba 1 1 1 1>}
       uVRegion {bIsValid=false, max=<0 0>, min=<0 0>}
      }
      verticalBottomSlotImage {
       bIsDynamicallyLoaded=false
       drawAs=NoDrawType
       imageType=NoImage
       mirroring=NoMirror
       resourceName=None
       resourceObject=null
       tiling=NoTile
       imageSize=<0 0>
       margin=<ltrb 0 0 0 0>
       outlineSettings {
        bUseBrushTransparency=false
        roundingType=HalfHeightRadius
        width=0
        color {colorUseRule=UseColor_Specified, specifiedColor=<rgba 0 0 0 0>}
        cornerRadii=<0 0 0 0>
       }
       tintColor {colorUseRule=UseColor_Specified, specifiedColor=<rgba 1 1 1 1>}
       uVRegion {bIsValid=false, max=<0 0>, min=<0 0>}
      }
      verticalTopSlotImage {
       bIsDynamicallyLoaded=false
       drawAs=NoDrawType
       imageType=NoImage
       mirroring=NoMirror
       resourceName=None
       resourceObject=null
       tiling=NoTile
       imageSize=<0 0>
       margin=<ltrb 0 0 0 0>
       outlineSettings {
        bUseBrushTransparency=false
        roundingType=HalfHeightRadius
        width=0
        color {colorUseRule=UseColor_Specified, specifiedColor=<rgba 0 0 0 0>}
        cornerRadii=<0 0 0 0>
       }
       tintColor {colorUseRule=UseColor_Specified, specifiedColor=<rgba 1 1 1 1>}
       uVRegion {bIsValid=false, max=<0 0>, min=<0 0>}
      }
     }
     textStyle {
      overflowPolicy=Clip
      transformPolicy=None
      colorAndOpacity {colorUseRule=UseColor_Foreground, specifiedColor=<rgba 1 0 1 1>}
      font {
       bForceMonospaced=false
       bMaterialIsStencil=false
       fontMaterial=null
       fontObject=null
       letterSpacing=0
       monospacedWidth=1
       size=10
       skewAmount=0
       typefaceFontName=Regular
       outlineSettings {
        bApplyOutlineToDropShadows=false
        bMiteredCorners=false
        bSeparateFillAlpha=false
        outlineMaterial=null
        outlineSize=0
        outlineColor=<rgba 0 0 0 1>
       }
      }
      highlightColor {colorUseRule=UseColor_ColorTable, specifiedColor=<rgba 0 0 0 0>}
      highlightShape {
       bIsDynamicallyLoaded=false
       drawAs=Image
       imageType=NoImage
       mirroring=NoMirror
       resourceName=None
       resourceObject=null
       tiling=NoTile
       imageSize=<0 0>
       margin=<ltrb 0 0 0 0>
       outlineSettings {
        bUseBrushTransparency=false
        roundingType=HalfHeightRadius
        width=0
        color {colorUseRule=UseColor_Specified, specifiedColor=<rgba 0 0 0 0>}
        cornerRadii=<0 0 0 0>
       }
       tintColor {colorUseRule=UseColor_ColorTable, specifiedColor=<rgba 0 0 0 0>}
       uVRegion {bIsValid=false, max=<0 0>, min=<0 0>}
      }
      selectedBackgroundColor {colorUseRule=UseColor_ColorTable, specifiedColor=<rgba 0 0 0 0>}
      shadowColorAndOpacity=<rgba 0 0 0 1>
      shadowOffset=<0 0>
      strikeBrush {
       bIsDynamicallyLoaded=false
       drawAs=Image
       imageType=NoImage
       mirroring=NoMirror
       resourceName=None
       resourceObject=null
       tiling=NoTile
       imageSize=<32 32>
       margin=<ltrb 0 0 0 0>
       outlineSettings {
        bUseBrushTransparency=false
        roundingType=HalfHeightRadius
        width=0
        color {colorUseRule=UseColor_Specified, specifiedColor=<rgba 0 0 0 0>}
        cornerRadii=<0 0 0 0>
       }
       tintColor {colorUseRule=UseColor_Specified, specifiedColor=<rgba 1 1 1 1>}
       uVRegion {bIsValid=false, max=<0 0>, min=<0 0>}
      }
      underlineBrush {
       bIsDynamicallyLoaded=false
       drawAs=Image
       imageType=NoImage
       mirroring=NoMirror
       resourceName=None
       resourceObject=null
       tiling=NoTile
       imageSize=<32 32>
       margin=<ltrb 0 0 0 0>
       outlineSettings {
        bUseBrushTransparency=false
        roundingType=HalfHeightRadius
        width=0
        color {colorUseRule=UseColor_Specified, specifiedColor=<rgba 0 0 0 0>}
        cornerRadii=<0 0 0 0>
       }
       tintColor {colorUseRule=UseColor_Specified, specifiedColor=<rgba 1 1 1 1>}
       uVRegion {bIsValid=false, max=<0 0>, min=<0 0>}
      }
     }
     vScrollBarPadding=<ltrb 0 0 0 0>
    }
   }
   slot {class=HorizontalBoxSlot}
  }
  HorizontalBox_CurDist {
   widgetClassPath=&HorizontalBox
   bIsVariable=true
   parent=HorizontalBox_ViewOptions
   widget=HorizontalBox_CurDist
   slot {class=HorizontalBoxSlot, properties {padding=<ltrb 0 0 10 0>}}
  }
  Overlay_2 {widgetClassPath=&Overlay, parent=HorizontalBox_CurDist, widget=Overlay_2, slot {class=HorizontalBoxSlot}}
  Image_2 {
   widgetClassPath=&Image
   parent=Overlay_2
   widget=Image_2
   properties {
    brush {
     bIsDynamicallyLoaded=false
     drawAs=Box
     imageType=NoImage
     mirroring=NoMirror
     resourceName=None
     resourceObject=&Impostor_Button
     tiling=NoTile
     imageSize=<24 28>
     margin=<ltrb 0.5 0.5 0.5 0.5>
     outlineSettings {
      bUseBrushTransparency=true
      roundingType=FixedRadius
      width=8
      color {colorUseRule=UseColor_Foreground, specifiedColor=<rgba 2 2 0 1>}
      cornerRadii=<0.2 0.2 0.2 0.2>
     }
     tintColor {colorUseRule=UseColor_Specified, specifiedColor=<rgba 1 1 1 1>}
     uVRegion {bIsValid=false, max=<0 0>, min=<0 0>}
    }
    colorAndOpacity=<rgba 0.1 0.1 0.1 0.75>
   }
   slot {class=OverlaySlot, properties {horizontalAlignment=HAlign_Fill, verticalAlignment=VAlign_Fill}}
  }
  HorizontalBox_5 {
   widgetClassPath=&HorizontalBox
   parent=Overlay_2
   widget=HorizontalBox_5
   slot {class=OverlaySlot, properties {verticalAlignment=VAlign_Center, padding=<ltrb 10 0 10 0>}}
  }
  TextBlock_6 {
   widgetClassPath=&TextBlock
   parent=HorizontalBox_5
   widget=TextBlock_6
   properties {
    text=Current Distance:
    colorAndOpacity {colorUseRule=UseColor_Specified, specifiedColor=<rgba 0.5312 0.5312 0.5312 1>}
    font {
     bForceMonospaced=false
     bMaterialIsStencil=false
     fontMaterial=null
     fontObject=&Roboto
     letterSpacing=0
     monospacedWidth=1
     size=10
     skewAmount=0
     typefaceFontName=Regular
     outlineSettings {
      bApplyOutlineToDropShadows=false
      bMiteredCorners=false
      bSeparateFillAlpha=false
      outlineMaterial=null
      outlineSize=0
      outlineColor=<rgba 0 0 0 1>
     }
    }
    shadowColorAndOpacity=<rgba 0 0 0 0.3>
   }
   slot {class=HorizontalBoxSlot, properties {verticalAlignment=VAlign_Center, padding=<ltrb 5 0 0 0>}}
  }
  CurrentDistanceDisplay {
   widgetClassPath=&TextBlock
   bIsVariable=true
   parent=HorizontalBox_5
   widget=CurrentDistanceDisplay
   properties {
    text=xxxxx
    font {
     bForceMonospaced=false
     bMaterialIsStencil=false
     fontMaterial=null
     fontObject=&Roboto
     letterSpacing=0
     monospacedWidth=1
     size=10
     skewAmount=0
     typefaceFontName=Regular
     outlineSettings {
      bApplyOutlineToDropShadows=false
      bMiteredCorners=false
      bSeparateFillAlpha=false
      outlineMaterial=null
      outlineSize=0
      outlineColor=<rgba 0 0 0 1>
     }
    }
   }
   slot {class=HorizontalBoxSlot, properties {horizontalAlignment=HAlign_Left, verticalAlignment=VAlign_Center, padding=<ltrb 8 4 4 4>}}
  }
  Text_ImpostorNotRenderd {
   widgetClassPath=&TextBlock
   bIsVariable=true
   parent=Overlay_46
   widget=Text_ImpostorNotRenderd
   properties {
    justification=Center
    text=Impostor not rendered
    visibility=Hidden
    colorAndOpacity {colorUseRule=UseColor_Specified, specifiedColor=<rgba 0.3906 0 0 1>}
   }
   slot {class=OverlaySlot, properties {horizontalAlignment=HAlign_Center, verticalAlignment=VAlign_Center, padding=<ltrb 0 100 0 0>}}
  }
  Text_DisplayingWhat {
   widgetClassPath=&TextBlock
   bIsVariable=true
   parent=Overlay_46
   widget=Text_DisplayingWhat
   properties {
    justification=Center
    text="Displaying: "
    visibility=HitTestInvisible
    colorAndOpacity {colorUseRule=UseColor_Specified, specifiedColor=<rgba 0.5469 0.5677 0.5376 1>}
    font {
     bForceMonospaced=false
     bMaterialIsStencil=false
     fontMaterial=null
     fontObject=&Roboto
     letterSpacing=0
     monospacedWidth=1
     size=16
     skewAmount=0
     typefaceFontName=Bold
     outlineSettings {
      bApplyOutlineToDropShadows=false
      bMiteredCorners=false
      bSeparateFillAlpha=false
      outlineMaterial=null
      outlineSize=0
      outlineColor=<rgba 0 0 0 0.3>
     }
    }
    shadowColorAndOpacity=<rgba 0 0 0 1>
   }
   slot {class=OverlaySlot, properties {horizontalAlignment=HAlign_Center, verticalAlignment=VAlign_Bottom, padding=<ltrb 0 75 0 25>}}
  }
  Text_GBuffer_Doesnt_Support_Translucency {
   widgetClassPath=&TextBlock
   bIsVariable=true
   parent=Overlay_46
   widget=Text_GBuffer_Doesnt_Support_Translucency
   properties {
    justification=Center
    text="Warning: 'Capture using GBuffer' \r\nis checked but the StaticMesh \r\nincludes translucent materials.\r\n\r\nTo capture translucency, uncheck \r\n'Capture Using GBuffer'  and then\r\nadd the 'ImpostorCaptureSwitch'\r\nmaterial function to mesh materials."
    visibility=Hidden
    colorAndOpacity {colorUseRule=UseColor_Specified, specifiedColor=<rgba 0.3906 0 0 1>}
    font {
     bForceMonospaced=false
     bMaterialIsStencil=false
     fontMaterial=null
     fontObject=&Roboto
     letterSpacing=0
     monospacedWidth=1
     size=16
     skewAmount=0
     typefaceFontName=Bold
     outlineSettings {
      bApplyOutlineToDropShadows=false
      bMiteredCorners=false
      bSeparateFillAlpha=false
      outlineMaterial=null
      outlineSize=0
      outlineColor=<rgba 0 0 0 1>
     }
    }
   }
   slot {class=OverlaySlot, properties {horizontalAlignment=HAlign_Center, verticalAlignment=VAlign_Bottom, padding=<ltrb 0 75 0 125>}}
  }
  VerticalBox_LODDialogue {
   widgetClassPath=&VerticalBox
   bIsVariable=true
   parent=Overlay_46
   widget=VerticalBox_LODDialogue
   properties {visibility=Hidden}
   slot {class=OverlaySlot, properties {horizontalAlignment=HAlign_Center, verticalAlignment=VAlign_Center}}
  }
  Overlay_133 {widgetClassPath=&Overlay, parent=VerticalBox_LODDialogue, widget=Overlay_133, slot {class=VerticalBoxSlot}}
  VerticalBox_297 {
   widgetClassPath=&VerticalBox
   parent=Overlay_133
   widget=VerticalBox_297
   slot {class=OverlaySlot, properties {horizontalAlignment=HAlign_Fill, verticalAlignment=VAlign_Fill}}
  }
  Image_3 {
   widgetClassPath=&Image
   parent=VerticalBox_297
   widget=Image_3
   properties {
    brush {
     bIsDynamicallyLoaded=false
     drawAs=Box
     imageType=NoImage
     mirroring=NoMirror
     resourceName=None
     resourceObject=/ImpostorBaker/Materials/Slate/Impostor_Button_TopOnly
     tiling=NoTile
     imageSize=<24 24>
     margin=<ltrb 0.5 0.5 0.5 0.5>
     outlineSettings {
      bUseBrushTransparency=true
      roundingType=FixedRadius
      width=8
      color {colorUseRule=UseColor_Foreground, specifiedColor=<rgba 2 2 0 1>}
      cornerRadii=<0.2 0.2 0.2 0.2>
     }
     tintColor {colorUseRule=UseColor_Specified, specifiedColor=<rgba 1 1 1 1>}
     uVRegion {bIsValid=false, max=<0 0>, min=<0 0>}
    }
    colorAndOpacity=<rgba 0.01 0.01 0.01 1>
   }
   slot {class=VerticalBoxSlot}
  }
  Image_4 {
   widgetClassPath=&Image
   parent=VerticalBox_297
   widget=Image_4
   properties {
    brush {
     bIsDynamicallyLoaded=false
     drawAs=Box
     imageType=NoImage
     mirroring=NoMirror
     resourceName=None
     resourceObject=/ImpostorBaker/Materials/Slate/Impostor_Button_BottomOnly
     tiling=NoTile
     imageSize=<24 24>
     margin=<ltrb 0.5 0.5 0.5 0.5>
     outlineSettings {
      bUseBrushTransparency=true
      roundingType=FixedRadius
      width=8
      color {colorUseRule=UseColor_Foreground, specifiedColor=<rgba 2 2 0 1>}
      cornerRadii=<0.2 0.2 0.2 0.2>
     }
     tintColor {colorUseRule=UseColor_Specified, specifiedColor=<rgba 1 1 1 1>}
     uVRegion {bIsValid=false, max=<0 0>, min=<0 0>}
    }
    colorAndOpacity=<rgba 0.1 0.1 0.1 0.875>
   }
   slot {class=VerticalBoxSlot, properties {size {sizeRule=Fill, value=1}}}
  }
  VerticalBox_206 {
   widgetClassPath=&VerticalBox
   parent=Overlay_133
   widget=VerticalBox_206
   slot {class=OverlaySlot, properties {horizontalAlignment=HAlign_Fill, verticalAlignment=VAlign_Fill}}
  }
  TextBlock_9 {
   widgetClassPath=&TextBlock
   parent=VerticalBox_206
   widget=TextBlock_9
   properties {
    text=Message:
    colorAndOpacity {colorUseRule=UseColor_Specified, specifiedColor=<rgba 0.5312 0.5312 0.5312 1>}
    font {
     bForceMonospaced=false
     bMaterialIsStencil=false
     fontMaterial=null
     fontObject=&Roboto
     letterSpacing=0
     monospacedWidth=1
     size=10
     skewAmount=0
     typefaceFontName=Regular
     outlineSettings {
      bApplyOutlineToDropShadows=false
      bMiteredCorners=false
      bSeparateFillAlpha=false
      outlineMaterial=null
      outlineSize=0
      outlineColor=<rgba 0 0 0 1>
     }
    }
    shadowColorAndOpacity=<rgba 0 0 0 0.3>
   }
   slot {class=VerticalBoxSlot, properties {horizontalAlignment=HAlign_Center, verticalAlignment=VAlign_Center, padding=<ltrb 15 5 15 5>}}
  }
  TextBlock_1 {
   widgetClassPath=&TextBlock
   parent=VerticalBox_206
   widget=TextBlock_1
   properties {
    text="StaticMesh appears to already have \r\nImpostor or Billboard as the last LOD.\r\n\r\nWhat would you like to do?"
    colorAndOpacity {colorUseRule=UseColor_Specified, specifiedColor=<rgba 0.75 0.75 0.75 1>}
    font {
     bForceMonospaced=false
     bMaterialIsStencil=false
     fontMaterial=null
     fontObject=&Roboto
     letterSpacing=0
     monospacedWidth=1
     size=10
     skewAmount=0
     typefaceFontName=Regular
     outlineSettings {
      bApplyOutlineToDropShadows=false
      bMiteredCorners=false
      bSeparateFillAlpha=false
      outlineMaterial=null
      outlineSize=0
      outlineColor=<rgba 0 0 0 1>
     }
    }
    shadowColorAndOpacity=<rgba 0 0 0 0.3>
   }
   slot {class=VerticalBoxSlot, properties {verticalAlignment=VAlign_Center, padding=<ltrb 15 15 15 15>}}
  }
  HorizontalBox_288 {widgetClassPath=&HorizontalBox, parent=VerticalBox_206, widget=HorizontalBox_288, slot {class=VerticalBoxSlot}}
  Button_ReplaceLOD {
   widgetClassPath=&EditorUtilityButton
   parent=HorizontalBox_288
   widget=Button_ReplaceLOD
   slot {class=HorizontalBoxSlot, properties {padding=<ltrb 10 10 10 10>}}
  }
  TextBlock_2 {
   widgetClassPath=&TextBlock
   parent=Button_ReplaceLOD
   widget=TextBlock_2
   properties {
    text=Replace last LOD
    font {
     bForceMonospaced=false
     bMaterialIsStencil=false
     fontMaterial=null
     fontObject=&Roboto
     letterSpacing=0
     monospacedWidth=1
     size=10
     skewAmount=0
     typefaceFontName=Bold
     outlineSettings {
      bApplyOutlineToDropShadows=false
      bMiteredCorners=false
      bSeparateFillAlpha=false
      outlineMaterial=null
      outlineSize=0
      outlineColor=<rgba 0 0 0 1>
     }
    }
   }
   slot {class=ButtonSlot, properties {padding=<ltrb 0 4 0 4>}}
  }
  Button_AddNewLOD {
   widgetClassPath=&EditorUtilityButton
   parent=HorizontalBox_288
   widget=Button_AddNewLOD
   slot {class=HorizontalBoxSlot, properties {padding=<ltrb 10 10 10 10>}}
  }
  TextBlock_7 {
   widgetClassPath=&TextBlock
   parent=Button_AddNewLOD
   widget=TextBlock_7
   properties {
    text=Add new LOD
    font {
     bForceMonospaced=false
     bMaterialIsStencil=false
     fontMaterial=null
     fontObject=&Roboto
     letterSpacing=0
     monospacedWidth=1
     size=10
     skewAmount=0
     typefaceFontName=Bold
     outlineSettings {
      bApplyOutlineToDropShadows=false
      bMiteredCorners=false
      bSeparateFillAlpha=false
      outlineMaterial=null
      outlineSize=0
      outlineColor=<rgba 0 0 0 1>
     }
    }
   }
   slot {class=ButtonSlot, properties {padding=<ltrb 0 4 0 4>}}
  }
  Button_Skip {
   widgetClassPath=&EditorUtilityButton
   parent=HorizontalBox_288
   widget=Button_Skip
   slot {class=HorizontalBoxSlot, properties {padding=<ltrb 10 10 10 10>}}
  }
  TextBlock_8 {
   widgetClassPath=&TextBlock
   parent=Button_Skip
   widget=TextBlock_8
   properties {
    text="Skip "
    font {
     bForceMonospaced=false
     bMaterialIsStencil=false
     fontMaterial=null
     fontObject=&Roboto
     letterSpacing=0
     monospacedWidth=1
     size=10
     skewAmount=0
     typefaceFontName=Bold
     outlineSettings {
      bApplyOutlineToDropShadows=false
      bMiteredCorners=false
      bSeparateFillAlpha=false
      outlineMaterial=null
      outlineSize=0
      outlineColor=<rgba 0 0 0 1>
     }
    }
   }
   slot {class=ButtonSlot, properties {padding=<ltrb 0 4 0 4>}}
  }
  VerticalBox_GeneratingImpostor {
   widgetClassPath=&VerticalBox
   bIsVariable=true
   parent=Overlay_46
   widget=VerticalBox_GeneratingImpostor
   properties {visibility=Collapsed}
   slot {class=OverlaySlot, properties {horizontalAlignment=HAlign_Center, verticalAlignment=VAlign_Center}}
  }
  Overlay_3 {widgetClassPath=&Overlay, parent=VerticalBox_GeneratingImpostor, widget=Overlay_3, slot {class=VerticalBoxSlot}}
  VerticalBox {
   widgetClassPath=&VerticalBox
   parent=Overlay_3
   widget=VerticalBox
   slot {class=OverlaySlot, properties {horizontalAlignment=HAlign_Fill, verticalAlignment=VAlign_Fill}}
  }
  Image_5 {
   widgetClassPath=&Image
   parent=VerticalBox
   widget=Image_5
   properties {
    brush {
     bIsDynamicallyLoaded=false
     drawAs=Box
     imageType=NoImage
     mirroring=NoMirror
     resourceName=None
     resourceObject=&Impostor_Button_TopOnly
     tiling=NoTile
     imageSize=<24 24>
     margin=<ltrb 0.5 0.5 0.5 0.5>
     outlineSettings {
      bUseBrushTransparency=true
      roundingType=FixedRadius
      width=8
      color {colorUseRule=UseColor_Foreground, specifiedColor=<rgba 2 2 0 1>}
      cornerRadii=<0.2 0.2 0.2 0.2>
     }
     tintColor {colorUseRule=UseColor_Specified, specifiedColor=<rgba 1 1 1 1>}
     uVRegion {bIsValid=false, max=<0 0>, min=<0 0>}
    }
    colorAndOpacity=<rgba 0.01 0.01 0.01 1>
   }
   slot {class=VerticalBoxSlot}
  }
  Image_6 {
   widgetClassPath=&Image
   parent=VerticalBox
   widget=Image_6
   properties {
    brush {
     bIsDynamicallyLoaded=false
     drawAs=Box
     imageType=NoImage
     mirroring=NoMirror
     resourceName=None
     resourceObject=&Impostor_Button_BottomOnly
     tiling=NoTile
     imageSize=<24 24>
     margin=<ltrb 0.5 0.5 0.5 0.5>
     outlineSettings {
      bUseBrushTransparency=true
      roundingType=FixedRadius
      width=8
      color {colorUseRule=UseColor_Foreground, specifiedColor=<rgba 2 2 0 1>}
      cornerRadii=<0.2 0.2 0.2 0.2>
     }
     tintColor {colorUseRule=UseColor_Specified, specifiedColor=<rgba 1 1 1 1>}
     uVRegion {bIsValid=false, max=<0 0>, min=<0 0>}
    }
    colorAndOpacity=<rgba 0.1 0.1 0.1 0.875>
   }
   slot {class=VerticalBoxSlot, properties {size {sizeRule=Fill, value=1}}}
  }
  VerticalBox_1 {
   widgetClassPath=&VerticalBox
   parent=Overlay_3
   widget=VerticalBox_1
   slot {class=OverlaySlot, properties {horizontalAlignment=HAlign_Fill, verticalAlignment=VAlign_Fill}}
  }
  TextBlock_14 {
   widgetClassPath=&TextBlock
   parent=VerticalBox_1
   widget=TextBlock_14
   properties {
    text=Generating Impostor:
    colorAndOpacity {colorUseRule=UseColor_Specified, specifiedColor=<rgba 0.5312 0.5312 0.5312 1>}
    font {
     bForceMonospaced=false
     bMaterialIsStencil=false
     fontMaterial=null
     fontObject=&Roboto
     letterSpacing=0
     monospacedWidth=1
     size=10
     skewAmount=0
     typefaceFontName=Regular
     outlineSettings {
      bApplyOutlineToDropShadows=false
      bMiteredCorners=false
      bSeparateFillAlpha=false
      outlineMaterial=null
      outlineSize=0
      outlineColor=<rgba 0 0 0 1>
     }
    }
    shadowColorAndOpacity=<rgba 0 0 0 0.3>
   }
   slot {class=VerticalBoxSlot, properties {horizontalAlignment=HAlign_Center, verticalAlignment=VAlign_Center, padding=<ltrb 15 5 15 5>}}
  }
  TextBlock_15 {
   widgetClassPath=&TextBlock
   parent=VerticalBox_1
   widget=TextBlock_15
   properties {
    text="Current Operation:                       "
    colorAndOpacity {colorUseRule=UseColor_Specified, specifiedColor=<rgba 0.75 0.75 0.75 1>}
    font {
     bForceMonospaced=false
     bMaterialIsStencil=false
     fontMaterial=null
     fontObject=&Roboto
     letterSpacing=0
     monospacedWidth=1
     size=10
     skewAmount=0
     typefaceFontName=Regular
     outlineSettings {
      bApplyOutlineToDropShadows=false
      bMiteredCorners=false
      bSeparateFillAlpha=false
      outlineMaterial=null
      outlineSize=0
      outlineColor=<rgba 0 0 0 1>
     }
    }
    shadowColorAndOpacity=<rgba 0 0 0 0.3>
   }
   slot {class=VerticalBoxSlot, properties {verticalAlignment=VAlign_Center, padding=<ltrb 25 15 25 15>}}
  }
  Progress_CurrentMap {
   widgetClassPath=&TextBlock
   bIsVariable=true
   parent=VerticalBox_1
   widget=Progress_CurrentMap
   properties {
    text=Capturing BaseColor
    colorAndOpacity {colorUseRule=UseColor_Specified, specifiedColor=<rgba 0.75 0.75 0.75 1>}
    font {
     bForceMonospaced=false
     bMaterialIsStencil=false
     fontMaterial=null
     fontObject=&Roboto
     letterSpacing=0
     monospacedWidth=1
     size=10
     skewAmount=0
     typefaceFontName=Regular
     outlineSettings {
      bApplyOutlineToDropShadows=false
      bMiteredCorners=false
      bSeparateFillAlpha=false
      outlineMaterial=null
      outlineSize=0
      outlineColor=<rgba 0 0 0 1>
     }
    }
    shadowColorAndOpacity=<rgba 0 0 0 0.3>
   }
   slot {class=VerticalBoxSlot, properties {verticalAlignment=VAlign_Center, padding=<ltrb 25 0 25 10>}}
  }
  EditorUtilityProgressBar_93 {
   widgetClassPath=/Script/Blutility.EditorUtilityProgressBar
   bIsVariable=true
   parent=VerticalBox_1
   widget=EditorUtilityProgressBar_93
   properties {percent=0.5, fillColorAndOpacity=<rgba 0 0.5 1 1>}
   slot {class=VerticalBoxSlot, properties {verticalAlignment=VAlign_Top, padding=<ltrb 15 0 15 15>}}
  }
  HorizontalBox_2 {widgetClassPath=&HorizontalBox, parent=VerticalBox_1, widget=HorizontalBox_2, slot {class=VerticalBoxSlot}}
 ]
}
