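# Editor utility widget that batch-generates and saves impostors for a list of static meshes, using a settings preset chosen in its details view.
# The render button opens EUW_Impostor_Editor in a tab and feeds it each mesh with the preset and existing-LOD behavior.
# Render-complete and save-complete delegates chain rendering, asset creation and the next index, closing the tab after the last mesh.
# setAsset fills the mesh list; construct loads the default preset from an asset path.
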EditorUtilityWidgetBlueprint /ImpostorBaker/BP/EUW_Generate_Impostor_using_Preset
graphs [
 EventGraph {nodes [
   "On Clicked (Button_Render)"() {is=title K2Node_ComponentBoundEvent}
   > n1 = SpawnAndRegisterTabAndGetID!(
    execute
    self:Editor Utility Subsystem Object Reference=Get EditorUtilitySubsystem()
    InBlueprint:Editor Utility Widget Blueprint Object Reference=&/EUW_Impostor_Editor
   ) -> (then, NewTabID:Name=None, ReturnValue:Editor Utility Widget Object Reference)
   > Set ID(ID:Name=@n1) -> (Output_Get:Name=None)
   > n2 = Cast To EUW_Impostor_Editor!(execute, Object=@n1.ReturnValue) -> (then, CastFailed, AsEUW Impostor Editor:EUW Impostor Editor Object Reference) {is=title K2Node_DynamicCast}
   > Set Impostor Editor(Impostor Editor:EUW Impostor Editor Object Reference=@n2) -> (Output_Get:EUW Impostor Editor Object Reference)
   > Set Current Index(Current Index:int=0) -> (Output_Get:int=0)
   > Render Impostor()
   n3 = Save Complete() {is=title K2Node_CustomEvent}
   > PrintString(InString=Bound event 'Impostor Render Save Complete' triggered on PresetRenderer)
   > Increment Int(Value=Get Current Index() -> (Current Index:int=0))
   > Render Impostor()
   n4 = Render Completed() {is=title K2Node_CustomEvent}
   > PrintString(InString=Bound event 'Impostor Render Complete' triggered on PresetRenderer)
   > Bind Event to Impostor Save Complete(self:EUW Impostor Editor Object Reference=Get Impostor Editor() -> (Impostor Editor:EUW Impostor Editor Object Reference), Delegate=@n3) {is=title K2Node_AddDelegate}
   > Create Static Assets(self:EUW Impostor Editor Object Reference=Get Impostor Editor())
   Render Impostor() {is=title K2Node_CustomEvent}
   n5 = Get StaticMeshes() -> (StaticMeshes:Array of Static Mesh Object References)
   > n6 = Branch(Condition="integer <= integer"(A=Get Current Index() -> (Current Index=0), B=Last Index!(TargetArray:Array of Static Mesh Object References=@n5) -> (ReturnValue=0)))
   > setAsset!(
    execute
    self:EUW Impostor Editor Object Reference=Get Impostor Editor()
    Object:Object Reference="Get (a copy)"(Array:Array of Static Mesh Object References=Get StaticMeshes(), Dimension 1=Get Current Index() -> (Current Index=0)) -> (Output:Static Mesh Object Reference)
   ) -> (then)
   > Set SettingsPreset!(
    execute
    SettingsPreset:Impostor Presets Object Reference=Get SettingsPreset() -> (SettingsPreset:Impostor Presets Object Reference)
    self:EUW Impostor Editor Object Reference=Get Impostor Editor()
   ) -> (then, Output_Get:Impostor Presets Object Reference)
   > Apply Settings Preset(self:EUW Impostor Editor Object Reference=Get Impostor Editor())
   > Set ExistingImpostorFoundBehavior!(
    execute
    ExistingImpostorFoundBehavior:EImpostorOverwriteLOD Enum=Get What to do if ImpostorLOD Found() -> (What to do if ImpostorLOD Found:EImpostorOverwriteLOD Enum=NewEnumerator0)
    self:EUW Impostor Editor Object Reference=Get Impostor Editor()
   ) -> (then, Output_Get:EImpostorOverwriteLOD Enum=NewEnumerator0)
   > Bind Event to Impostor Render Complete(self:EUW Impostor Editor Object Reference=Get Impostor Editor(), Delegate=@n4) {is=title K2Node_AddDelegate}
   > RenderFrames(self:EUW Impostor Editor Object Reference=Get Impostor Editor())
   CloseTabByID!(execute=@n6.else, self:Editor Utility Subsystem Object Reference=Get EditorUtilitySubsystem(), NewTabID:Name=Get ID() -> (ID:Name=None)) -> (then, ReturnValue:bool)
   Event Construct()
   > Get Settings Preset from INI!(execute:Exec) -> (OutputPin:Exec) {is=title K2Node_MacroInstance}
   Event PreConstruct()
   > SetObject(self=Get DetailsView_159() -> (DetailsView_159:Details View Object Reference), NewObject=Self-Reference())
  ]}
 GetSettingsPresetFromINI {nodes [
   Inputs()
   > n1 = LoadAsset(AssetPath=Get DefaultSettingsAssetPath() -> (DefaultSettingsAssetPath:String))
   > n2 = Cast To ImpostorPresets!(execute, Object=@n1) -> (then, CastFailed, AsImpostor Presets:Impostor Presets Object Reference) {is=title K2Node_DynamicCast}
   > n3 = Set SettingsPreset!(execute, SettingsPreset=@n2) -> (then, Output_Get)
   Outputs!(OutputPin:Exec=[@n3.then, @n2.CastFailed])
  ]}
 setAsset {nodes [
   n1 = Clear(TargetArray:Array of Static Mesh Object References=Get Meshes() -> (Meshes:Array of Static Mesh Object References))
   n2 = setAsset!() -> (then, Object:Object Reference) {is=title K2Node_FunctionEntry}
   n3 = Cast To StaticMesh(execute=@n1.then, Object=@n2)
   > Add(TargetArray:Array of Static Mesh Object References=Get Meshes(), NewItem:"Static Mesh Object Reference (by ref)"=@n3) {is=title K2Node_CallArrayFunction}
   > Set StaticMeshes(StaticMeshes:Array of Static Mesh Object References=Get Meshes()) -> (Output_Get:Array of Static Mesh Object References)
   n4 = GetSelectedAssets(execute=@n2.then)
   n5 = For Each Loop(Exec=@n7.then, Array:Array of Object References=@n4) -> (Array Element:Object Reference)
   > n6 = Cast To StaticMesh(Object=@n5)
   > Add(TargetArray=Get Meshes(), NewItem=@n6)
   Set StaticMeshes(execute=@n5.Completed, StaticMeshes=Get Meshes())
   n7 = Clear(execute=@n4.then, TargetArray=Get Meshes())
  ]}
]
variables [
 Current Index {type=integer}
 DefaultSettingsAssetPath {type=string, default=/Game/ImpostorPresets_EngineTest}
 ID {type=string, default=None}
 Impostor Editor {type=EUW_Impostor_Editor_C, default=null}
 SettingsPreset {type=ImpostorPresets_C, default=/ImpostorBaker/BP/DataAsset/ImpostorPresets_Default}
 StaticMeshes {type="array<StaticMesh>", default []}
 What to do if ImpostorLOD Found {type=EImpostorOverwriteLOD, default=Present Dialogue}
]
widgets {
 info {rootWidgetClass=&HorizontalBox, widgetCount=6}
 widgets [
  HorizontalBox_168 {widgetClassPath=&HorizontalBox, widget=HorizontalBox_168}
  VerticalBox_950 {
   widgetClassPath=&VerticalBox
   parent=HorizontalBox_168
   widget=VerticalBox_950
   slot {class=HorizontalBoxSlot, properties {size {sizeRule=Fill, value=1}}}
  }
  DetailsView_159 {
   widgetClassPath=&DetailsView
   parent=VerticalBox_950
   widget=DetailsView_159
   properties {categoriesToShow [Bake Settings]}
   slot {class=VerticalBoxSlot}
  }
  VerticalBox_1242 {
   widgetClassPath=&VerticalBox
   parent=VerticalBox_950
   widget=VerticalBox_1242
   slot {class=VerticalBoxSlot, properties {size {sizeRule=Fill, value=1}}}
  }
  Button_Render {
   widgetClassPath=&EditorUtilityButton
   parent=VerticalBox_1242
   widget=Button_Render
   slot {class=VerticalBoxSlot, properties {padding=<ltrb 15 15 15 5>}}
  }
  TextBlock_87 {
   widgetClassPath=&TextBlock
   parent=Button_Render
   widget=TextBlock_87
   properties {
    text=Generate and Save Impostors
    font {
     bForceMonospaced=false
     bMaterialIsStencil=false
     fontMaterial=null
     fontObject=&Roboto
     letterSpacing=0
     monospacedWidth=1
     size=12
     skewAmount=0
     typefaceFontName=Bold
     outlineSettings {
      bApplyOutlineToDropShadows=false
      bMiteredCorners=false
      bSeparateFillAlpha=false
      outlineMaterial=null
      outlineSize=0
      outlineColor=<rgba 0 0 0 1>
     }
    }
   }
   slot {class=ButtonSlot, properties {padding=<ltrb 4 4 4 4>}}
  }
 ]
}
